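# Editor utility widget that serves as a floating context menu container, apparently for the AssetPlaylist tech-art tool.
# Its add Child Widget function stores the given parent as spawningWidget, then replaces the auto-sized vertical box's contents with the supplied widget.
# Mouse leave collapses the whole menu, and both mouse enter and leave print debug strings.
# The stored spawningWidget is not read within this widget.
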EditorUtilityWidgetBlueprint /Game/NeverCook/ArtTools/Tech_Art/AssetPlaylist/EUW_FloatingContextMenu
graphs [
 EventGraph {nodes [
   Event OnMouseEnter()
   > PrintString(InString=Enter)
   Event OnMouseLeave!() -> (OutputDelegate, then, MouseEvent:"Pointer Event Structure (by ref)") {is=title K2Node_Event}
   > PrintString(InString=Leave)
   > SetVisibility(self=Self-Reference(), InVisibility=Collapsed)
   Event PreConstruct()
   Event Construct()
   Event Tick()
  ]}
 add Child Widget {nodes [
   n1 = add Child Widget!() -> (then, widget:Widget Object Reference, parent:Widget Object Reference) {is=title K2Node_FunctionEntry}
   > Set spawningWidget(spawningWidget:Widget Object Reference=@n1.parent) -> (Output_Get:Widget Object Reference)
   n2 = Get VerticalBox_408() -> (VerticalBox_408:Vertical Box Object Reference)
   > ClearChildren(self=@n2)
   > AddChild(self=@n2, Content=@n1)
  ]}
]
variables [spawningWidget {type=Widget, default=null}]
widgets {
 info {rootWidgetClass=&CanvasPanel, widgetCount=2}
 widgets [
  CanvasPanel_0 {widgetClassPath=&CanvasPanel, bIsVariable=true, widget=CanvasPanel_0}
  VerticalBox_408 {
   widgetClassPath=&VerticalBox
   bIsVariable=true
   parent=CanvasPanel_0
   widget=VerticalBox_408
   slot {class=CanvasPanelSlot, properties {bAutoSize=true, layoutData=~}}
  }
 ]
}
