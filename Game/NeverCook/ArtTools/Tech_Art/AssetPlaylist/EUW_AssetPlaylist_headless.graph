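# Headless Editor Utility Widget that serves as a function library for the Asset Playlist tech-art tool for cosmetics, with an empty canvas and event graph.
# Embedded Python calls CosmeticPlaylist and CosmeticPicker modules to fetch ShotGrid assets by playlist, release, user or group, packed as comma-separated strings and parsed into structs.
# Asset Registry searches match cosmetic item definitions by ShotGrid ID, variant ID or name.
# Helpers cover release data, variant tags, slow-task progress and floating context menus.
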EditorUtilityWidgetBlueprint /Game/NeverCook/ArtTools/Tech_Art/AssetPlaylist/EUW_AssetPlaylist_headless
graphs [
 DaysBetweenDates {nodes [
   n1 = DaysBetweenDates!() -> (then, YearA:int, MonthA:int, DayA:int, YearB:int, MonthB:int, DayB:int) {is=title K2Node_FunctionEntry}
   > n2 = Execute Python Script!(
    execute
    PythonScript="from datetime import date\r\nd0 = date(year_a, month_a, day_a)\r\nd1 = date(year_b, month_b, day_b)\r\n\r\nprint (d0)\r\nprint (d1)\r\n\r\ndelta = d1 - d0\r\n\r\ndays = delta.days"
    year_a:int=@n1
    month_a:int=@n1.MonthA
    day_a:int=@n1.DayA
    year_b:int=@n1.YearB
    month_b:int=@n1.MonthB
    day_b:int=@n1.DayB
   ) -> (then, ReturnValue, days:int) {inputs [year_a,month_a,day_a,year_b,month_b,day_b], outputs [days]}
   > Return Node!(execute, days:int=@n2.days)
  ]}
 EventGraph {nodes [
   Event PreConstruct()
   Event Construct()
   Event Tick()
  ]}
 Get Lock Dates {nodes [
   n1 = Get Lock Dates!() -> (then, release:String) {is=title K2Node_FunctionEntry}
   > n2 = Get Data Table Row!(
    execute:Exec
    DataTable:Data Table Object Reference=Get dataTable() -> (dataTable:Data Table Object Reference)
    RowName:Name=String To Name(InString=@n1)
   ) -> (then:Exec, RowNotFound:Exec, ReturnValue:Asset Playlist Releases Structure Structure)
   > Return Node!(execute, releaseStructure:Asset Playlist Releases Structure Structure=@n2)
  ]}
 add Floating Context Menu {nodes [
   n1 = add Floating Context Menu!() -> (then, widgetToAdd:User Widget Class Reference, spawningWidget:User Widget Object Reference) {is=title K2Node_FunctionEntry}
   > n2 = FindParentWidgetOfType!(execute, StartingWidget:Widget Object Reference=@n1.spawningWidget, Type:Widget Class Reference=&Overlay) -> (then, ReturnValue:Overlay Object Reference)
   > n3 = Cast To Overlay!(execute, Object=@n2) -> (then, CastFailed, AsOverlay:Overlay Object Reference) {is=title K2Node_DynamicCast}
   > n4 = Create Widget!(
    execute:Exec
    WorldContextObject:Object Reference
    Class:User Widget Class Reference=&/EUW_FloatingContextMenu.EUW_FloatingContextMenu_C
    OwningPlayer:Player Controller Object Reference
   ) -> (then:Exec, ReturnValue:EUW Floating Context Menu Object Reference)
   > n5 = Set floating(floating:EUW Floating Context Menu Object Reference=@n4) -> (Output_Get:EUW Floating Context Menu Object Reference)
   > AddChild(self=@n3, Content=@n5)
   > n6 = GetMousePositionOnViewport!(execute, WorldContextObject:Object Reference) -> (then, ReturnValue:Vector 2D Structure)
   > n7 = Create Widget(Class=@n1) -> (ReturnValue:User Widget Object Reference)
   n8 = Get floating() -> (floating:EUW Floating Context Menu Object Reference)
   > add Child Widget!(
    execute
    self:EUW Floating Context Menu Object Reference=@n8
    widget:Widget Object Reference=@n7
    parent:Widget Object Reference=@n1.spawningWidget
   ) -> (then)
   > SetVisibility(self=@n8, InVisibility=Visible)
   n9 = BreakVector2D(InVec="Absolute to Local (Coordinates)"!(
     Geometry:"Geometry Structure (by ref)"=GetCachedGeometry!(self:Widget Object Reference=@n3) -> (ReturnValue:"Geometry Structure (by ref)")
     AbsoluteCoordinate:Vector 2D Structure=@n6
    ) -> (ReturnValue:Vector 2D Structure))
   > SetRenderTranslation!(execute, self=[Get floating(), @n8], Translation_X:double=float - float(A=@n9, B=5), Translation_Y:double=float - float(A=@n9.Y, B=5)) -> (then)
   > Return Node!(execute, floatingWidget:EUW Floating Context Menu Object Reference=Get floating())
  ]}
 find Assets Matching Shot Grid Id {nodes [
   n1 = PrintString(InString=Append(A="num of ids found: "))
   n2 = Set found(execute=[@n1.then, @n4.then], found:bool=false) -> (Output_Get:bool)
   n3 = find Assets Matching Shot Grid Id!() -> (then, id:int, cosmeticType:CosmeticPicker_CosmeticTypes Enum=NewEnumerator0) {is=title K2Node_FunctionEntry}
   n4 = get Cosmetic Assets of CosmeticType!(execute=@n27.then, self, cosmeticType:CosmeticPicker_CosmeticTypes Enum=@n3.cosmeticType) -> (then, assets:Array of Asset Data Structures)
   n5 = Length(TargetArray:Array of Asset Data Structures=@n4)
   PrintString(execute=@n2.then, InString=Append!(A="searching ", B="To String (Integer)"(InInt=@n5), C:String=" assets") -> (ReturnValue))
   > n6 = For Each Loop(Array:Array of Asset Data Structures=@n4) -> (Array Element:Asset Data Structure)
   n7 = Branch(execute=@n6.Completed, Condition=Get found() -> (found:bool))
   Return Node!(
    execute=[@n7.then, @n9.then]
    foundAssets:Array of Asset Playlist Asset Return Structures=Get foundAssetList() -> (foundAssetList:Array of Asset Playlist Asset Return Structures)
   )
   n8 = Make Array("[0]":int=0) -> (Array:Array of Integers)
   n9 = Set foundAssetList(
    execute=@n7.else
    foundAssetList:Array of Asset Playlist Asset Return Structures=Make Array!(
     "[0]_object_2_#1":Fort Item Definition Object Reference
     "[0]_variant_5_#2":bool=False
     "[0]_variant_index_9_#3":Array of Integers=@n8
     "[0]_option_index_11_#4":Array of Integers=@n8
    ) -> (Array:Array of Asset Playlist Asset Return Structures)
   ) -> (Output_Get:Array of Asset Playlist Asset Return Structures)
   n10 = GetAsset(InAssetData=@n6)
   n11 = GetEditorProperty!(execute:Exec=@n6.LoopBody, Object:Object Reference=@n10, PropertyName:Name=BaseShotGridId) -> (then:Exec, PropertyValue:int, ReturnValue:bool)
   > n12 = Set BaseShotGridId(BaseShotGridId:int=@n11) -> (Output_Get:int=0)
   > n13 = Branch(Condition="Equal (Integer)"(A=@n12, B=@n3))
   > Set isVariant(isVariant:bool=false) -> (Output_Get:bool)
   n14 = Make Array!("[0]":int=0) -> (Array:Array of Integers)
   > Set _variantIndex(_variantIndex:Array of Integers=@n14) -> (Output_Get:Array of Integers)
   > n15 = Set _optionIndex(_optionIndex:Array of Integers=@n14) -> (Output_Get:Array of Integers)
   n16 = Cast To FortItemDefinition!(execute=[@n15.then, @n26.then], Object=@n10) -> (then, CastFailed, AsFort Item Definition:Fort Item Definition Object Reference) {is=title K2Node_DynamicCast}
   > n17 = For Each Loop(Array:Array of Integers=Get _variantIndex() -> (_variantIndex:Array of Integers)) -> (Array Element:int)
   n18 = Make Asset Playlist Asset Return!(
    object_2_#1:Fort Item Definition Object Reference=@n16
    variant_5_#2:bool=Get isVariant() -> (isVariant:bool)
    variant_index_9_#3:Array of Integers=Get _variantIndex()
    option_index_11_#4:Array of Integers=Get _optionIndex() -> (_optionIndex:Array of Integers)
   ) -> (AssetPlaylist_assetReturn:Asset Playlist Asset Return Structure) {is=title K2Node_MakeStruct}
   n19 = Branch(
    execute=@n17.Completed
    Condition=Contains Item(
     TargetArray:Array of Asset Playlist Asset Return Structures=Get foundAssetList()
     ItemToFind:"Asset Playlist Asset Return Structure (by ref)"=@n18
    )
   )
   Add(
    execute=@n19.else
    TargetArray:Array of Asset Playlist Asset Return Structures=Get foundAssetList()
    NewItem:"Asset Playlist Asset Return Structure (by ref)"=@n18
   ) {is=title K2Node_CallArrayFunction}
   > Set found(found=true)
   n20 = Get isVariant()
   n21 = "Get (a copy)"(Array:Array of Integers=Get _optionIndex(), Dimension 1=@n17.Array Index) -> (Output:int)
   PrintString(
    execute=@n17.LoopBody
    InString=Append!(
     A="found "
     B=GetDisplayName(Object=@n16)
     C=" "
     D:String="To String (Boolean)"(InBool=@n20)
     E:String=" "
     F:String="To String (Integer)"(InInt=@n17)
     G:String=" "
     H:String="To String (Integer)"(InInt=@n21)
    ) -> (ReturnValue)
   )
   n22 = GetEditorProperty(execute=@n13.else, Object=@n10, PropertyName=VariantShotGridIds) -> (PropertyValue:Array of Integers)
   > n23 = Set VariantShotGridIds(VariantShotGridIds:Array of Integers=@n22) -> (Output_Get:Array of Integers)
   > n24 = For Each Loop(Array=@n23)
   > Branch(Condition="Equal (Integer)"(A=@n24, B=@n3))
   > n25 = find Variant matching ShotGridId!(execute, self, object:Object Reference=@n10, shotGridId:int=@n3) -> (then, variantIndexArray:Array of Integers, optionIndexArray:Array of Integers)
   > Set isVariant(isVariant=true)
   > Set _variantIndex(_variantIndex=@n25)
   > n26 = Set _optionIndex(_optionIndex=@n25.optionIndexArray)
   Clear(execute=@n3.then, TargetArray:Array of Asset Playlist Asset Return Structures=Get foundAssetList())
   > n27 = PrintString(InString=Append!(
     A="searching for "
     B="To String (Integer)"(InInt=@n3)
     C=" "
     D=Enum to String!(Enumerator:CosmeticPicker_CosmeticTypes Enum=@n3.cosmeticType) -> (ReturnValue:String)
    ) -> (ReturnValue))
  ]}
 find Substring in Assets {nodes [
   n1 = find Substring in Assets!() -> (then, Array:Array of Asset Data Structures, Substring:String) {is=title K2Node_FunctionEntry}
   > PrintString(InString=Append!(A="checking for ", B=@n1.Substring) -> (ReturnValue))
   > n2 = Reverse for Each Loop!(Exec:Exec, Array:Array of Asset Data Structures=@n1) -> (LoopBody:Exec, ArrayIndex:int, ArrayElement:Asset Data Structure, Completed:Exec) {is=title K2Node_MacroInstance}
   Return Node!(execute=[@n2.Completed, @n5.Completed], OutputPin:Object Reference)
   n3 = GetAsset(InAssetData=@n2.ArrayElement)
   n4 = GetDisplayName(Object=@n3)
   > Branch(Condition=Contains(SearchIn=@n4, Substring=@n1.Substring))
   > PrintString(InString=Append(A="found ", B=@n4))
   > Return Node(OutputPin=@n3)
   n5 = For Each Loop(Array:Array of Asset Data Structures=@n1) -> (Array Element:Asset Data Structure)
   PrintString(InString=Append!(A=@n4, B=" - ", C:String=@n1.Substring) -> (ReturnValue))
  ]}
 find Variant matching ShotGridId {nodes [
   n1 = find Variant matching ShotGridId!() -> (then, object:Object Reference, shotGridId:int) {is=title K2Node_FunctionEntry}
   > n2 = Execute Python Script!(
    execute
    PythonScript="import TechArt.CosmeticPlaylist.CosmeticPlaylistModule as CosmeticPlaylist\r\nvariant_array, option_array = CosmeticPlaylist.findVariantShotGridIdOnObject(object, shot_grid_id)"
    object:Object Reference=@n1
    shot_grid_id:int=@n1.shotGridId
   ) -> (then, ReturnValue, variant_array:Array of Integers, option_array:Array of Integers) {inputs [object,shot_grid_id], outputs [variant_array, option_array]}
   > Return Node!(execute, variantIndexArray:Array of Integers=@n2.variant_array, optionIndexArray:Array of Integers=@n2.option_array)
  ]}
 get Assets from ShotGrid playlist {nodes [
   n1 = get Assets from ShotGrid playlist!() -> (then, playlist_id:int, useAssetGroups:bool=~) {is=title K2Node_FunctionEntry}
   > n2 = Execute Python Script!(
    execute
    PythonScript="import TechArt.CosmeticPlaylist.CosmeticPlaylistModule as CosmeticPlaylist\r\nassets = CosmeticPlaylist.get_ShotGridAssetsByPlaylist(playlist_id)\r\n\r\nasset_strings = []\r\n\r\nfor asset in assets:\r\n sg_dev_complete = asset[\"sg_dev_complete\"]\r\n version = \"\"\r\n if sg_dev_complete:\r\n  version = sg_dev_complete['name']\r\n else:\r\n  sg_fix_version = asset[\"sg_fix_version\"]\r\n  if sg_fix_version:\r\n   version = sg_fix_version[0]['name']\r\n\r\n assetString = (str(asset[\"id\"]) + \",\"  + str(asset[\"sg_type_brkdn\"]) + \",\" + asset[\"code\"] + \",\" + str(asset[\"sg_gender\"]) + \",\" + str(asset[\"sg_asset_group\"]) + \",\"  + str(asset[\"sg_primary_asset_group\"]) + \",\" + str(asset[\"sg_status_list\"]) + \",\" + version)\r\n\r\n print (\"assetString: \" + assetString)\r\n asset_strings.append(assetString)\r\n"
    playlist_id:int=@n1
   ) -> (then, ReturnValue, asset_strings:Array of Strings) {inputs [playlist_id], outputs [asset_strings]}
   > n3 = For Each Loop(Array:Array of Strings=@n2.asset_strings) -> (Array Element:String)
   Return Node!(
    execute=@n3.Completed
    Assets:Array of Asset Playlist Asset Structures=Get csv_assets() -> (csv_assets:Array of Asset Playlist Asset Structures)
   )
   > n4 = make Asset Playlist Asset from Asset Info!(execute, self, assetInfo:String=@n3, useAssetGroups:bool=@n1.useAssetGroups) -> (then, asset:Asset Playlist Asset Structure)
   > Add(TargetArray:Array of Asset Playlist Asset Structures=Get csv_assets(), NewItem:"Asset Playlist Asset Structure (by ref)"=@n4)
  ]}
 get Assets from ShotGrid release {nodes [
   n1 = get Assets from ShotGrid release!() -> (then, release:String, useAssetGroups:bool=~) {is=title K2Node_FunctionEntry}
   > n2 = Execute Python Script!(
    execute
    PythonScript="import TechArt.CosmeticPlaylist.CosmeticPlaylistModule as CosmeticPlaylist\r\nassets = CosmeticPlaylist.get_ShotGridAssetsByRelease(release)\r\n\r\nasset_strings = []\r\n\r\nprint (assets[0].items())\r\n\r\nfor asset in assets:\r\n sg_fix_version = asset[\"sg_fix_version\"][0]\r\n version = sg_fix_version['name']\r\n assetString = (str(asset[\"id\"]) + \",\"  + str(asset[\"sg_type_brkdn\"]) + \",\" + asset[\"code\"] + \",\" + str(asset[\"sg_gender\"]) + \",\" + str(asset[\"sg_asset_group\"]) + \",\" + str(asset[\"sg_primary_asset_group\"]) + \",\" + str(asset[\"sg_status_list\"]) + \",\" + version)\r\n\r\n print (\"assetString: \" + assetString)\r\n asset_strings.append(assetString)"
    release:String=@n1
   ) -> (then, ReturnValue, asset_strings) {inputs [release], outputs [asset_strings]}
   > n3 = For Each Loop(Array=@n2.asset_strings)
   Return Node(execute=@n3.Completed, Assets=Get csv_assets())
   > n4 = make Asset Playlist Asset from Asset Info(assetInfo=@n3, useAssetGroups=@n1.useAssetGroups)
   > Add(TargetArray=Get csv_assets(), NewItem=@n4)
  ]}
 get Assets from ShotGrid user {nodes [
   n1 = get Assets from ShotGrid user!() -> (then, username:String, useAssetGroups:bool=~) {is=title K2Node_FunctionEntry}
   n2 = Split(SourceString=@n1, InStr=" ", SearchDir=FromStart)
   > n3 = Execute Python Script!(
    execute
    PythonScript="import TechArt.CosmeticPlaylist.CosmeticPlaylistModule as CosmeticPlaylist\r\n\r\nassets = CosmeticPlaylist.getUsersTasks(firstname, lastname)\r\nasset_strings = []\r\n\r\nfor asset in assets:\r\n sg_fix_version = asset[\"sg_fix_version\"]\r\n version = \"\"\r\n if sg_fix_version:\r\n  version = sg_fix_version[0]['name']\r\n assetString = (str(asset[\"id\"]) + \",\"  + str(asset[\"sg_type_brkdn\"]) + \",\" + asset[\"code\"] + \",\" + str(asset[\"sg_gender\"]) + \",\" + str(asset[\"sg_asset_group\"]) + \",\" + str(asset[\"sg_primary_asset_group\"]) + \",\" + str(asset[\"sg_status_list\"]) + \",\" + version)\r\n\r\n print (\"assetString: \" + assetString)\r\n asset_strings.append(assetString)\r\n"
    firstname:String=@n2
    lastname:String=@n2.RightS
   ) -> (then, ReturnValue, asset_strings) {inputs [firstname,lastname], outputs [asset_strings]}
   > n4 = For Each Loop(Array=@n3.asset_strings)
   Return Node(execute=@n4.Completed, Assets=Get csv_assets())
   > n5 = make Asset Playlist Asset from Asset Info(assetInfo=@n4, useAssetGroups=@n1.useAssetGroups)
   > Add(TargetArray=Get csv_assets(), NewItem=@n5)
  ]}
 get Assets of Group {nodes [
   n1 = get Assets of Group!() -> (then, sg_asset_group:String, matchGender:bool=~, gender:String) {is=title K2Node_FunctionEntry}
   > n2 = Execute Python Script!(
    execute
    PythonScript="import TechArt.CosmeticPlaylist.CosmeticPlaylistModule as CosmeticPlaylist\r\nassets = CosmeticPlaylist.get_ShotGridAssetsByGroupName(sg_asset_group)\r\n\r\nasset_strings = []\r\n\r\nfor asset in assets:\r\n sg_fix_version = asset[\"sg_fix_version\"]\r\n if sg_fix_version:\r\n  version = sg_fix_version[0]['name']\r\n assetString = (str(asset[\"id\"]) + \",\"  + str(asset[\"sg_type_brkdn\"]) + \",\" + asset[\"code\"] + \",\" + str(asset[\"sg_gender\"]) + \",\" + str(asset[\"sg_asset_group\"]) + \",\" + str(asset[\"sg_status_list\"]) + \",\" + version)\r\n\r\n print (\"assetString: \" + assetString)\r\n asset_strings.append(assetString)"
    sg_asset_group:String=@n1
   ) -> (then, ReturnValue, asset_strings) {inputs [sg_asset_group], outputs [asset_strings]}
   n3 = Length(TargetArray:Array of Strings=@n2.asset_strings)
   > PrintString(InString=Append!(A="To String (Integer)"(InInt=@n3), B=" assetStrings") -> (ReturnValue))
   > n4 = For Each Loop(Array=@n2.asset_strings)
   Return Node(execute=@n4.Completed, Assets=Get csv_assets())
   > PrintString(InString=@n4)
   > n5 = make Asset Playlist Asset from Asset Info(assetInfo=@n4, useAssetGroups=false)
   > n6 = Branch(Condition=@n1.matchGender)
   > n7 = get Gender!(execute, self, asset:Asset Playlist Asset Structure=@n5) -> (then, gender:String)
   n8 = OR Boolean(A="Equal, Case Insensitive (String)"(A=@n7, B=@n1.gender), B="Equal, Case Insensitive (String)"(A=@n1.gender, B=None))
   > PrintString(InString=Append!(A="gender test ", B=@n7, C:String=" vs ", D:String=@n1.gender, E:String=" : ", F:String="To String (Boolean)"(InBool=@n8)) -> (ReturnValue))
   > n9 = Branch(Condition=@n8)
   n10 = Branch(
    execute=[@n9.then, @n6.else, @n11.then]
    Condition="Not Equal, Case Insenstive (String)"!(
     A:String=Break Asset Playlist Asset!(AssetPlaylist_asset:"Asset Playlist Asset Structure (by ref)"=@n5) -> (
      shotGridId_27_#5:int
      cosmeticType_20_#6:CosmeticPicker_CosmeticTypes Enum
      name_22_#7:String
      gender_25_#8:String
      assetGroup_41_#9:String
      primaryAssetGroup_42_#10:String
      status_33_#11:String
      fixVersion_37_#12:String
     ) {is=title K2Node_BreakStruct}.status_33_#11
     B:String=nope
    ) -> (ReturnValue:bool)
   )
   > Add(TargetArray=Get csv_assets(), NewItem=@n5)
   PrintString(execute=@n10.else, InString=skipping due to status)
   n11 = Branch(execute=@n9.else, Condition="Equal, Case Insensitive (String)"(A=@n7, B=none))
   PrintString(
    execute=@n11.else
    InString=Append(
     A="skipping due to gender "
     B=@n1.gender
     C=" vs "
     D=@n7
     E=" vs "
     F=Break Asset Playlist Asset(AssetPlaylist_asset=@n5).gender_25_#8
    )
   )
  ]}
 get Cosmetic Assets of CosmeticType {nodes [
   n1 = get Cosmetic Assets of CosmeticType!() -> (then, cosmeticType:CosmeticPicker_CosmeticTypes Enum=NewEnumerator0) {is=title K2Node_FunctionEntry}
   n2 = PrintString(InString=Append!(A=GetCosmeticAssetsOfCosmeticType, B=" ", C=Enum to String(Enumerator=@n1)) -> (ReturnValue))
   n3 = Switch on CosmeticPicker_CosmeticTypes!(execute=[@n2.then, @n1.then], Selection:CosmeticPicker_CosmeticTypes Enum=@n1) -> (
    NewEnumerator0:Exec
    NewEnumerator1:Exec
    NewEnumerator2:Exec
    NewEnumerator3:Exec
    NewEnumerator4:Exec
    NewEnumerator5:Exec
    NewEnumerator6:Exec
    NewEnumerator7:Exec
    NewEnumerator8:Exec
    NewEnumerator9:Exec
    NewEnumerator10:Exec
    NewEnumerator11:Exec
    NewEnumerator12:Exec
    NewEnumerator13:Exec
    NewEnumerator14:Exec
    NewEnumerator15:Exec
   ) {is=title K2Node_SwitchEnum}
   > n4 = GetAssetsByClass!(
    execute
    self:Asset Registry Interface=GetAssetRegistry()
    ClassPathName_PackageName:String=/Script/FortniteGame
    ClassPathName_AssetName:String=AthenaCharacterItemDefinition
    bSearchSubClasses:bool=true
   ) -> (then, OutAssetData:Array of Asset Data Structures, ReturnValue:bool)
   > n5 = Set AssetData(AssetData:Array of Asset Data Structures=@n4) -> (Output_Get:Array of Asset Data Structures)
   n6 = Get AssetData() -> (AssetData:Array of Asset Data Structures)
   Return Node!(execute=[@n20.then, @n10.then, @n12.then, @n14.then, @n5.then, @n19.then, @n16.Completed], assets:Array of Asset Data Structures=@n6)
   n7 = GetAssetRegistry()
   n8 = GetAssetsByClass(
    execute=@n3.NewEnumerator1
    self=@n7
    ClassPathName_PackageName=&FortniteGame
    ClassPathName_AssetName=AthenaBackpackItemDefinition
    bSearchSubClasses=true
   )
   > n9 = GetAssetsByClass(self=@n7, ClassPathName_PackageName=&FortniteGame, ClassPathName_AssetName=AthenaPetCarrierItemDefinition, bSearchSubClasses=true)
   > Append Array!(execute, TargetArray:Array of Asset Data Structures=@n8, SourceArray:Array of Asset Data Structures=@n9) -> (then)
   > n10 = Set AssetData(AssetData=@n8)
   n11 = GetAssetsByClass(
    execute=@n3.NewEnumerator2
    self=GetAssetRegistry()
    ClassPathName_PackageName=&FortniteGame
    ClassPathName_AssetName=AthenaPickaxeItemDefinition
    bSearchSubClasses=true
   )
   > n12 = Set AssetData(AssetData=@n11)
   n13 = GetAssetsByClass(
    execute=@n3.NewEnumerator3
    self=GetAssetRegistry()
    ClassPathName_PackageName=&FortniteGame
    ClassPathName_AssetName=AthenaGliderItemDefinition
    bSearchSubClasses=true
   )
   > n14 = Set AssetData(AssetData=@n13)
   n15 = GetAssetsByClass(
    execute=@n3.NewEnumerator4
    self=GetAssetRegistry()
    ClassPathName_PackageName=&FortniteGame
    ClassPathName_AssetName=AthenaDanceItemDefinition
    bSearchSubClasses=true
   )
   > n16 = For Each Loop(Array:Array of Asset Data Structures=@n15) -> (Array Element:Asset Data Structure)
   n17 = GetAsset(InAssetData=@n16)
   > Branch(Condition=Contains(SearchIn=GetDisplayName(Object=@n17), Substring=eid))
   > Add(TargetArray:Array of Asset Data Structures=Get AssetData(), NewItem:"Asset Data Structure (by ref)"=@n16)
   n18 = GetAssetsByClass(
    execute=@n3.NewEnumerator5
    self=GetAssetRegistry()
    ClassPathName_PackageName=&FortniteGame
    ClassPathName_AssetName=AthenaItemWrapDefinition
    bSearchSubClasses=true
   )
   > n19 = Set AssetData(AssetData=@n18)
   n20 = PrintString(InString="To String (Integer)"(InInt=Length(TargetArray:Array of Asset Data Structures=@n6)))
  ]}
 get Gender {nodes [
   n1 = get Gender!() -> (
    then
    asset_shotGridId_27_#5:int
    asset_cosmeticType_20_#6:CosmeticPicker_CosmeticTypes Enum
    asset_name_22_#7:String
    asset_gender_25_#8:String
    asset_assetGroup_41_#9:String
    asset_primaryAssetGroup_42_#10:String
    asset_status_33_#11:String
    asset_fixVersion_37_#12:String
   ) {is=title K2Node_FunctionEntry}
   > n2 = Branch(Condition="Equal, Case Insensitive (String)"(A=@n1.asset_gender_25_#8, B=none))
   > n3 = Branch(Condition=Contains(SearchIn=@n1.asset_name_22_#7, Substring=female))
   > Return Node!(execute, gender:String=F)
   n4 = Branch(execute=@n3.else, Condition=Contains(SearchIn=@n1.asset_name_22_#7, Substring=male))
   > Return Node(gender=M)
   Return Node(execute=@n4.else, gender=@n1.asset_gender_25_#8)
   Return Node(execute=@n2.else, gender=@n1.asset_gender_25_#8)
  ]}
 get Point Releases {nodes [
   n1 = get Point Releases!() -> (then, season:String) {is=title K2Node_FunctionEntry}
   > n2 = Execute Python Script!(
    execute
    PythonScript="import TechArt.CosmeticPlaylist.CosmeticPlaylistModule as CosmeticPlaylist\r\npoint_releases = CosmeticPlaylist.getPointReleases(season)\r\n\r\n"
    season:String=@n1
   ) -> (then, ReturnValue, point_releases:Array of Strings) {inputs [season], outputs [point_releases]}
   > Return Node!(execute, PointReleases:Array of Strings=@n2.point_releases)
  ]}
 get Recent Playlists {nodes [
   get Recent Playlists() {is=title K2Node_FunctionEntry}
   > n1 = Execute Python Script!(
    execute
    PythonScript="import TechArt.CosmeticPlaylist.CosmeticPlaylistModule as CosmeticPlaylist\r\n\r\nreleases = CosmeticPlaylist.getRecentPlaylists(days=100, reverse=True)\r\n\r\nids = []\r\ncodes = []\r\n\r\nfor release in releases:\r\n id = release['id']\r\n code = release['code']\r\n ids.append(id)\r\n codes.append(code)\r\n\r\nprint (\"yo\") \r\n"
   ) -> (then, ReturnValue, ids:Array of Integers, codes:Array of Strings) {outputs [ids,codes]}
   > Return Node!(execute, ids:Array of Integers=@n1.ids, names:Array of Strings=@n1.codes)
  ]}
 get Release {nodes [
   get Release() {is=title K2Node_FunctionEntry}
   n1 = GetBuildVersion!() -> (ReturnValue:String)
   > Return Node!(execute, release:String=Split(SourceString=Split(SourceString=@n1, InStr=+, SearchDir=FromEnd).RightS, InStr=-, SearchDir=FromStart))
  ]}
 get Releases {nodes [
   n1 = get Releases!() -> (then, months:String) {is=title K2Node_FunctionEntry}
   > n2 = Execute Python Script!(
    execute
    PythonScript="import TechArt.CosmeticPlaylist.CosmeticPlaylistModule as CosmeticPlaylist\r\nreleases = CosmeticPlaylist.getReleases(months)\r\n\r\n"
    months:String=@n1
   ) -> (then, ReturnValue, releases:Array of Strings) {inputs [months], outputs [releases]}
   > Return Node!(execute, releases:Array of Strings=@n2.releases)
  ]}
 get Seasons {nodes [
   get Seasons() {is=title K2Node_FunctionEntry}
   > n1 = Execute Python Script!(
    execute
    PythonScript="import TechArt.CosmeticPlaylist.CosmeticPlaylistModule as CosmeticPlaylist\r\nseasons = CosmeticPlaylist.getSeasons()\r\n\r\n"
   ) -> (then, ReturnValue, seasons:Array of Strings) {outputs [seasons]}
   > Return Node!(execute, seasons:Array of Strings=@n1.seasons)
  ]}
 get find checkbox {nodes [
   n1 = get find checkbox!() -> (then, widget:Widget Object Reference) {is=title K2Node_FunctionEntry}
   > n2 = FindParentWidgetOfType(StartingWidget=@n1, Type=&/AssetPlaylist.AssetPlaylist_C) -> (ReturnValue:Asset Playlist Object Reference)
   > n3 = Cast To AssetPlaylist!(execute, Object=@n2) -> (then, CastFailed, AsAsset Playlist:Asset Playlist Object Reference) {is=title K2Node_DynamicCast}
   > n4 = getOptions!(
    execute
    self:EUW Asset Playlist Options Widget Object Reference=Get EUW_AssetPlaylistOptions!(self:Asset Playlist Object Reference=@n3) -> (EUW_AssetPlaylistOptions:EUW Asset Playlist Options Widget Object Reference)
   ) -> (then, cosmeticType:CosmeticPicker_CosmeticTypes Enum=NewEnumerator0, items:bool, unknown:bool, variants:bool, builtin:bool, thumbnails:bool)
   > Return Node!(execute, isChecked:bool=@n4.thumbnails)
  ]}
 getActiveVariantTag {nodes [
   n1 = getActiveVariantTag!() -> (then, item:Athena Cosmetic Item Definition Object Reference, index:int, option_index:int) {is=title K2Node_FunctionEntry}
   > n2 = Execute Python Script!(
    execute
    PythonScript="import TechArt.CosmeticPicker.CosmeticPickerModule as CosmeticPicker\r\n\r\nactive_variant_tag = CosmeticPicker.getActiveVariantTag(item, index, option_index)"
    item:Athena Cosmetic Item Definition Object Reference=@n1
    index:int=@n1.index
    option_index:int=@n1.option_index
   ) -> (then, ReturnValue, active_variant_tag:Gameplay Tag Structure) {inputs [item,index,option_index], outputs [active_variant_tag]}
   > Return Node!(execute, tag:Gameplay Tag Structure=@n2.active_variant_tag)
  ]}
 getAssetPlaylist {nodes [
   n1 = getAssetPlaylist!() -> (then, widget:Widget Object Reference) {is=title K2Node_FunctionEntry}
   > n2 = FindParentWidgetOfType(StartingWidget=@n1, Type=&AssetPlaylist_C)
   > n3 = Cast To AssetPlaylist(Object=@n2)
   > Return Node!(execute, assetPlaylist:Asset Playlist Object Reference=@n3)
  ]}
 getVariantTag {nodes [
   n1 = getVariantTag!() -> (then, item:Athena Cosmetic Item Definition Object Reference, index:int) {is=title K2Node_FunctionEntry}
   > n2 = Execute Python Script!(
    execute
    PythonScript="import TechArt.CosmeticPicker.CosmeticPickerModule as CosmeticPicker\r\n\r\nvariant_tag = CosmeticPicker.getVariantChannelTag(item, index)\r\nvalid = True\r\nif variant_tag == None:\r\n   valid = False"
    item=@n1
    index=@n1.index
   ) -> (then, ReturnValue, variant_tag:Gameplay Tag Structure, valid:bool=~) {inputs [item,index], outputs [variant_tag,valid]}
   > n3 = Branch(Condition=@n2.valid)
   > PrintString(InString=ValidTag)
   > Return Node!(execute, tag:Gameplay Tag Structure=@n2.variant_tag, success:bool=true)
   PrintString(execute=@n3.else, InString=InvalidTag)
   > Return Node(tag=(TagName="Cosmetics.Variant.Channel.Material"), success=false)
  ]}
 make Asset Playlist Asset from Asset Info {nodes [
   n1 = make Asset Playlist Asset from Asset Info!() -> (then, assetInfo:String, useAssetGroups:bool=~) {is=title K2Node_FunctionEntry}
   > n2 = Execute Python Script!(
    execute
    PythonScript="\r\ninfo = asset_info.split(\",\")\r\n\r\nasset_id=info[0]\r\ncosmetic_type = info[1]\r\nname = info[2]\r\ngender = info[3]\r\nasset_group = info[4]\r\nprimary_asset_group = info[5]\r\nstatus = info[6]\r\nfix_version = info[7]"
    assetInfo:String=@n1
   ) -> (
    then
    ReturnValue
    asset_id:String
    cosmetic_type:String
    name:String
    gender:String
    status:String
    fix_version:String
    asset_group:String
    primary_asset_group:String
   ) {inputs [assetInfo], outputs [asset_id, cosmetic_type, name, gender, status, fix_version, asset_group, primary_asset_group]}
   > n3 = string to CosmeticType!(execute, self, Selection:String=@n2.cosmetic_type) -> (then, cosmeticType:CosmeticPicker_CosmeticTypes Enum=NewEnumerator0)
   > n4 = Branch(Condition=@n1.useAssetGroups)
   n5 = String To Integer(InString=@n2.asset_id)
   > n6 = Set l_asset(l_asset:Asset Playlist Asset Structure=Make Asset Playlist Asset!(
     shotGridId_27_#5:int=@n5
     cosmeticType_20_#6:CosmeticPicker_CosmeticTypes Enum=@n3
     name_22_#7:String=@n2.name
     gender_25_#8:String=@n2.gender
     assetGroup_41_#9:String=@n2.asset_group
     primaryAssetGroup_42_#10:String=@n2.primary_asset_group
     status_33_#11:String=@n2.status
     fixVersion_37_#12:String=@n2.fix_version
    ) -> (AssetPlaylist_asset:Asset Playlist Asset Structure) {is=title K2Node_MakeStruct}) -> (Output_Get:Asset Playlist Asset Structure)
   n7 = Break Asset Playlist Asset(AssetPlaylist_asset=Get l_asset() -> (l_asset:Asset Playlist Asset Structure))
   PrintString(
    execute=[@n8.then, @n6.then]
    InString=Append!(
     A=name:
     B=@n7.name_22_#7
     C="\r\n  asset_group:"
     D:String=@n7.assetGroup_41_#9
     E:String="\r\n  asset_groups:"
     F:String=@n7.primaryAssetGroup_42_#10
    ) -> (ReturnValue)
   )
   > Return Node!(execute, asset:Asset Playlist Asset Structure=Get l_asset())
   n8 = Set l_asset(
    execute=@n4.else
    l_asset=Make Asset Playlist Asset(
     shotGridId_27_#5=@n5
     cosmeticType_20_#6=@n3
     name_22_#7=@n2.name
     gender_25_#8=@n2.gender
     assetGroup_41_#9=@n2.asset_group
     status_33_#11=@n2.status
     fixVersion_37_#12=@n2.fix_version
    )
   )
  ]}
 reorderGroupsArray {nodes [
   n1 = reorderGroupsArray!() -> (then, Array:Array of EUW Playlist Group Object References) {is=title K2Node_FunctionEntry}
   > n2 = Execute Python Script!(
    execute
    PythonScript="import TechArt.CosmeticPlaylist.CosmeticPlaylistModule as CosmeticPlaylist\r\nout_groups = CosmeticPlaylist.reorderGroupsArray(in_groups)\r\n\r\n"
    in_groups:Array of EUW Playlist Group Object References=@n1
   ) -> (then, ReturnValue, out_groups:Array of EUW Playlist Group Object References) {inputs [in_groups], outputs [out_groups]}
   > Return Node!(execute, OutputArray:Array of EUW Playlist Group Object References=@n2.out_groups)
  ]}
 search For Cosmetic by Name {nodes [
   n1 = search For Cosmetic by Name!() -> (then, assets:Array of Asset Data Structures, name:String, gender:String) {is=title K2Node_FunctionEntry}
   > Set _name(_name:String=@n1.name) -> (Output_Get:String)
   > Set _gender(_gender:String=@n1.gender) -> (Output_Get:String)
   > Set _assets(_assets:Array of Asset Data Structures=@n1) -> (Output_Get:Array of Asset Data Structures)
   > n2 = Sequence!(execute) -> (then_0, then_1, then_2:Exec, then_3:Exec)
   n3 = Branch(execute=@n2.then_2, Condition="Equal, Case Insensitive (String)"(A=Get _gender() -> (_gender:String), B=m))
   > n4 = Set longGender(longGender:String=male) -> (Output_Get:String)
   n5 = Get _name() -> (_name:String)
   n6 = find Substring in Assets!(
    execute=[@n7.then, @n4.then]
    self
    Array:Array of Asset Data Structures=Get _assets() -> (_assets:Array of Asset Data Structures)
    Substring:String=Append!(A=Replace(SourceString=@n5, From=" "), B=Get longGender() -> (longGender:String)) -> (ReturnValue)
   ) -> (then, OutputPin:Object Reference) {is=title K2Node_CallFunction}
   > Is Valid(InputObject=@n6)
   > Return Node!(execute, item:Object Reference=@n6)
   n7 = Set longGender(execute=@n3.else, longGender=female)
   n8 = find Substring in Assets(execute=@n2.then_3, Array=Get _assets(), Substring=Replace(SourceString=Get _name(), From=" "))
   > Is Valid(InputObject=@n8)
   > Return Node(item=@n8)
   n9 = Get _name()
   n10 = find Substring in Assets(
    execute=@n2.then_1
    Array=Get _assets()
    Substring=Append!(A=Get _gender(), B=_, C:String=Replace(SourceString=@n9, From=" ")) -> (ReturnValue)
   )
   > Is Valid(InputObject=@n10)
   > Return Node(item=@n10)
   n11 = Get _name()
   n12 = find Substring in Assets(execute=@n2.then_0, Array=Get _assets(), Substring=Append(A=Get _gender(), B=_, C=Replace(SourceString=@n11, From=" ", To=_)))
   > Is Valid(InputObject=@n12)
   > Return Node(item=@n12)
  ]}
 slowTask_end {nodes [
   Execute Python Script!(execute, PythonScript="import TechArt.CosmeticPlaylist.CosmeticPlaylistModule as CosmeticPlaylist\r\nCosmeticPlaylist.endSlowTask()\r\n") -> (then, ReturnValue)
   slowTask_end() {is=title K2Node_FunctionEntry}
  ]}
 slowTask_progress {nodes [Execute Python Script!(
    execute
    PythonScript="import TechArt.CosmeticPlaylist.CosmeticPlaylistModule as CosmeticPlaylist\r\nCosmeticPlaylist.progressSlowTask(progress)\r\n"
    progress:int=slowTask_progress!() -> (then, progress:int) {is=title K2Node_FunctionEntry}
   ) -> (then, ReturnValue) {inputs [progress]}]}
 slowTask_start {nodes [
   Execute Python Script!(execute, PythonScript="import TechArt.CosmeticPlaylist.CosmeticPlaylistModule as CosmeticPlaylist\r\nCosmeticPlaylist.startSlowTask()\r\n") -> (then, ReturnValue)
   slowTask_start() {is=title K2Node_FunctionEntry}
  ]}
 string to CosmeticType {nodes [
   n1 = string to CosmeticType!() -> (then, Selection:String) {is=title K2Node_FunctionEntry}
   n2 = Get type() -> (type:CosmeticPicker_CosmeticTypes Enum=NewEnumerator0)
   n3 = PrintString(InString=Append(A=@n1, B="<->", C=Enum to String(Enumerator=@n2)))
   Return Node!(
    execute=[@n3.then, @n7.then, @n6.then, @n5.then, @n8.then, @n9.then, @n11.then, @n10.then]
    cosmeticType:CosmeticPicker_CosmeticTypes Enum=Get type() -> (type=NewEnumerator0)
   )
   n4 = Switch on String!(execute:Exec=@n1.then, Selection:String=@n1) -> (Default:Exec, Hero Skin:Exec, Character:Exec, Backpack:Exec, Emote:Exec, Glider:Exec, Wrap:Exec, Pickaxe:Exec) {pinNames [Hero Skin, Character, Backpack, Emote, Glider, Wrap, Pickaxe]}
   n5 = Set type(execute=@n4.Backpack, type:CosmeticPicker_CosmeticTypes Enum=NewEnumerator1) -> (Output_Get:CosmeticPicker_CosmeticTypes Enum=NewEnumerator0)
   n6 = Set type(execute=[@n4.Character, @n4.Hero Skin], type=NewEnumerator0) -> (Output_Get=NewEnumerator0)
   n7 = Set type(execute=@n4.Default, type=NewEnumerator6) -> (Output_Get=NewEnumerator0)
   n8 = Set type(execute=@n4.Emote, type=NewEnumerator4) -> (Output_Get=NewEnumerator0)
   n9 = Set type(execute=@n4.Glider, type=NewEnumerator3) -> (Output_Get=NewEnumerator0)
   n10 = Set type(execute=@n4.Pickaxe, type=NewEnumerator2) -> (Output_Get=NewEnumerator0)
   n11 = Set type(execute=@n4.Wrap, type=NewEnumerator5) -> (Output_Get=NewEnumerator0)
  ]}
]
variables [dataTable {type=DataTable, default=&/AssetPlaylist_releases}]
widgets {
 info {rootWidgetClass=&CanvasPanel, widgetCount=1}
 widgets [CanvasPanel_0 {widgetClassPath=&CanvasPanel, widget=CanvasPanel_0}]
}
