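# Editor Utility Widget in the tech-art AssetPlaylist tools: a reusable search box with a filterable suggestion list over a `strings` array.
# Each text edit broadcasts `onTextChanged`, then, if the text differs from `selectedText`, shows and rebuilds the TA_Generic01_List with substring matches.
# Picking an item copies its text into the box and `selectedText`, collapses the list and broadcasts `selectedItem`.
# `clearText` empties the box; `count` tallies matches but is never read outside its increment.
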EditorUtilityWidgetBlueprint /Game/NeverCook/ArtTools/Tech_Art/AssetPlaylist/EUW_searchableList
graphs [
 EventGraph {nodes [
   n1 = "OnTextChanged (Editable Text) (EditableText_0)"!() -> (OutputDelegate, then, Text:"Text (by ref)") {is=title K2Node_ComponentBoundEvent}
   > Set text(text:String="To String (Text)"(InText=@n1.Text)) -> (Output_Get:String)
   > Call On Text Changed!(execute:Exec, self:EUW Searchable List Object Reference) -> (then:Exec) {is=title K2Node_CallDelegate}
   > Branch(Condition="Not Equal Exactly (Text)"!(A:"Text (by ref)"=@n1.Text, B:"Text (by ref)"=Get selectedText() -> (selectedText:Text)) -> (ReturnValue:bool))
   > Set count(count:int=0) -> (Output_Get:int=0)
   n2 = "To String (Text)"(InText=@n1.Text)
   > PrintString(InString=@n2)
   > SetVisibility(self=Get TA_Generic01_List() -> (TA_Generic01_List:TA Generic 01 List Object Reference), InVisibility=Visible)
   n3 = Get TA_Generic01_List()
   > clearItems(self:TA Generic 01 List Object Reference=@n3)
   > n4 = For Each Loop with Break!(Exec:Exec, Array:Array of Strings=Get strings() -> (strings:Array of Strings), Break:Exec) -> (LoopBody:Exec, Array Element:String, Array Index:int, Completed:Exec)
   > Branch(Condition=Contains(SearchIn=@n4, Substring=@n2))
   > addItemGeneric!(execute, self:TA Generic 01 List Object Reference=@n3, displayString:String=@n4) -> (then, widget:User Widget Object Reference, data:Object Reference)
   > Increment Int(Value=Get count() -> (count:int=0))
   n5 = "On Item Selection Changed (TA_Generic01_List)"!() -> (
    OutputDelegate
    then
    selectedData:Array of Object References
    selectedWidgets:Array of User Widget Object References
    selectedIndices:Array of Integers
   ) {is=title K2Node_ComponentBoundEvent}
   > n6 = Cast To TA_Generic01_ListItem!(execute, Object="Get (a copy)"(Array:Array of User Widget Object References=@n5.selectedWidgets) -> (Output:User Widget Object Reference)) -> (then, CastFailed, AsTA Generic 01 List Item:TA Generic 01 List Item Object Reference) {is=title K2Node_DynamicCast}
   n7 = "GetText (Text)"!(self:Text Object Reference=Get txtMain!(self:TA Generic 01 List Item Object Reference=@n6) -> (txtMain:Text Object Reference)) -> (ReturnValue:Text)
   > PrintString(InString=Append(A="* ", B="To String (Text)"(InText=@n7)))
   > "SetText (Editable Text)"!(execute, self:Editable Text Object Reference=Get EditableText_0() -> (EditableText_0:Editable Text Object Reference), InText:Text=@n7) -> (then)
   > Set selectedText(selectedText:Text=@n7) -> (Output_Get:Text)
   > SetVisibility(self=Get TA_Generic01_List(), InVisibility=Collapsed)
   > Call Selected Item!(execute:Exec, self:EUW Searchable List Object Reference, text:Text=@n7) -> (then:Exec) {is=title K2Node_CallDelegate}
   Event Construct()
   > SetHintText!(execute, self:Editable Text Object Reference=Get EditableText_0(), InHintText:Text=Get hintText() -> (hintText:Text)) -> (then)
   Event PreConstruct()
   Event Tick()
  ]}
 clearText {nodes [
   clearText() {is=title K2Node_FunctionEntry}
   > "SetText (Editable Text)"(self=Get EditableText_0())
  ]}
 onTextChanged {nodes [onTextChanged() {is=title K2Node_FunctionEntry}]}
 selectedItem {nodes [selectedItem!() -> (then, text:Text) {is=title K2Node_FunctionEntry}]}
]
variables [
 count {type=integer}
 hintText {type=string, default=""}
 onTextChanged {default=()}
 retrievedUsers {type=boolean}
 selectedItem {default=()}
 selectedText {type=string, default=""}
 strings {type="array<string>", default []}
 text {type=string, default=""}
]
widgets {
 info {rootWidgetClass=&CanvasPanel, widgetCount=5}
 widgets [
  CanvasPanel_0 {widgetClassPath=&CanvasPanel, widget=CanvasPanel_0}
  VerticalBox_107 {
   widgetClassPath=&VerticalBox
   parent=CanvasPanel_0
   widget=VerticalBox_107
   slot {class=CanvasPanelSlot, properties {layoutData {alignment=<0 0>, anchors {maximum=<1 1>, minimum=<0 0>}, offsets=<ltrb 0 0 0 0>}}}
  }
  SizeBox_75 {
   widgetClassPath=&SizeBox
   parent=VerticalBox_107
   widget=SizeBox_75
   properties {bOverride_HeightOverride=true, heightOverride=20}
   slot {class=VerticalBoxSlot}
  }
  EditableText_0 {
   widgetClassPath=/Script/UMG.EditableText
   bIsVariable=true
   parent=SizeBox_75
   widget=EditableText_0
   properties {
    hintText=Text
    widgetStyle {
     backgroundImageComposing {
      bIsDynamicallyLoaded=false
      drawAs=Border
      imageType=FullColor
      mirroring=NoMirror
      resourceName=../../../Engine/Content/Slate/Old/HyperlinkDotted.png
      resourceObject=null
      tiling=Both
      imageSize=<0 0>
      margin=<ltrb 0 0 0 0.1875>
      outlineSettings {
       bUseBrushTransparency=false
       roundingType=HalfHeightRadius
       width=0
       color {colorUseRule=UseColor_Specified, specifiedColor=<rgba 0 0 0 0>}
       cornerRadii=<0 0 0 0>
      }
      tintColor {colorUseRule=UseColor_Specified, specifiedColor=<rgba 1 1 1 1>}
      uVRegion {bIsValid=false, max=<0 0>, min=<0 0>}
     }
     backgroundImageSelected {
      bIsDynamicallyLoaded=false
      drawAs=Image
      imageType=NoImage
      mirroring=NoMirror
      resourceName=None
      resourceObject=/Game/Global/Textures/Default/Blanks/T_Blank_LegalWhite_sRGBColor
      tiling=NoTile
      imageSize=<16 16>
      margin=<ltrb 0 0 0 0>
      outlineSettings {
       bUseBrushTransparency=false
       roundingType=HalfHeightRadius
       width=0
       color {colorUseRule=UseColor_Specified, specifiedColor=<rgba 0 0 0 0>}
       cornerRadii=<0 0 0 0>
      }
      tintColor {colorUseRule=UseColor_Specified, specifiedColor=<rgba 0 0.162 0.7454 1>}
      uVRegion {bIsValid=false, max=<0 0>, min=<0 0>}
     }
     caretImage {
      bIsDynamicallyLoaded=false
      drawAs=Image
      imageType=NoImage
      mirroring=NoMirror
      resourceName=None
      resourceObject=null
      tiling=NoTile
      imageSize=<0 0>
      margin=<ltrb 0 0 0 0>
      outlineSettings {
       bUseBrushTransparency=false
       roundingType=HalfHeightRadius
       width=0
       color {colorUseRule=UseColor_Specified, specifiedColor=<rgba 0 0 0 0>}
       cornerRadii=<0 0 0 0>
      }
      tintColor {colorUseRule=UseColor_Specified, specifiedColor=<rgba 1 1 1 1>}
      uVRegion {bIsValid=false, max=<0 0>, min=<0 0>}
     }
     colorAndOpacity {colorUseRule=UseColor_Foreground, specifiedColor=<rgba 1 0 1 1>}
     font {
      bForceMonospaced=false
      bMaterialIsStencil=false
      fontMaterial=null
      fontObject=&NotoSans
      letterSpacing=0
      monospacedWidth=1
      size=12
      skewAmount=0
      typefaceFontName=NotoSans-Regular
      outlineSettings {
       bApplyOutlineToDropShadows=false
       bMiteredCorners=false
       bSeparateFillAlpha=false
       outlineMaterial=null
       outlineSize=0
       outlineColor=<rgba 0 0 0 1>
      }
     }
     hintColorAndOpacity {colorUseRule=UseColor_Foreground, specifiedColor=<rgba 1 0 1 1>}
     selectedBackgroundColor {colorUseRule=UseColor_Foreground, specifiedColor=<rgba 1 0 1 1>}
    }
   }
   slot {class=SizeBoxSlot}
  }
  TA_Generic01_List {
   widgetClassPath=/Game/Blueprints/Blutilities/ForTechArt/Widgets/List/TA_Generic01_List
   bIsVariable=true
   parent=VerticalBox_107
   widget=TA_Generic01_List
   properties {allowMultiSelect=false}
   slot {class=VerticalBoxSlot, properties {size {sizeRule=Fill, value=1}}}
  }
 ]
}
