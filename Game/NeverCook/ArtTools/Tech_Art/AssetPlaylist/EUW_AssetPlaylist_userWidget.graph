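# Editor utility widget that picks a user in the Asset Playlist art tool and loads that user's assets into the playlist.
# When its text changes while empty, the searchable list fills with user names from the Python CosmeticPlaylist module.
# A selection finds the AssetPlaylist through the headless helper widget and calls Load User Assets.
# Only the collapsed user-name widget stores User, which Reload reuses or reports as unset.
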EditorUtilityWidgetBlueprint /Game/NeverCook/ArtTools/Tech_Art/AssetPlaylist/EUW_AssetPlaylist_userWidget
graphs [
 EventGraph {nodes [
   n1 = "Selected Item (EUW_userName)"!() -> (OutputDelegate, then, text:Text) {is=title K2Node_ComponentBoundEvent}
   > n2 = Set User(User:String="To String (Text)"(InText=@n1.text)) -> (Output_Get:String)
   n3 = getAssetPlaylist!(
    execute=[@n2.then, @n8.then]
    self:EUW Asset Playlist Headless Object Reference=Get EUW_AssetPlaylist_headless() -> (EUW_AssetPlaylist_headless:EUW Asset Playlist Headless Object Reference)
    widget:Widget Object Reference=Self-Reference()
   ) -> (then, assetPlaylist:Asset Playlist Object Reference)
   > Event Load User Assets!(execute, self:Asset Playlist Object Reference=@n3, UserName:String=@n2) -> (then)
   n4 = "Selected Item (EUW_searchableList)"!() -> (OutputDelegate, then, text:Text) {is=title K2Node_ComponentBoundEvent}
   > n5 = getAssetPlaylist(self=Get EUW_AssetPlaylist_headless(), widget=Self-Reference())
   > Event Load User Assets(self=@n5, UserName="To String (Text)"(InText=@n4.text))
   "On Text Changed (EUW_searchableList)"() {is=title K2Node_ComponentBoundEvent}
   n6 = Get EUW_searchableList() -> (EUW_searchableList:EUW Searchable List Object Reference)
   > Branch(Condition=Is Empty!(TargetArray=Get strings!(self:EUW Searchable List Object Reference=@n6) -> (strings:Array of Strings)) -> (ReturnValue:bool))
   > n7 = Execute Python Script!(execute, PythonScript="import TechArt.CosmeticPlaylist.CosmeticPlaylistModule as CosmeticPlaylist\r\nusers = CosmeticPlaylist.getUsers()") -> (then, ReturnValue, users:Array of Strings) {outputs [users]}
   > Set strings!(execute, strings:Array of Strings=@n7.users, self:EUW Searchable List Object Reference=@n6) -> (then, Output_Get:Array of Strings)
   Reload() {is=title K2Node_CustomEvent}
   > n8 = Branch(Condition="Not Equal, Case Insenstive (String)"!(A:String=Get User() -> (User:String), B:String) -> (ReturnValue:bool))
   PrintString(execute=@n8.else, InString=User not set)
   Event PreConstruct()
   Event Construct()
  ]}
 getAssetPlaylist {nodes [
   getAssetPlaylist() {is=title K2Node_FunctionEntry}
   > n1 = FindParentWidgetOfType!(
    execute
    StartingWidget:Widget Object Reference=Self-Reference()
    Type:Widget Class Reference=&/AssetPlaylist.AssetPlaylist_C
   ) -> (then, ReturnValue:Asset Playlist Object Reference)
   > n2 = Cast To AssetPlaylist!(execute, Object=@n1) -> (then, CastFailed, AsAsset Playlist:Asset Playlist Object Reference) {is=title K2Node_DynamicCast}
   > Return Node!(execute, assetPlaylist:Asset Playlist Object Reference=@n2)
  ]}
]
variables [User {type=string, default=""}]
widgets {
 info {rootWidgetClass=&CanvasPanel, widgetCount=5}
 widgets [
  CanvasPanel_0 {widgetClassPath=&CanvasPanel, widget=CanvasPanel_0}
  VerticalBox_39 {
   widgetClassPath=&VerticalBox
   parent=CanvasPanel_0
   widget=VerticalBox_39
   slot {class=CanvasPanelSlot, properties {layoutData {alignment=<0 0>, anchors {maximum=<1 1>, minimum=<0 0>}, offsets=<ltrb 0 0 0 0>}}}
  }
  EUW_userName {
   widgetClassPath=&/EUW_userName
   bIsVariable=true
   parent=VerticalBox_39
   widget=EUW_userName
   properties {visibility=Collapsed}
   slot {class=VerticalBoxSlot}
  }
  EUW_searchableList {
   widgetClassPath=&/EUW_searchableList
   bIsVariable=true
   parent=VerticalBox_39
   widget=EUW_searchableList
   properties {hintText="<User Name>"}
   slot {class=VerticalBoxSlot, properties {size {sizeRule=Fill, value=1}}}
  }
  EUW_AssetPlaylist_headless {
   widgetClassPath=&/EUW_AssetPlaylist_headless
   bIsVariable=true
   parent=CanvasPanel_0
   widget=EUW_AssetPlaylist_headless
   slot {class=CanvasPanelSlot, properties=~}
  }
 ]
}
