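# Editor utility widget for the Asset Playlist tech-art tool: a scrollable vertical list of playlist release widgets and their groups.
# Functions add or clear ScrollBox children and find a release by case-insensitive name.
# Get Groups casts children to EUW_PlaylistRelease and pools each one's getGroups output, which deselectAllGroups uses to deselect every group.
# Debug prints trace additions, group counts and release comparisons.
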EditorUtilityWidgetBlueprint /Game/NeverCook/ArtTools/Tech_Art/AssetPlaylist/EUW_AssetPlaylist_verticalGroupList
graphs [
 EventGraph {nodes [
   n1 = For Each Loop(Array:Array of Widget Object References=GetAllChildren(self=Get ScrollBox() -> (ScrollBox:Scroll Box Object Reference))) -> (Array Element:Widget Object Reference)
   > PrintString(InString=GetDisplayName(Object=@n1))
   Event PreConstruct()
   Event Construct()
   Event Tick()
  ]}
 Get Groups {nodes [
   Get Groups() {is=title K2Node_FunctionEntry}
   > n1 = For Each Loop(Array=GetAllChildren(self=Get ScrollBox()))
   n2 = For Each Loop(
    Exec=@n1.Completed
    Array:Array of EUW Playlist Release Object References=Get releaseArray() -> (releaseArray:Array of EUW Playlist Release Object References)
   ) -> (Array Element:EUW Playlist Release Object Reference)
   n3 = Length(TargetArray:Array of EUW Playlist Group Object References=Get groupArray() -> (groupArray:Array of EUW Playlist Group Object References))
   PrintString(execute=@n2.Completed, InString=Append(A="number of groups in verticalGroupList ", B="To String (Integer)"(InInt=@n3)))
   > Return Node!(execute, groups:Array of EUW Playlist Group Object References=Get groupArray())
   n4 = getGroups!(execute=@n2.LoopBody, self:EUW Playlist Release Object Reference=@n2) -> (then, groups:Array of EUW Playlist Group Object References)
   > n5 = For Each Loop(Array:Array of EUW Playlist Group Object References=@n4) -> (Array Element:EUW Playlist Group Object Reference)
   > Add(TargetArray:Array of EUW Playlist Group Object References=Get groupArray(), NewItem:"EUW Playlist Group Object Reference (by ref)"=@n5) {is=title K2Node_CallArrayFunction}
   n6 = Cast To EUW_PlaylistRelease!(execute=@n1.LoopBody, Object=@n1) -> (then, CastFailed, AsEUW Playlist Release:EUW Playlist Release Object Reference) {is=title K2Node_DynamicCast}
   > Add(TargetArray:Array of EUW Playlist Release Object References=Get releaseArray(), NewItem:"EUW Playlist Release Object Reference (by ref)"=@n6)
  ]}
 addWidget {nodes [
   n1 = addWidget!() -> (then, group:EUW Playlist Group Object Reference) {is=title K2Node_FunctionEntry}
   n2 = Self-Reference()
   > PrintString(InString=Append!(A="adding group to stack ", B=GetDisplayName(Object=@n1), C:String=", ", D:String=GetDisplayName(Object=@n2)) -> (ReturnValue))
   > AddChild(self=Get ScrollBox(), Content=@n1)
  ]}
 clearWidgets {nodes [
   clearWidgets() {is=title K2Node_FunctionEntry}
   > ClearChildren(self=Get ScrollBox())
  ]}
 deselectAllGroups {nodes [
   deselectAllGroups() {is=title K2Node_FunctionEntry}
   > n1 = Get Groups!(execute, self) -> (then, groups:Array of EUW Playlist Group Object References) {is=title K2Node_CallFunction}
   > n2 = For Each Loop(Array=@n1)
   > deselect Group(self:EUW Playlist Group Object Reference=@n2)
  ]}
 getRelease {nodes [
   n1 = getRelease!() -> (then, release:String) {is=title K2Node_FunctionEntry}
   > n2 = For Each Loop(Array:Array of Widget Object References=GetAllChildren(self=Get ScrollBox())) -> (Array Element:Widget Object Reference)
   Return Node!(execute=@n2.Completed, playlistRelease:EUW Playlist Release Object Reference)
   > n3 = Cast To EUW_PlaylistRelease(Object=@n2)
   n4 = Get release!(self:EUW Playlist Release Object Reference=@n3) -> (release:String)
   > PrintString(InString=Append!(A=GetDisplayName(Object=@n3), B=" ", C=@n4, D=" vs ", E:String=@n1) -> (ReturnValue))
   > Branch(Condition="Equal, Case Insensitive (String)"(A=@n4, B=@n1))
   > Return Node(playlistRelease=@n3)
  ]}
 getReleases {nodes [
   getReleases() {is=title K2Node_FunctionEntry}
   > n1 = For Each Loop(Array=GetAllChildren(self=Get ScrollBox()))
   Return Node!(
    execute=@n1.Completed
    groups:Array of EUW Playlist Release Object References=Get groupArray() -> (groupArray:Array of EUW Playlist Release Object References)
   )
   > n2 = Cast To EUW_PlaylistRelease(Object=@n1)
   > Add(TargetArray=Get groupArray(), NewItem=@n2)
  ]}
]
variables []
widgets {
 info {rootWidgetClass=&CanvasPanel, widgetCount=3}
 widgets [
  CanvasPanel_0 {widgetClassPath=&CanvasPanel, widget=CanvasPanel_0, properties {bIsVolatile=true}}
  VerticalBox_18 {
   widgetClassPath=&VerticalBox
   parent=CanvasPanel_0
   widget=VerticalBox_18
   slot {
    class=CanvasPanelSlot
    properties {bAutoSize=true, layoutData {alignment=<0 0>, anchors {maximum=<1 1>, minimum=<0 0>}, offsets=<ltrb 0 0 0 0>}}
   }
  }
  ScrollBox {
   widgetClassPath=&ScrollBox
   bIsVariable=true
   parent=VerticalBox_18
   widget=ScrollBox
   properties {
    widgetBarStyle {
     thickness=8
     draggedThumbImage {
      bIsDynamicallyLoaded=false
      drawAs=RoundedBox
      imageType=NoImage
      mirroring=NoMirror
      resourceName=None
      resourceObject=null
      tiling=NoTile
      imageSize=<0 0>
      margin=<ltrb 0 0 0 0>
      outlineSettings {
       bUseBrushTransparency=false
       roundingType=FixedRadius
       width=0
       color {colorUseRule=UseColor_Specified, specifiedColor=<rgba 0 0 0 0>}
       cornerRadii=<4 4 4 4>
      }
      tintColor {colorUseRule=UseColor_Specified, specifiedColor=<rgba 0.2159 0.2159 0.2159 1>}
      uVRegion {bIsValid=false, max=<0 0>, min=<0 0>}
     }
     horizontalBackgroundImage {
      bIsDynamicallyLoaded=false
      drawAs=NoDrawType
      imageType=NoImage
      mirroring=NoMirror
      resourceName=None
      resourceObject=null
      tiling=NoTile
      imageSize=<0 0>
      margin=<ltrb 0 0 0 0>
      outlineSettings {
       bUseBrushTransparency=false
       roundingType=HalfHeightRadius
       width=0
       color {colorUseRule=UseColor_Specified, specifiedColor=<rgba 0 0 0 0>}
       cornerRadii=<0 0 0 0>
      }
      tintColor {colorUseRule=UseColor_Specified, specifiedColor=<rgba 1 1 1 1>}
      uVRegion {bIsValid=false, max=<0 0>, min=<0 0>}
     }
     horizontalBottomSlotImage {
      bIsDynamicallyLoaded=false
      drawAs=NoDrawType
      imageType=NoImage
      mirroring=NoMirror
      resourceName=None
      resourceObject=null
      tiling=NoTile
      imageSize=<0 0>
      margin=<ltrb 0 0 0 0>
      outlineSettings {
       bUseBrushTransparency=false
       roundingType=HalfHeightRadius
       width=0
       color {colorUseRule=UseColor_Specified, specifiedColor=<rgba 0 0 0 0>}
       cornerRadii=<0 0 0 0>
      }
      tintColor {colorUseRule=UseColor_Specified, specifiedColor=<rgba 1 1 1 1>}
      uVRegion {bIsValid=false, max=<0 0>, min=<0 0>}
     }
     horizontalTopSlotImage {
      bIsDynamicallyLoaded=false
      drawAs=NoDrawType
      imageType=NoImage
      mirroring=NoMirror
      resourceName=None
      resourceObject=null
      tiling=NoTile
      imageSize=<0 0>
      margin=<ltrb 0 0 0 0>
      outlineSettings {
       bUseBrushTransparency=false
       roundingType=HalfHeightRadius
       width=0
       color {colorUseRule=UseColor_Specified, specifiedColor=<rgba 0 0 0 0>}
       cornerRadii=<0 0 0 0>
      }
      tintColor {colorUseRule=UseColor_Specified, specifiedColor=<rgba 1 1 1 1>}
      uVRegion {bIsValid=false, max=<0 0>, min=<0 0>}
     }
     hoveredThumbImage {
      bIsDynamicallyLoaded=false
      drawAs=RoundedBox
      imageType=NoImage
      mirroring=NoMirror
      resourceName=None
      resourceObject=null
      tiling=NoTile
      imageSize=<0 0>
      margin=<ltrb 0 0 0 0>
      outlineSettings {
       bUseBrushTransparency=false
       roundingType=FixedRadius
       width=0
       color {colorUseRule=UseColor_Specified, specifiedColor=<rgba 0 0 0 0>}
       cornerRadii=<4 4 4 4>
      }
      tintColor {colorUseRule=UseColor_Specified, specifiedColor=<rgba 0.2159 0.2159 0.2159 1>}
      uVRegion {bIsValid=false, max=<0 0>, min=<0 0>}
     }
     normalThumbImage {
      bIsDynamicallyLoaded=false
      drawAs=RoundedBox
      imageType=NoImage
      mirroring=NoMirror
      resourceName=None
      resourceObject=null
      tiling=NoTile
      imageSize=<0 0>
      margin=<ltrb 0 0 0 0>
      outlineSettings {
       bUseBrushTransparency=false
       roundingType=FixedRadius
       width=0
       color {colorUseRule=UseColor_Specified, specifiedColor=<rgba 0 0 0 0>}
       cornerRadii=<4 4 4 4>
      }
      tintColor {colorUseRule=UseColor_Specified, specifiedColor=<rgba 0.0953 0.0953 0.0953 1>}
      uVRegion {bIsValid=false, max=<0 0>, min=<0 0>}
     }
     verticalBackgroundImage {
      bIsDynamicallyLoaded=false
      drawAs=NoDrawType
      imageType=NoImage
      mirroring=NoMirror
      resourceName=None
      resourceObject=null
      tiling=NoTile
      imageSize=<0 0>
      margin=<ltrb 0 0 0 0>
      outlineSettings {
       bUseBrushTransparency=false
       roundingType=HalfHeightRadius
       width=0
       color {colorUseRule=UseColor_Specified, specifiedColor=<rgba 0 0 0 0>}
       cornerRadii=<0 0 0 0>
      }
      tintColor {colorUseRule=UseColor_Specified, specifiedColor=<rgba 1 1 1 1>}
      uVRegion {bIsValid=false, max=<0 0>, min=<0 0>}
     }
     verticalBottomSlotImage {
      bIsDynamicallyLoaded=false
      drawAs=NoDrawType
      imageType=NoImage
      mirroring=NoMirror
      resourceName=None
      resourceObject=null
      tiling=NoTile
      imageSize=<0 0>
      margin=<ltrb 0 0 0 0>
      outlineSettings {
       bUseBrushTransparency=false
       roundingType=HalfHeightRadius
       width=0
       color {colorUseRule=UseColor_Specified, specifiedColor=<rgba 0 0 0 0>}
       cornerRadii=<0 0 0 0>
      }
      tintColor {colorUseRule=UseColor_Specified, specifiedColor=<rgba 1 1 1 1>}
      uVRegion {bIsValid=false, max=<0 0>, min=<0 0>}
     }
     verticalTopSlotImage {
      bIsDynamicallyLoaded=false
      drawAs=NoDrawType
      imageType=NoImage
      mirroring=NoMirror
      resourceName=None
      resourceObject=null
      tiling=NoTile
      imageSize=<0 0>
      margin=<ltrb 0 0 0 0>
      outlineSettings {
       bUseBrushTransparency=false
       roundingType=HalfHeightRadius
       width=0
       color {colorUseRule=UseColor_Specified, specifiedColor=<rgba 0 0 0 0>}
       cornerRadii=<0 0 0 0>
      }
      tintColor {colorUseRule=UseColor_Specified, specifiedColor=<rgba 1 1 1 1>}
      uVRegion {bIsValid=false, max=<0 0>, min=<0 0>}
     }
    }
    widgetStyle {
     barThickness=8
     bottomShadowBrush {
      bIsDynamicallyLoaded=false
      drawAs=Box
      imageType=FullColor
      mirroring=NoMirror
      resourceName=../../../Engine/Content/Slate/Common/ScrollBoxShadowBottom.png
      resourceObject=null
      tiling=NoTile
      imageSize=<16 8>
      margin=<ltrb 0.5 0 0.5 1>
      outlineSettings {
       bUseBrushTransparency=false
       roundingType=HalfHeightRadius
       width=0
       color {colorUseRule=UseColor_Specified, specifiedColor=<rgba 0 0 0 0>}
       cornerRadii=<0 0 0 0>
      }
      tintColor {colorUseRule=UseColor_Specified, specifiedColor=<rgba 1 1 1 1>}
      uVRegion {bIsValid=false, max=<0 0>, min=<0 0>}
     }
     horizontalScrolledContentPadding=<ltrb 0 0 1 0>
     leftShadowBrush {
      bIsDynamicallyLoaded=false
      drawAs=Box
      imageType=FullColor
      mirroring=NoMirror
      resourceName=../../../Engine/Content/Slate/Common/ScrollBoxShadowLeft.png
      resourceObject=null
      tiling=NoTile
      imageSize=<8 16>
      margin=<ltrb 1 0.5 0 0.5>
      outlineSettings {
       bUseBrushTransparency=false
       roundingType=HalfHeightRadius
       width=0
       color {colorUseRule=UseColor_Specified, specifiedColor=<rgba 0 0 0 0>}
       cornerRadii=<0 0 0 0>
      }
      tintColor {colorUseRule=UseColor_Specified, specifiedColor=<rgba 1 1 1 1>}
      uVRegion {bIsValid=false, max=<0 0>, min=<0 0>}
     }
     rightShadowBrush {
      bIsDynamicallyLoaded=false
      drawAs=Box
      imageType=FullColor
      mirroring=NoMirror
      resourceName=../../../Engine/Content/Slate/Common/ScrollBoxShadowRight.png
      resourceObject=null
      tiling=NoTile
      imageSize=<8 16>
      margin=<ltrb 0 0.5 1 0.5>
      outlineSettings {
       bUseBrushTransparency=false
       roundingType=HalfHeightRadius
       width=0
       color {colorUseRule=UseColor_Specified, specifiedColor=<rgba 0 0 0 0>}
       cornerRadii=<0 0 0 0>
      }
      tintColor {colorUseRule=UseColor_Specified, specifiedColor=<rgba 1 1 1 1>}
      uVRegion {bIsValid=false, max=<0 0>, min=<0 0>}
     }
     topShadowBrush {
      bIsDynamicallyLoaded=false
      drawAs=Box
      imageType=FullColor
      mirroring=NoMirror
      resourceName=../../../Engine/Content/Slate/Common/ScrollBoxShadowTop.png
      resourceObject=null
      tiling=NoTile
      imageSize=<16 8>
      margin=<ltrb 0.5 1 0.5 0>
      outlineSettings {
       bUseBrushTransparency=false
       roundingType=HalfHeightRadius
       width=0
       color {colorUseRule=UseColor_Specified, specifiedColor=<rgba 0 0 0 0>}
       cornerRadii=<0 0 0 0>
      }
      tintColor {colorUseRule=UseColor_Specified, specifiedColor=<rgba 1 1 1 1>}
      uVRegion {bIsValid=false, max=<0 0>, min=<0 0>}
     }
     verticalScrolledContentPadding=<ltrb 0 0 0 1>
    }
   }
   slot {class=VerticalBoxSlot, properties {size {sizeRule=Fill, value=1}}}
  }
 ]
}
