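# Editor utility widget for one group column in the Asset Playlist cosmetic tool, holding playlist items under a clickable header.
# Clicking selects the group, clearing picker overrides via Python and selecting the first item of several cosmetic types.
# Cancelled drags double as drops: a cursor line trace locates where items spawn, with Fort player mannequins for one cosmetic type.
# Helpers add group assets, filter, and reorder items by type, slotting variants after their base.
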EditorUtilityWidgetBlueprint /Game/NeverCook/ArtTools/Tech_Art/AssetPlaylist/EUW_PlaylistGroup
graphs [
 Add Cosmetic Assets Of Sg Asset Group {nodes [
   Add Cosmetic Assets Of Sg Asset Group() {is=title K2Node_FunctionEntry}
   > PrintString(InString=Append!(A=sg_asset_group: ', B=Get sg_asset_group() -> (sg_asset_group:String), C:String=') -> (ReturnValue))
   > Clear(TargetArray=Get names() -> (names:Array of Strings))
   > n1 = For Each Loop(Array:Array of Widget Object References=GetAllChildren(self=Get VerticalBox_itemGroup() -> (VerticalBox_itemGroup:Vertical Box Object Reference))) -> (Array Element:Widget Object Reference)
   > n2 = Cast To EUW_PlaylistItem!(execute, Object=@n1) -> (then, CastFailed, AsEUW Playlist Item:EUW Playlist Item Object Reference) {is=title K2Node_DynamicCast}
   n3 = Get originalName!(self:EUW Playlist Item Object Reference=@n2) -> (originalName:String)
   > Add(TargetArray=Get names(), NewItem=@n3) {is=title K2Node_CallArrayFunction}
   > PrintString(InString=Append!(A="exclude : ", B=@n3) -> (ReturnValue))
   PrintString(execute=@n1.Completed, InString=Get sg_asset_group())
   > n4 = FindParentWidgetOfType!(
    execute
    StartingWidget:Widget Object Reference=Self-Reference()
    Type:Widget Class Reference=&/AssetPlaylist.AssetPlaylist_C
   ) -> (then, ReturnValue:Asset Playlist Object Reference)
   > n5 = Cast To AssetPlaylist!(execute, Object=@n4) -> (then, CastFailed, AsAsset Playlist:Asset Playlist Object Reference) {is=title K2Node_DynamicCast}
   > add Assets of Group!(
    execute
    self:Asset Playlist Object Reference=@n5
    sg_asset_group:String=Get sg_asset_group()
    gender:String=Get gender() -> (gender:String)
    exclude:Array of Strings=Get names()
   ) -> (then)
   > filter Items(self:Asset Playlist Object Reference=@n5)
   > n6 = For Each Loop(Array=GetAllChildren(self=Get VerticalBox_itemGroup()))
   reorder Group(execute=@n6.Completed)
   > select Group()
   n7 = Cast To EUW_PlaylistItem(execute=@n6.LoopBody, Object=@n6)
   > n8 = Branch(Condition=Contains Item(TargetArray:Array of Strings=Get names(), ItemToFind:"String (by ref)"=Get originalName(self=@n7)))
   "setup Widget as Added "(execute=@n8.else, self:EUW Playlist Item Object Reference=@n7)
  ]}
 EventGraph {nodes [
   Event OnMouseEnter()
   > SetColorAndOpacity(
    self:Image Object Reference=Get Image_button_outer() -> (Image_button_outer:Image Object Reference)
    InColorAndOpacity:Linear Color Structure=Get Hovered Outline Color() -> (Hovered Outline Color:Linear Color Structure)
   )
   > SetColorAndOpacity(
    self=Get Image_button_inner() -> (Image_button_inner:Image Object Reference)
    InColorAndOpacity=Get Hovered Fill Color() -> (Hovered Fill Color:Linear Color Structure)
   )
   Event OnMouseLeave!() -> (OutputDelegate, then, MouseEvent:"Pointer Event Structure (by ref)") {is=title K2Node_Event}
   > SetColorAndOpacity(self=Get Image_button_outer(), InColorAndOpacity=Get Normal Outline Color() -> (Normal Outline Color:Linear Color Structure))
   > SetColorAndOpacity(self=Get Image_button_inner(), InColorAndOpacity=Get Normal Fill Color() -> (Normal Fill Color:Linear Color Structure))
   Event OnDragCancelled!() -> (OutputDelegate, then, PointerEvent:"Pointer Event Structure (by ref)", Operation:Drag Drop Operation Object Reference) {is=title K2Node_Event}
   > Delay!(execute, WorldContextObject:Object Reference, Duration=0.2) -> (then)
   > PrintString(InString=Drag Cancelled)
   > n1 = Get Cursor World Ray!(execute) -> (then, CameraLocation:Vector, RayOrigin:Vector, RayDirection:Vector, ReturnValue:bool) {is=title K2Node_CallFunction}
   > n2 = Line Trace By Channel!(
    execute
    WorldContextObject:Object Reference
    Start:Vector=@n1.RayOrigin
    End:Vector=vector + vector(A=@n1.RayOrigin, B=vector * vector(A=@n1.RayDirection, B:Vector="9999,9999,9999") -> (ReturnValue=~))
    TraceChannel:ETraceTypeQuery Enum
    bTraceComplex:bool=false
    ActorsToIgnore:Array of Actor Object References
    DrawDebugType:EDrawDebugTrace Enum
    bIgnoreSelf:bool
    TraceColor:Linear Color Structure
    TraceHitColor:Linear Color Structure
    DrawTime:float
   ) -> (
    then
    OutHit_bBlockingHit:bool
    OutHit_bInitialOverlap:bool
    OutHit_Time:float
    OutHit_Distance:float
    OutHit_Location:Vector
    OutHit_ImpactPoint:Vector
    OutHit_Normal:Vector
    OutHit_ImpactNormal:Vector
    OutHit_PhysMat:Physical Material Object Reference
    OutHit_HitActor:Actor Object Reference
    OutHit_HitComponent:Primitive Component Object Reference
    OutHit_HitBoneName:Name=None
    OutHit_BoneName:Name=None
    OutHit_HitItem:int=0
    OutHit_ElementIndex:int=0
    OutHit_FaceIndex:int=0
    OutHit_TraceStart:Vector
    OutHit_TraceEnd:Vector
    ReturnValue:bool
   )
   n3 = Get assetPlaylistGroup() -> (assetPlaylistGroup:EUW Playlist Group Object Reference)
   > n4 = Is Valid(InputObject=@n3)
   spawn!(execute=@n4.Is Not Valid, self, spawnActorPosition:Vector=@n2.OutHit_Location) -> (then)
   spawn(execute=@n4.Is Valid, self=@n3, spawnActorPosition=@n2.OutHit_Location)
   Event PreConstruct()
   Event Construct()
   Event Tick()
  ]}
 OnDragDetected {nodes [
   OnDragDetected!() -> (then, MyGeometry:Geometry Structure, PointerEvent:"Pointer Event Structure (by ref)") {is=title K2Node_FunctionEntry}
   > PrintString(InString=Drag Detected)
   > n1 = Create Widget!(
    execute:Exec
    WorldContextObject:Object Reference
    Class:User Widget Class Reference=&/EUW_AssetPlaylist_DragDropWidget.EUW_AssetPlaylist_DragDropWidget_C
    OwningPlayer:Player Controller Object Reference
   ) -> (then:Exec, ReturnValue:EUW Asset Playlist Drag Drop Widget Object Reference)
   > n2 = "Create Drag && Drop Operation"!(
    execute:Exec
    WorldContextObject:Object Reference
    Class:Drag Drop Operation Class Reference=/Script/UMG.DragDropOperation
    Tag:String
    Payload:Object Reference
    DefaultDragVisual:Widget Object Reference=@n1
    Pivot:EDragPivot Enum=CenterCenter
    Offset:Vector 2D Structure=(X=0,Y=0)
   ) -> (then:Exec, ReturnValue:Drag Drop Operation Object Reference) {is=title K2Node_CreateDragDropOperation}
   > Return Node!(execute, Operation:Drag Drop Operation Object Reference=@n2)
  ]}
 OnMouseButtonDown {nodes [
   PrintString(InString=Right)
   > add Floating Context Menu!(
    execute
    self:EUW Asset Playlist Headless Object Reference=Get EUW_AssetPlaylist_headless() -> (EUW_AssetPlaylist_headless:EUW Asset Playlist Headless Object Reference)
    widgetToAdd:User Widget Class Reference=&/EUW_PlaylistGroupContextMenu.EUW_PlaylistGroupContextMenu_C
    spawningWidget:User Widget Object Reference=Self-Reference()
   ) -> (then, floatingWidget:EUW Floating Context Menu Object Reference)
   > Return Node!(execute, ReturnValue:Event Reply Structure=Handled())
   n1 = OnMouseButtonDown!() -> (then, MyGeometry:Geometry Structure, MouseEvent:"Pointer Event Structure (by ref)") {is=title K2Node_FunctionEntry}
   > n2 = Branch(Condition="Equal (Key)"(
     A:Key Structure=Get Effecting Button!(Input:"Pointer Event Structure (by ref)"=@n1.MouseEvent) -> (ReturnValue:Key Structure) {is=title K2Node_CallFunction}
     B:Key Structure=LeftMouseButton
    ) -> (ReturnValue:bool=~) {is=title K2Node_PromotableOperator})
   > PrintString(InString="Left  Down")
   > n3 = DetectDragIfPressed!(
    execute
    PointerEvent:"Pointer Event Structure (by ref)"=@n1.MouseEvent
    WidgetDetectingDrag:Widget Object Reference
    DragKey:Key Structure=LeftMouseButton
   ) -> (then, ReturnValue:Event Reply Structure)
   > Return Node(ReturnValue=@n3)
   PrintString(execute=@n2.else, InString=Right Down)
   > Return Node(ReturnValue=Handled())
  ]}
 OnMouseButtonUp {nodes [
   n1 = OnMouseButtonUp!() -> (then, MyGeometry:Geometry Structure, MouseEvent:"Pointer Event Structure (by ref)") {is=title K2Node_FunctionEntry}
   > n2 = Branch(Condition="Equal (Key)"(A=Get Effecting Button(Input=@n1.MouseEvent), B=LeftMouseButton) -> (ReturnValue=~))
   > PrintString(InString=Left Up)
   > leftUp()
   n3 = Handled()
   > Return Node(ReturnValue=@n3)
   PrintString(execute=@n2.else, InString=Right Up)
   > Return Node(ReturnValue=@n3)
  ]}
 SetImageButtonStyle {nodes [
   n1 = SetImageButtonStyle!() -> (
    then
    NorrmalFillColor:Linear Color Structure
    NormalOutlineColor:Linear Color Structure
    HoveredFillColor:Linear Color Structure
    HoveredOutlineColor:Linear Color Structure
   ) {is=title K2Node_FunctionEntry}
   > Set Normal Fill Color(Normal Fill Color:Linear Color Structure=@n1) -> (Output_Get:Linear Color Structure)
   > Set Normal Outline Color(Normal Outline Color:Linear Color Structure=@n1.NormalOutlineColor) -> (Output_Get:Linear Color Structure)
   > Set Hovered Fill Color(Hovered Fill Color:Linear Color Structure=@n1.HoveredFillColor) -> (Output_Get:Linear Color Structure)
   > Set Hovered Outline Color(Hovered Outline Color:Linear Color Structure=@n1.HoveredOutlineColor) -> (Output_Get:Linear Color Structure)
   > SetColorAndOpacity(self=Get Image_button_outer(), InColorAndOpacity=Get Normal Outline Color())
   > SetColorAndOpacity(self=Get Image_button_inner(), InColorAndOpacity=Get Normal Fill Color())
  ]}
 add Item Of Cosmetic Type To Array {nodes [
   n1 = add Item Of Cosmetic Type To Array!() -> (
    then
    TargetArray:Array of EUW Playlist Item Object References
    SourceArray:Array of EUW Playlist Item Object References
    cosmeticType:CosmeticPicker_CosmeticTypes Enum=NewEnumerator0
   ) {is=title K2Node_FunctionEntry}
   > n2 = For Each Loop(Array:Array of EUW Playlist Item Object References=@n1.SourceArray) -> (Array Element:EUW Playlist Item Object Reference)
   n3 = Get tempArray() -> (tempArray:Array of EUW Playlist Item Object References)
   n4 = For Each Loop(Exec=@n2.Completed, Array=@n3)
   n5 = Reverse for Each Loop!(Exec:Exec=@n4.Completed, Array:Array of EUW Playlist Item Object References=@n3) -> (LoopBody:Exec, ArrayIndex:int, ArrayElement:EUW Playlist Item Object Reference, Completed:Exec) {is=title K2Node_MacroInstance}
   n6 = For Each Loop(Exec=@n5.Completed, Array=Get sortedArray() -> (sortedArray:Array of EUW Playlist Item Object References))
   > Add(TargetArray:Array of EUW Playlist Item Object References=@n1, NewItem:"EUW Playlist Item Object Reference (by ref)"=@n6)
   Set added(execute=@n5.LoopBody, added:bool=false) -> (Output_Get:bool)
   > Branch(Condition=Get isVariant!(self:EUW Playlist Item Object Reference=@n5.ArrayElement) -> (isVariant:bool))
   > n7 = Set NameWithoutVariant(NameWithoutVariant:String=Split(SourceString=Get UIName!(self:EUW Playlist Item Object Reference=@n5.ArrayElement) -> (UIName:String), InStr=" ", SearchDir=FromEnd)) -> (Output_Get:String)
   > n8 = For Each Loop with Break!(Exec:Exec, Array:Array of EUW Playlist Item Object References=Get sortedArray(), Break:Exec=@n10.then) -> (LoopBody:Exec, Array Element:EUW Playlist Item Object Reference, Array Index:int, Completed:Exec)
   n9 = Branch(execute=@n8.Completed, Condition=Get added() -> (added:bool))
   Add(execute=@n9.else, TargetArray=Get sortedArray(), NewItem=@n5.ArrayElement)
   Branch(execute=@n8.LoopBody, Condition="Equal, Case Insensitive (String)"(A=Get UIName(self=@n8), B=@n7))
   > Insert!(
    execute
    TargetArray:Array of EUW Playlist Item Object References=Get sortedArray()
    NewItem:"EUW Playlist Item Object Reference (by ref)"=@n5.ArrayElement
    Index:int=int + int(A=@n8.Array Index)
   ) -> (then) {is=title K2Node_CallArrayFunction}
   > n10 = Set added(added=true)
   n11 = Branch(execute=@n4.LoopBody, Condition=Get isVariant(self=@n4))
   Add(execute=@n11.else, TargetArray=Get sortedArray(), NewItem=@n4)
   Branch(
    execute=@n2.LoopBody
    Condition="Equal (Enum)"(
     A:CosmeticPicker_CosmeticTypes Enum=Get cosmeticType!(self:EUW Playlist Item Object Reference=@n2) -> (cosmeticType:CosmeticPicker_CosmeticTypes Enum=NewEnumerator0)
     B:CosmeticPicker_CosmeticTypes Enum=@n1.cosmeticType
    )
   )
   > Add(TargetArray=Get tempArray(), NewItem=@n2)
   PrintString(InString=Get!(self:EUW Playlist Item Object Reference=@n4) -> (
     asset_shotGridId_27_#1:int
     asset_cosmeticType_20_#2:CosmeticPicker_CosmeticTypes Enum
     asset_name_22_#3:String
     asset_gender_25_#4:String
     asset_assetGroup_41_#5:String
     asset_primaryAssetGroup_42_#6:String
     asset_status_33_#7:String
     asset_fixVersion_37_#8:String
    ).asset_name_22_#3)
   GetDisplayName(Object=@n4)
  ]}
 clear Cosmetic Picker {nodes [
   clear Cosmetic Picker() {is=title K2Node_FunctionEntry}
   > PrintString(InString=ClearCosmeticPicker)
   > n1 = GetAllWidgetsOfClass!(
    execute
    WorldContextObject:Object Reference
    WidgetClass:User Widget Class Reference=/Game/NeverCook/ArtTools/Tech_Art/CosmeticPicker/EUW_CosmeticPickerv2.EUW_CosmeticPickerv2_C
    TopLevelOnly:bool=false
   ) -> (then, FoundWidgets:Array of EUW Cosmetic Pickerv 2 Object References)
   > n2 = For Each Loop(Array:Array of EUW Cosmetic Pickerv 2 Object References=@n1) -> (Array Element:EUW Cosmetic Pickerv 2 Object Reference)
   > clear(self:EUW Cosmetic Pickerv 2 Object Reference=@n2)
  ]}
 clear Overrides {nodes [
   clear Overrides() {is=title K2Node_FunctionEntry}
   > Execute Python Script(PythonScript="import TechArt.CosmeticPicker.CosmeticPickerModule as CosmeticPicker\r\nCosmeticPicker.clearOverrides()")
  ]}
 deselect All Widgets {nodes [
   deselect All Widgets() {is=title K2Node_FunctionEntry}
   > deselect Group()
   > n1 = For Each Loop(Array:Array of Widget Object References=GetAllChildren(self=Get VerticalBox_itemGroup())) -> (Array Element:Widget Object Reference)
   > n2 = Cast To EUW_PlaylistItem(Object=@n1)
   > deselect Item(self:EUW Playlist Item Object Reference=@n2)
  ]}
 deselect Group {nodes [
   deselect Group() {is=title K2Node_FunctionEntry}
   > SetImageButtonStyle!(
    execute
    self
    NorrmalFillColor:Linear Color Structure=(R=0.03,G=0.03,B=0.03,A=1)
    NormalOutlineColor:Linear Color Structure=(R=0.04,G=0.04,B=0.04,A=1)
    HoveredFillColor:Linear Color Structure=(R=0.03,G=0.03,B=0.03,A=1)
    HoveredOutlineColor:Linear Color Structure=(R=0.1,G=0.1,B=0.1,A=1)
   ) -> (then) {is=title K2Node_CallFunction}
  ]}
 deselect Widgets of CosmeticType {nodes [
   n1 = SetBackgroundColor!(execute, self:Button Object Reference, InBackgroundColor:Linear Color Structure=(R=1,G=1,B=1,A=1)) -> (then)
   n2 = For Each Loop(Exec=[@n1.then, @n4.then], Array=GetAllChildren(self=Get VerticalBox_itemGroup()))
   > n3 = Cast To EUW_PlaylistItem(Object=@n2)
   n4 = deselect Widgets of CosmeticType!() -> (then, cosmeticType:CosmeticPicker_CosmeticTypes Enum=NewEnumerator0) {is=title K2Node_FunctionEntry}
   Branch(execute=@n3.then, Condition="Equal (Enum)"(A=Get cosmeticType(self=@n3) -> (cosmeticType=NewEnumerator0), B=@n4))
   > deselect Item(self=@n3)
  ]}
 filter Items {nodes [
   n1 = filter Items!() -> (then, cosmeticType:CosmeticPicker_CosmeticTypes Enum=NewEnumerator0, knownItems:bool=~, unknown:bool=~, variants:bool=~, builtIn:bool=~) {is=title K2Node_FunctionEntry}
   > n2 = For Each Loop(Array=GetAllChildren(self=Get VerticalBox_itemGroup()))
   > n3 = Cast To EUW_PlaylistItem(Object=@n2)
   > Set visible(visible:bool=true) -> (Output_Get:bool)
   > n4 = Sequence!(execute) -> (then_0, then_1, then_2:Exec, then_3:Exec)
   > n5 = Branch(Condition=@n1.knownItems)
   n6 = Branch(execute=[@n5.then, @n10.Is Not Valid], Condition=@n1.variants)
   n7 = Branch(execute=[@n6.then, @n9.else], Condition=@n1.unknown)
   n8 = Is Valid(exec=@n7.else, InputObject=Get item!(self:EUW Playlist Item Object Reference=@n3) -> (item:Fort Item Definition Object Reference))
   Set visible(execute=@n8.Is Not Valid, visible=false)
   n9 = Branch(execute=@n6.else, Condition=Get isVariant(self=@n3))
   > Set visible(visible=false)
   n10 = Is Valid(exec=@n5.else, InputObject=Get item(self=@n3))
   > Set visible(visible=false)
   n11 = Branch(execute=@n4.then_1, Condition="Equal (Enum)"(A=@n1, B=NewEnumerator6))
   n12 = Branch(execute=@n11.else, Condition="Equal (Enum)"(A=Get cosmeticType(self=@n3) -> (cosmeticType=NewEnumerator0), B=@n1))
   Set visible(execute=@n12.else, visible=false)
   n13 = Branch(execute=@n4.then_3, Condition=Get visible() -> (visible:bool))
   > SetVisibility(self=@n3, InVisibility=Visible)
   SetVisibility(execute=@n13.else, self=@n3, InVisibility=Collapsed)
   n14 = Branch(execute=@n4.then_2, Condition=@n1.builtIn)
   n15 = Get originalName(self=@n3)
   n16 = Branch(execute=@n14.else, Condition=Contains(SearchIn=@n15, Substring=builtin))
   Set visible(execute=[@n17.then, @n16.then], visible=false)
   n17 = Branch(execute=@n16.else, Condition=Contains(SearchIn=@n15, Substring=built in))
  ]}
 get Items {nodes [
   get Items() {is=title K2Node_FunctionEntry}
   > n1 = For Each Loop(Array=GetAllChildren(self=Get VerticalBox_itemGroup()))
   Return Node!(
    execute=@n1.Completed
    result:Array of EUW Playlist Item Object References=Get items() -> (items:Array of EUW Playlist Item Object References)
   )
   > n2 = Cast To EUW_PlaylistItem(Object=@n1)
   > Add(TargetArray=Get items(), NewItem=@n2)
  ]}
 hide Empty Group {nodes [
   hide Empty Group() {is=title K2Node_FunctionEntry}
   n1 = GetAllChildren(self=Get VerticalBox_itemGroup())
   > Branch(Condition="Equal (Integer)"(A=Length(TargetArray:Array of Widget Object References=@n1)))
   > SetVisibility(InVisibility=Collapsed)
  ]}
 leftUp {nodes [
   leftUp() {is=title K2Node_FunctionEntry}
   > select Group()
  ]}
 reorder Group {nodes [
   reorder Group() {is=title K2Node_FunctionEntry}
   > PrintString(InString=reorderGroup)
   > n1 = For Each Loop(Array=GetAllChildren(self=Get VerticalBox_itemGroup()))
   add Item Of Cosmetic Type To Array!(
    execute=@n1.Completed
    self
    TargetArray:Array of EUW Playlist Item Object References=Get orderedChildren() -> (orderedChildren:Array of EUW Playlist Item Object References)
    SourceArray:Array of EUW Playlist Item Object References=Get children() -> (children:Array of EUW Playlist Item Object References)
    cosmeticType:CosmeticPicker_CosmeticTypes Enum=NewEnumerator0
   ) -> (then) {is=title K2Node_CallFunction}
   > add Item Of Cosmetic Type To Array(TargetArray=Get orderedChildren(), SourceArray=Get children(), cosmeticType=NewEnumerator2)
   > add Item Of Cosmetic Type To Array(TargetArray=Get orderedChildren(), SourceArray=Get children(), cosmeticType=NewEnumerator1)
   > add Item Of Cosmetic Type To Array(TargetArray=Get orderedChildren(), SourceArray=Get children(), cosmeticType=NewEnumerator3)
   > add Item Of Cosmetic Type To Array(TargetArray=Get orderedChildren(), SourceArray=Get children(), cosmeticType=NewEnumerator4)
   > add Item Of Cosmetic Type To Array(TargetArray=Get orderedChildren(), SourceArray=Get children(), cosmeticType=NewEnumerator5)
   > n2 = For Each Loop(Array:Array of EUW Playlist Item Object References=Get children()) -> (Array Element:EUW Playlist Item Object Reference)
   ClearChildren(execute=@n2.Completed, self=Get VerticalBox_itemGroup())
   > n3 = For Each Loop(Array=Get orderedChildren())
   > AddChild(self=Get VerticalBox_itemGroup(), Content=@n3)
   n4 = Branch(
    execute=@n2.LoopBody
    Condition=Contains Item(TargetArray:Array of EUW Playlist Item Object References=Get orderedChildren(), ItemToFind:"EUW Playlist Item Object Reference (by ref)"=@n2)
   )
   Add(execute=@n4.else, TargetArray=Get orderedChildren(), NewItem=@n2)
   n5 = Cast To EUW_PlaylistItem(execute=@n1.LoopBody, Object=@n1)
   > Add(TargetArray=Get children(), NewItem=@n5)
  ]}
 select Button {nodes [
   select Button() {is=title K2Node_FunctionEntry}
   > SetImageButtonStyle(
    NorrmalFillColor=(R=0.2,G=0.2,B=0.2,A=1)
    NormalOutlineColor=(R=0.3,G=0.3,B=0.3,A=1)
    HoveredFillColor=(R=0.4,G=0.4,B=0.4,A=1)
    HoveredOutlineColor=(R=0.7,G=0.7,B=0.7,A=1)
   )
  ]}
 select First of CosmeticType {nodes [
   n1 = select First of CosmeticType!() -> (then, Array:Array of Widget Object References, cosmeticType:CosmeticPicker_CosmeticTypes Enum=NewEnumerator0) {is=title K2Node_FunctionEntry}
   > n2 = For Each Loop(Array:Array of Widget Object References=@n1) -> (Array Element:Widget Object Reference)
   > n3 = Cast To EUW_PlaylistItem(Object=@n2)
   > Branch(Condition="Equal (Enum)"(A=Get cosmeticType(self=@n3) -> (cosmeticType=NewEnumerator0), B=@n1.cosmeticType))
   > Is Valid(InputObject=Get item(self=@n3))
   > Select Item(self:EUW Playlist Item Object Reference=@n3)
   > Return Node!(execute, success:bool=true)
  ]}
 select Group {nodes [
   select Group() {is=title K2Node_FunctionEntry}
   n1 = Self-Reference()
   > n2 = FindParentWidgetOfType(StartingWidget=@n1, Type=&ScrollBox) -> (ReturnValue:Scroll Box Object Reference)
   > n3 = Cast To ScrollBox!(execute, Object=@n2) -> (then, CastFailed, AsScroll Box:Scroll Box Object Reference) {is=title K2Node_DynamicCast}
   > PrintString(InString=Append!(A="ScrollBox ", B=GetDisplayName(Object=@n3), C:String=",", D:String=GetDisplayName(Object=@n1)) -> (ReturnValue))
   > n4 = ScrollWidgetIntoView!(
    execute
    self:Scroll Box Object Reference=@n3
    WidgetToFind:Widget Object Reference=@n1
    AnimateScroll:bool=true
    ScrollDestination:EDescendantScrollDestination Enum=IntoView
    Padding:float=0
   ) -> (then)
   PrintString(execute=[@n4.then, @n13.then], InString=Append!(A=Get groupName() -> (groupName:String), B=" group clicked") -> (ReturnValue))
   > n5 = GetAllWidgetsOfClass(WidgetClass=&AssetPlaylist_C, TopLevelOnly=false) -> (FoundWidgets:Array of Asset Playlist Object References)
   > n6 = For Each Loop(Array:Array of Asset Playlist Object References=@n5) -> (Array Element:Asset Playlist Object Reference)
   select Button(execute=@n6.Completed) {is=title K2Node_CallFunction}
   n7 = Get cosmeticPickerGroup() -> (cosmeticPickerGroup:EUW Playlist Group Object Reference)
   > n8 = Is Valid(InputObject=@n7)
   > n9 = select Button(self=@n7)
   n10 = Get assetPlaylistGroup()
   n11 = Is Valid(exec=[@n9.then, @n8.Is Not Valid], InputObject=@n10)
   clear Overrides(execute=@n11.Is Not Valid) {is=title K2Node_CallFunction}
   > clear Cosmetic Picker() {is=title K2Node_CallFunction}
   n12 = GetAllChildren(self=Get VerticalBox_itemGroup())
   > select First of CosmeticType!(execute, self, Array:Array of Widget Object References=@n12, cosmeticType:CosmeticPicker_CosmeticTypes Enum=NewEnumerator0) -> (then, success:bool) {is=title K2Node_CallFunction}
   > select First of CosmeticType(Array=@n12, cosmeticType=NewEnumerator1)
   > select First of CosmeticType(Array=@n12, cosmeticType=NewEnumerator2)
   > select First of CosmeticType(Array=@n12, cosmeticType=NewEnumerator3)
   > select First of CosmeticType(Array=@n12, cosmeticType=NewEnumerator5)
   select Group(execute=@n11.Is Valid, self=@n10) {is=title K2Node_CallFunction}
   deselect all Widgets(execute=@n6.LoopBody, self:Asset Playlist Object Reference=@n6)
   n13 = PrintString(execute=@n3.CastFailed, InString=NoScrollBox)
  ]}
 set Thumbnail {nodes [
   n1 = set Thumbnail!() -> (then, thumbnail:bool=~) {is=title K2Node_FunctionEntry}
   > n2 = Set thumbnails(thumbnails:bool=@n1) -> (Output_Get:bool)
   > PrintString(InString=Append!(A="Set Thumbnail ", B=Get groupName(), C:String=" ", D:String="To String (Boolean)"(InBool=@n2)) -> (ReturnValue))
   > n3 = Branch(Condition=@n2)
   > n4 = SetWidthOverride(self=Get SizeBox_top() -> (SizeBox_top:Size Box Object Reference), InWidthOverride=200)
   n5 = For Each Loop(Exec=[@n4.then, @n7.then], Array:Array of Widget Object References=GetAllChildren(self=Get VerticalBox_itemGroup())) -> (Array Element:Widget Object Reference)
   > n6 = Cast To EUW_PlaylistItem(Object=@n5)
   > set Thumbnail!(execute, self:EUW Playlist Item Object Reference=@n6, thumbnailVisibility:bool=@n1) -> (then) {is=title K2Node_CallFunction}
   n7 = SetWidthOverride(execute=@n3.else, self=Get SizeBox_top(), InWidthOverride=150)
  ]}
 setupGroup {nodes [
   n1 = setupGroup!() -> (then, group:String, gender:String) {is=title K2Node_FunctionEntry}
   > PrintString(InString=Append!(A="setUpGroup ", B=@n1) -> (ReturnValue))
   > Set gender(gender:String=@n1.gender) -> (Output_Get:String)
   > n2 = Branch(Condition="integer > integer"(A=Len(S=@n1), B=25))
   > n3 = Set groupDisplayName(groupDisplayName:String=Append(A=GetSubstring!(SourceString:String=@n1, StartIndex:int=0, Length:int=22) -> (ReturnValue:String), B=...)) -> (Output_Get:String)
   Set groupName(execute=[@n5.then, @n3.then], groupName:String=@n1) -> (Output_Get:String)
   n4 = Get groupDisplayName() -> (groupDisplayName:String)
   > "SetText (Text)"(self=Get CommonTextBlock_177() -> (CommonTextBlock_177:Common Text Object Reference), InText="To Text (String)"(InString=@n4))
   > PrintString(InString=Append(A="set group label to ", B=@n4))
   > SetToolTipText!(
    execute
    self:Widget Object Reference
    InToolTipText:"Text (by ref)"="To Text (String)"(InString=Append!(A=Press to select Cosmetics in ', B=@n1, C:String=') -> (ReturnValue))
   ) -> (then)
   > SetImageButtonStyle(
    NorrmalFillColor=(R=0.03,G=0.03,B=0.03,A=1)
    NormalOutlineColor=(R=0.04,G=0.04,B=0.04,A=1)
    HoveredFillColor=(R=0.03,G=0.03,B=0.03,A=1)
    HoveredOutlineColor=(R=0.1,G=0.1,B=0.1,A=1)
   )
   n5 = Set groupDisplayName(execute=@n2.else, groupDisplayName=@n1)
  ]}
 show Items {nodes [
   n1 = show Items!() -> (then, visibility:bool=~) {is=title K2Node_FunctionEntry}
   > n2 = get Items!(execute, self) -> (then, result:Array of EUW Playlist Item Object References) {is=title K2Node_CallFunction}
   > n3 = For Each Loop(Array:Array of EUW Playlist Item Object References=@n2) -> (Array Element:EUW Playlist Item Object Reference)
   > n4 = Branch(Condition=@n1)
   > SetVisibility(self=@n3, InVisibility=Visible)
   SetVisibility(execute=@n4.else, self=@n3, InVisibility=Collapsed)
  ]}
 spawn {nodes [
   n1 = spawn!() -> (then, spawnActorPosition:Vector=~) {is=title K2Node_FunctionEntry}
   > PrintString(InString=Spawn)
   > n2 = Set SpawnActorStartPosition(SpawnActorStartPosition:Vector=@n1) -> (Output_Get:Vector)
   > Set spawnActorPosition(spawnActorPosition:Vector=@n2) -> (Output_Get:Vector)
   > n3 = get Items()
   > n4 = For Each Loop(Array=@n3)
   n5 = Get item(self=@n4)
   > Is Valid(InputObject=@n5)
   > n6 = Branch(Condition="integer > integer"(A=@n4.Array Index))
   n7 = Get cosmeticType(self=@n4) -> (cosmeticType=NewEnumerator0)
   > n8 = Branch(Condition="Equal (Enum)"(A=@n7, B=Get lastCosmeticType() -> (lastCosmeticType:CosmeticPicker_CosmeticTypes Enum=NewEnumerator0)))
   n9 = GetDisplayName(Object=@n5)
   > PrintString(InString=Append!(A=Move Right, B=@n9) -> (ReturnValue))
   > n10 = Set spawnActorPosition(spawnActorPosition=vector + vector(A=Get spawnActorPosition() -> (spawnActorPosition:Vector), B="100,0,0"))
   Set lastCosmeticType(execute=[@n12.then, @n10.then, @n6.else], lastCosmeticType:CosmeticPicker_CosmeticTypes Enum=@n7) -> (Output_Get:CosmeticPicker_CosmeticTypes Enum=NewEnumerator0)
   > spawnItem!(execute, self:EUW Playlist Item Object Reference=@n4, position:Vector=Get spawnActorPosition()) -> (then)
   PrintString(execute=@n8.else, InString=Append(A="Move Forward ", B=@n9))
   > n11 = Set SpawnActorStartPosition(SpawnActorStartPosition=vector + vector(A=Get SpawnActorStartPosition() -> (SpawnActorStartPosition:Vector), B="0,50,0"))
   > n12 = Set spawnActorPosition(spawnActorPosition=@n11)
   n13 = For Each Loop(Array=@n3)
   n14 = Get item(self=@n13)
   > Is Valid(InputObject=@n14)
   > Branch(Condition="Equal (Enum)"(A=Get(self=@n13).asset_cosmeticType_20_#2, B=NewEnumerator0))
   > n15 = SpawnActor Fort Player Mannequin!(
    execute:Exec
    WorldContextObject:Object Reference
    Class:Actor Class Reference=/Script/FortniteGame.FortPlayerMannequin
    SpawnTransform_Location:Vector=Get spawnActorPosition()
    SpawnTransform_Rotation:Rotator="0, 0, 0"
    SpawnTransform_Scale:Vector="1,1,1"
    CollisionHandlingOverride:ESpawnActorCollisionHandlingMethod Enum=Undefined
    TransformScaleMethod:ESpawnActorScaleMethod Enum=MultiplyWithRoot
    Owner:Actor Object Reference
    Instigator:Pawn Object Reference
   ) -> (then:Exec, ReturnValue:Fort Player Mannequin Object Reference) {is=title K2Node_SpawnActorFromClass}
   > Set spawnActorPosition(spawnActorPosition=vector + vector(A=Get spawnActorPosition(), B=Get vectorOffset() -> (vectorOffset:Vector)))
   > SetEditorProperty(Object=@n15, PropertyName=athenaCharacter, PropertyValue:Fort Item Definition Object Reference=@n14)
   > Branch(Condition=Get isVariant(self=@n13))
   > n16 = get Override Variants!(execute, self:EUW Playlist Item Object Reference=@n13) -> (then, variants:Array of Mcp Variant Channel Info Structures)
   > SetEditorProperty(Object=@n15, PropertyName=overrideVariants, PropertyValue:Array of Mcp Variant Channel Info Structures=@n16)
  ]}
]
variables [
 assetPlaylistGroup {type=EUW_PlaylistGroup_C, default=null}
 cosmeticPickerGroup {type=EUW_PlaylistGroup_C, default=null}
 gender {type=string, default=""}
 groupDisplayName {type=string, default=""}
 groupName {type=string, default=""}
 Hovered Fill Color {type=LinearColor, default=<rgba 0 0 0 0>}
 Hovered Outline Color {type=LinearColor, default=<rgba 0 0 0 0>}
 NewVar_0 {type=integer}
 Normal Fill Color {type=LinearColor, default=<rgba 0 0 0 0>}
 Normal Outline Color {type=LinearColor, default=<rgba 0 0 0 0>}
 Partial Seconds {type=number}
 sg_asset_group {type=string, default=""}
 spawnActorPosition {type=Vector}
 thumbnails {type=boolean}
 vectorOffset {type=Vector, default=<100 0 0>}
]
widgets {
 info {rootWidgetClass=&CanvasPanel, widgetCount=11}
 widgets [
  CanvasPanel_0 {widgetClassPath=&CanvasPanel, widget=CanvasPanel_0}
  SizeBox_top {
   widgetClassPath=&SizeBox
   bIsVariable=true
   parent=CanvasPanel_0
   widget=SizeBox_top
   properties {bOverride_WidthOverride=true, widthOverride=200}
   slot {
    class=CanvasPanelSlot
    properties {bAutoSize=true, layoutData {alignment=<0 0>, anchors {maximum=<0 1>, minimum=<0 0>}, offsets=<ltrb 0 0 180 0>}}
   }
  }
  VerticalBox_topGroup {widgetClassPath=&VerticalBox, parent=SizeBox_top, widget=VerticalBox_topGroup, slot {class=SizeBoxSlot}}
  SizeBox {
   widgetClassPath=&SizeBox
   parent=VerticalBox_topGroup
   widget=SizeBox
   properties {bOverride_HeightOverride=true, bOverride_WidthOverride=true, heightOverride=40, widthOverride=80}
   slot {class=VerticalBoxSlot}
  }
  Overlay_453 {widgetClassPath=&Overlay, parent=SizeBox, widget=Overlay_453, slot {class=SizeBoxSlot}}
  Image_button_outer {
   widgetClassPath=&Image
   bIsVariable=true
   parent=Overlay_453
   widget=Image_button_outer
   properties {
    brush {
     bIsDynamicallyLoaded=false
     drawAs=RoundedBox
     imageType=NoImage
     mirroring=NoMirror
     resourceName=None
     resourceObject=null
     tiling=NoTile
     imageSize=<32 32>
     margin=<ltrb 0 0 0 0>
     outlineSettings {
      bUseBrushTransparency=false
      roundingType=HalfHeightRadius
      width=0
      color {colorUseRule=UseColor_Specified, specifiedColor=<rgba 0 0 0 0>}
      cornerRadii=<0 0 0 0>
     }
     tintColor {colorUseRule=UseColor_Specified, specifiedColor=<rgba 1 1 1 1>}
     uVRegion {bIsValid=false, max=<0 0>, min=<0 0>}
    }
    colorAndOpacity=<rgba 0.03 0.03 0.03 1>
   }
   slot {class=OverlaySlot, properties {horizontalAlignment=HAlign_Fill, verticalAlignment=VAlign_Fill}}
  }
  Image_button_inner {
   widgetClassPath=&Image
   bIsVariable=true
   parent=Overlay_453
   widget=Image_button_inner
   properties {
    brush {
     bIsDynamicallyLoaded=false
     drawAs=RoundedBox
     imageType=NoImage
     mirroring=NoMirror
     resourceName=None
     resourceObject=null
     tiling=NoTile
     imageSize=<32 32>
     margin=<ltrb 0 0 0 0>
     outlineSettings {
      bUseBrushTransparency=false
      roundingType=HalfHeightRadius
      width=0
      color {colorUseRule=UseColor_Specified, specifiedColor=<rgba 0 0 0 0>}
      cornerRadii=<0 0 0 0>
     }
     tintColor {colorUseRule=UseColor_Specified, specifiedColor=<rgba 1 1 1 1>}
     uVRegion {bIsValid=false, max=<0 0>, min=<0 0>}
    }
    colorAndOpacity=<rgba 0.03 0.03 0.03 1>
   }
   slot {class=OverlaySlot, properties {horizontalAlignment=HAlign_Fill, verticalAlignment=VAlign_Fill, padding=<ltrb 5 5 5 5>}}
  }
  CommonTextBlock_177 {
   widgetClassPath=&CommonTextBlock
   bIsVariable=true
   parent=Overlay_453
   widget=CommonTextBlock_177
   properties {
    justification=Center
    text=Group
    font {
     bForceMonospaced=false
     bMaterialIsStencil=false
     fontMaterial=null
     fontObject=&NotoSans
     letterSpacing=0
     monospacedWidth=1
     size=8
     skewAmount=0
     typefaceFontName=NotoSans-Regular
     outlineSettings {
      bApplyOutlineToDropShadows=false
      bMiteredCorners=false
      bSeparateFillAlpha=false
      outlineMaterial=null
      outlineSize=0
      outlineColor=<rgba 0 0 0 1>
     }
    }
    shadowOffset=<0 0>
   }
   slot {class=OverlaySlot, properties {horizontalAlignment=HAlign_Center, verticalAlignment=VAlign_Center, padding=<ltrb 4 2 4 2>}}
  }
  ScrollBox_160 {
   widgetClassPath=&ScrollBox
   parent=VerticalBox_topGroup
   widget=ScrollBox_160
   properties {
    scrollbarPadding=<ltrb 0 0 0 0>
    scrollbarThickness=<4 4>
    widgetBarStyle {
     thickness=4
     draggedThumbImage {
      bIsDynamicallyLoaded=false
      drawAs=RoundedBox
      imageType=NoImage
      mirroring=NoMirror
      resourceName=None
      resourceObject=null
      tiling=NoTile
      imageSize=<0 0>
      margin=<ltrb 0 0 0 0>
      outlineSettings {
       bUseBrushTransparency=false
       roundingType=FixedRadius
       width=0
       color {colorUseRule=UseColor_Specified, specifiedColor=<rgba 0 0 0 0>}
       cornerRadii=<4 4 4 4>
      }
      tintColor {colorUseRule=UseColor_Specified, specifiedColor=<rgba 0.2159 0.2159 0.2159 1>}
      uVRegion {bIsValid=false, max=<0 0>, min=<0 0>}
     }
     horizontalBackgroundImage {
      bIsDynamicallyLoaded=false
      drawAs=NoDrawType
      imageType=NoImage
      mirroring=NoMirror
      resourceName=None
      resourceObject=null
      tiling=NoTile
      imageSize=<0 0>
      margin=<ltrb 0 0 0 0>
      outlineSettings {
       bUseBrushTransparency=false
       roundingType=HalfHeightRadius
       width=0
       color {colorUseRule=UseColor_Specified, specifiedColor=<rgba 0 0 0 0>}
       cornerRadii=<0 0 0 0>
      }
      tintColor {colorUseRule=UseColor_Specified, specifiedColor=<rgba 1 1 1 1>}
      uVRegion {bIsValid=false, max=<0 0>, min=<0 0>}
     }
     horizontalBottomSlotImage {
      bIsDynamicallyLoaded=false
      drawAs=NoDrawType
      imageType=NoImage
      mirroring=NoMirror
      resourceName=None
      resourceObject=null
      tiling=NoTile
      imageSize=<0 0>
      margin=<ltrb 0 0 0 0>
      outlineSettings {
       bUseBrushTransparency=false
       roundingType=HalfHeightRadius
       width=0
       color {colorUseRule=UseColor_Specified, specifiedColor=<rgba 0 0 0 0>}
       cornerRadii=<0 0 0 0>
      }
      tintColor {colorUseRule=UseColor_Specified, specifiedColor=<rgba 1 1 1 1>}
      uVRegion {bIsValid=false, max=<0 0>, min=<0 0>}
     }
     horizontalTopSlotImage {
      bIsDynamicallyLoaded=false
      drawAs=NoDrawType
      imageType=NoImage
      mirroring=NoMirror
      resourceName=None
      resourceObject=null
      tiling=NoTile
      imageSize=<0 0>
      margin=<ltrb 0 0 0 0>
      outlineSettings {
       bUseBrushTransparency=false
       roundingType=HalfHeightRadius
       width=0
       color {colorUseRule=UseColor_Specified, specifiedColor=<rgba 0 0 0 0>}
       cornerRadii=<0 0 0 0>
      }
      tintColor {colorUseRule=UseColor_Specified, specifiedColor=<rgba 1 1 1 1>}
      uVRegion {bIsValid=false, max=<0 0>, min=<0 0>}
     }
     hoveredThumbImage {
      bIsDynamicallyLoaded=false
      drawAs=RoundedBox
      imageType=NoImage
      mirroring=NoMirror
      resourceName=None
      resourceObject=null
      tiling=NoTile
      imageSize=<0 0>
      margin=<ltrb 0 0 0 0>
      outlineSettings {
       bUseBrushTransparency=false
       roundingType=FixedRadius
       width=0
       color {colorUseRule=UseColor_Specified, specifiedColor=<rgba 0 0 0 0>}
       cornerRadii=<4 4 4 4>
      }
      tintColor {colorUseRule=UseColor_Specified, specifiedColor=<rgba 0.2159 0.2159 0.2159 1>}
      uVRegion {bIsValid=false, max=<0 0>, min=<0 0>}
     }
     normalThumbImage {
      bIsDynamicallyLoaded=false
      drawAs=RoundedBox
      imageType=NoImage
      mirroring=NoMirror
      resourceName=None
      resourceObject=null
      tiling=NoTile
      imageSize=<0 0>
      margin=<ltrb 0 0 0 0>
      outlineSettings {
       bUseBrushTransparency=false
       roundingType=FixedRadius
       width=0
       color {colorUseRule=UseColor_Specified, specifiedColor=<rgba 0 0 0 0>}
       cornerRadii=<4 4 4 4>
      }
      tintColor {colorUseRule=UseColor_Specified, specifiedColor=<rgba 0.0953 0.0953 0.0953 1>}
      uVRegion {bIsValid=false, max=<0 0>, min=<0 0>}
     }
     verticalBackgroundImage {
      bIsDynamicallyLoaded=false
      drawAs=NoDrawType
      imageType=NoImage
      mirroring=NoMirror
      resourceName=None
      resourceObject=null
      tiling=NoTile
      imageSize=<0 0>
      margin=<ltrb 0 0 0 0>
      outlineSettings {
       bUseBrushTransparency=false
       roundingType=HalfHeightRadius
       width=0
       color {colorUseRule=UseColor_Specified, specifiedColor=<rgba 0 0 0 0>}
       cornerRadii=<0 0 0 0>
      }
      tintColor {colorUseRule=UseColor_Specified, specifiedColor=<rgba 1 1 1 1>}
      uVRegion {bIsValid=false, max=<0 0>, min=<0 0>}
     }
     verticalBottomSlotImage {
      bIsDynamicallyLoaded=false
      drawAs=NoDrawType
      imageType=NoImage
      mirroring=NoMirror
      resourceName=None
      resourceObject=null
      tiling=NoTile
      imageSize=<0 0>
      margin=<ltrb 0 0 0 0>
      outlineSettings {
       bUseBrushTransparency=false
       roundingType=HalfHeightRadius
       width=0
       color {colorUseRule=UseColor_Specified, specifiedColor=<rgba 0 0 0 0>}
       cornerRadii=<0 0 0 0>
      }
      tintColor {colorUseRule=UseColor_Specified, specifiedColor=<rgba 1 1 1 1>}
      uVRegion {bIsValid=false, max=<0 0>, min=<0 0>}
     }
     verticalTopSlotImage {
      bIsDynamicallyLoaded=false
      drawAs=NoDrawType
      imageType=NoImage
      mirroring=NoMirror
      resourceName=None
      resourceObject=null
      tiling=NoTile
      imageSize=<0 0>
      margin=<ltrb 0 0 0 0>
      outlineSettings {
       bUseBrushTransparency=false
       roundingType=HalfHeightRadius
       width=0
       color {colorUseRule=UseColor_Specified, specifiedColor=<rgba 0 0 0 0>}
       cornerRadii=<0 0 0 0>
      }
      tintColor {colorUseRule=UseColor_Specified, specifiedColor=<rgba 1 1 1 1>}
      uVRegion {bIsValid=false, max=<0 0>, min=<0 0>}
     }
    }
    widgetStyle {
     barThickness=4
     bottomShadowBrush {
      bIsDynamicallyLoaded=false
      drawAs=Box
      imageType=FullColor
      mirroring=NoMirror
      resourceName=../../../Engine/Content/Slate/Common/ScrollBoxShadowBottom.png
      resourceObject=null
      tiling=NoTile
      imageSize=<16 8>
      margin=<ltrb 0.5 0 0.5 1>
      outlineSettings {
       bUseBrushTransparency=false
       roundingType=HalfHeightRadius
       width=0
       color {colorUseRule=UseColor_Specified, specifiedColor=<rgba 0 0 0 0>}
       cornerRadii=<0 0 0 0>
      }
      tintColor {colorUseRule=UseColor_Specified, specifiedColor=<rgba 1 1 1 1>}
      uVRegion {bIsValid=false, max=<0 0>, min=<0 0>}
     }
     horizontalScrolledContentPadding=<ltrb 0 0 1 0>
     leftShadowBrush {
      bIsDynamicallyLoaded=false
      drawAs=Box
      imageType=FullColor
      mirroring=NoMirror
      resourceName=../../../Engine/Content/Slate/Common/ScrollBoxShadowLeft.png
      resourceObject=null
      tiling=NoTile
      imageSize=<8 16>
      margin=<ltrb 1 0.5 0 0.5>
      outlineSettings {
       bUseBrushTransparency=false
       roundingType=HalfHeightRadius
       width=0
       color {colorUseRule=UseColor_Specified, specifiedColor=<rgba 0 0 0 0>}
       cornerRadii=<0 0 0 0>
      }
      tintColor {colorUseRule=UseColor_Specified, specifiedColor=<rgba 1 1 1 1>}
      uVRegion {bIsValid=false, max=<0 0>, min=<0 0>}
     }
     rightShadowBrush {
      bIsDynamicallyLoaded=false
      drawAs=Box
      imageType=FullColor
      mirroring=NoMirror
      resourceName=../../../Engine/Content/Slate/Common/ScrollBoxShadowRight.png
      resourceObject=null
      tiling=NoTile
      imageSize=<8 16>
      margin=<ltrb 0 0.5 1 0.5>
      outlineSettings {
       bUseBrushTransparency=false
       roundingType=HalfHeightRadius
       width=0
       color {colorUseRule=UseColor_Specified, specifiedColor=<rgba 0 0 0 0>}
       cornerRadii=<0 0 0 0>
      }
      tintColor {colorUseRule=UseColor_Specified, specifiedColor=<rgba 1 1 1 1>}
      uVRegion {bIsValid=false, max=<0 0>, min=<0 0>}
     }
     topShadowBrush {
      bIsDynamicallyLoaded=false
      drawAs=Box
      imageType=FullColor
      mirroring=NoMirror
      resourceName=../../../Engine/Content/Slate/Common/ScrollBoxShadowTop.png
      resourceObject=null
      tiling=NoTile
      imageSize=<16 8>
      margin=<ltrb 0.5 1 0.5 0>
      outlineSettings {
       bUseBrushTransparency=false
       roundingType=HalfHeightRadius
       width=0
       color {colorUseRule=UseColor_Specified, specifiedColor=<rgba 0 0 0 0>}
       cornerRadii=<0 0 0 0>
      }
      tintColor {colorUseRule=UseColor_Specified, specifiedColor=<rgba 1 1 1 1>}
      uVRegion {bIsValid=false, max=<0 0>, min=<0 0>}
     }
     verticalScrolledContentPadding=<ltrb 0 0 0 1>
    }
   }
   slot {class=VerticalBoxSlot, properties {size {sizeRule=Fill, value=1}}}
  }
  VerticalBox_itemGroup {
   widgetClassPath=&VerticalBox
   bIsVariable=true
   parent=ScrollBox_160
   widget=VerticalBox_itemGroup
   slot {class=ScrollBoxSlot, properties {verticalAlignment=VAlign_Top}}
  }
  EUW_AssetPlaylist_headless {
   widgetClassPath=&/EUW_AssetPlaylist_headless
   bIsVariable=true
   parent=CanvasPanel_0
   widget=EUW_AssetPlaylist_headless
   slot {class=CanvasPanelSlot, properties=~}
  }
 ]
}
