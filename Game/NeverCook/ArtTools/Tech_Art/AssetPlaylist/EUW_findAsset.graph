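# Editor utility widget that links a playlist item to a cosmetic item definition and stamps it with the item's ShotGrid ID.
# Setup embeds a cosmetic ID picker, then guesses the asset from UI name and gender, trimming trailing words for up to two retries.
# Apply writes non-default variant selections through Python, otherwise sets BaseShotGridId, and refreshes both playlist items.
# Save writes the picked definition to disk.
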EditorUtilityWidgetBlueprint /Game/NeverCook/ArtTools/Tech_Art/AssetPlaylist/EUW_findAsset
graphs [
 EventGraph {nodes [
   "On Clicked (Button_Guess)"() {is=title K2Node_ComponentBoundEvent}
   > PrintString(InString=Guess)
   > guessAsset()
   "On Clicked (Button_Apply)"() {is=title K2Node_ComponentBoundEvent}
   > PrintString(InString=Apply Button)
   > apply ShotGridId()
   "On Clicked (Button_Save)"() {is=title K2Node_ComponentBoundEvent}
   > PrintString(InString=Save Button)
   n1 = Get Item Definitiion!(self:EUW Cosmetic Pickerv 2 ID Object Reference=Get CosmeticPickerID() -> (CosmeticPickerID:EUW Cosmetic Pickerv 2 ID Object Reference)) -> (Item Definitiion:Fort Item Definition Object Reference)
   > Is Valid(InputObject=@n1)
   > PrintString(InString=GetDisplayName(Object=@n1))
   > SaveLoadedAsset(AssetToSave=@n1, bOnlyIfIsDirty=false)
   Event PreConstruct()
   Event Construct()
   Event Tick()
  ]}
 add Variant ShotGridId {nodes [
   n1 = add Variant ShotGridId!() -> (then, item:Object Reference, shotGridID:int, variantIndex:int, optionIndex:int) {is=title K2Node_FunctionEntry}
   > Execute Python Script!(
    execute
    PythonScript="import TechArt.CosmeticPlaylist.CosmeticPlaylistModule as CosmeticPlaylist\r\nCosmeticPlaylist.addVariantShotGridId(item, variant_index,  option_index, shot_grid_id)"
    item:Object Reference=@n1
    shot_grid_Id:int=@n1.shotGridID
    variant_index:int=@n1.variantIndex
    option_index:int=@n1.optionIndex
   ) -> (then, ReturnValue) {inputs [item, shot_grid_Id, variant_index, option_index]}
  ]}
 apply ShotGridId {nodes [
   apply ShotGridId() {is=title K2Node_FunctionEntry}
   n1 = Get Shot Grid ID!(self:EUW Playlist Item Object Reference=Get EUW_PlaylistItem() -> (EUW_PlaylistItem:EUW Playlist Item Object Reference)) -> (Shot Grid ID:int=0)
   > PrintString(InString="To String (Integer)"(InInt=@n1))
   > n2 = Cast To EUW_CosmeticPickerv2_ID!(
    execute
    Object=GetChildAt!(self:Panel Widget Object Reference=Get SizeBox_93() -> (SizeBox_93:Size Box Object Reference), Index:int=0) -> (ReturnValue:Widget Object Reference)
   ) -> (then, CastFailed, AsEUW Cosmetic Pickerv 2 ID:EUW Cosmetic Pickerv 2 ID Object Reference) {is=title K2Node_DynamicCast}
   > Set isVariant(isVariant:bool=false) -> (Output_Get:bool)
   > PrintString(InString=checking comboBox variants to see if any are non default)
   > n3 = getComboBoxVariantValues!(
    execute
    self:EUW Cosmetic Pickerv 2 Variants Object Reference=Get Variants!(self:EUW Cosmetic Pickerv 2 ID Object Reference=@n2) -> (Variants:EUW Cosmetic Pickerv 2 Variants Object Reference)
   ) -> (then, optionIndex:Array of Integers, defaultIndex:Array of Integers)
   > n4 = For Each Loop(Array:Array of Integers=@n3) -> (Array Element:int)
   n5 = Branch(execute=@n4.Completed, Condition=Get isVariant() -> (isVariant:bool))
   n6 = Get Item Definitiion(self=@n2)
   SetEditorProperty(execute=@n5.else, Object=@n6, PropertyName=BaseShotGridId, PropertyValue:int=@n1)
   > Is Valid(InputObject=Get spawning_assetPlaylistItem() -> (spawning_assetPlaylistItem:EUW Playlist Item Object Reference))
   n7 = Make Array("[0]":int=0) -> (Array:Array of Integers)
   n8 = Make Array("[0]"=0)
   n9 = Get spawning_assetPlaylistItem()
   > setup!(
    execute
    self:EUW Playlist Item Object Reference=Get EUW_PlaylistItem()
    item:Fort Item Definition Object Reference=@n6
    asset:Asset Playlist Asset Structure=Get asset!(self:EUW Playlist Item Object Reference=@n9) -> (asset:Asset Playlist Asset Structure)
    UIName:String=Get UIName!(self:EUW Playlist Item Object Reference=@n9) -> (UIName:String)
    isVariant:bool=false
    variantIndex:Array of Integers=@n7
    optionIndex:Array of Integers=@n8
   ) -> (then)
   n10 = Get spawning_assetPlaylistItem()
   > setup(self=@n10, item=@n6, asset=Get asset(self=@n10), UIName=Get UIName(self=@n10), isVariant=false, variantIndex=@n7, optionIndex=@n8)
   Branch(
    execute=@n4.LoopBody
    Condition="Not Equal (Integer)"(A=@n4, B="Get (a copy)"(Array:Array of Integers=@n3.defaultIndex, Dimension 1=@n4.Array Index) -> (Output:int))
   )
   > add Variant ShotGridId!(execute, self, item:Object Reference=@n6, shotGridID:int=@n1, variantIndex:int=@n4.Array Index, optionIndex:int=@n4) -> (then) {is=title K2Node_CallFunction}
   > Set isVariant(isVariant=true)
   > Is Valid(InputObject=Get spawning_assetPlaylistItem())
   n11 = Make Array("[0]"=@n4.Array Index)
   n12 = Make Array("[0]"=@n4)
   n13 = Get spawning_assetPlaylistItem()
   > setup(
    self=Get EUW_PlaylistItem()
    item=@n6
    asset=Get asset(self=@n13)
    UIName=Get UIName(self=@n13)
    isVariant=false
    variantIndex=@n11
    optionIndex=@n12
   )
   n14 = Get spawning_assetPlaylistItem()
   > setup(self=@n14, item=@n6, asset=Get asset(self=@n14), UIName=Get UIName(self=@n14), isVariant=false, variantIndex=@n11, optionIndex=@n12)
  ]}
 guessAsset {nodes [
   guessAsset() {is=title K2Node_FunctionEntry}
   > n1 = get Cosmetic Assets of CosmeticType!(
    execute
    self:EUW Asset Playlist Headless Object Reference=Get EUW_AssetPlaylist_headless() -> (EUW_AssetPlaylist_headless:EUW Asset Playlist Headless Object Reference)
    cosmeticType:CosmeticPicker_CosmeticTypes Enum=Get cosmeticType() -> (cosmeticType:CosmeticPicker_CosmeticTypes Enum=NewEnumerator0)
   ) -> (then, assets:Array of Asset Data Structures)
   n2 = Length(TargetArray:Array of Asset Data Structures=@n1)
   n3 = Get EUW_PlaylistItem()
   n4 = Get UIName(self=@n3)
   n5 = Get gender!(self:EUW Playlist Item Object Reference=@n3) -> (gender:String)
   > PrintString(InString=Append!(A=@n4, B=" '", C:String=@n5, D:String=', E:String=" numAssets:", F:String="To String (Integer)"(InInt=@n2)) -> (ReturnValue))
   > n6 = search For Cosmetic by Name!(
    execute
    self:EUW Asset Playlist Headless Object Reference=Get EUW_AssetPlaylist_headless()
    assets:Array of Asset Data Structures=@n1
    name:String=@n4
    gender:String=@n5
   ) -> (then, item:Object Reference)
   > n7 = Is Valid(InputObject=@n6)
   PrintString(execute=@n7.Is Not Valid, InString=Could NOT guess the asset - remove the last part of the string to see if it could be a variant)
   n8 = Split(SourceString=@n4, InStr=" ", SearchDir=FromEnd)
   > PrintString(InString=Append!(A=@n8, B=" '", C=@n5, D=') -> (ReturnValue))
   > n9 = Branch(Condition="integer > integer"(A=Len(S=@n8), B=1))
   > n10 = search For Cosmetic by Name(self=Get EUW_AssetPlaylist_headless(), assets=@n1, name=@n8, gender=@n5)
   > n11 = Is Valid(InputObject=@n10)
   PrintString(execute=@n11.Is Not Valid, InString=Could NOT guess the asset - remove the last part of the string to see if it could be a variant)
   n12 = Split(SourceString=@n8, InStr=" ", SearchDir=FromEnd)
   > PrintString(InString=Append(A=@n12, B=" '", C=@n5, D='))
   > n13 = Branch(Condition="integer > integer"(A=Len(S=@n12), B=1))
   > n14 = search For Cosmetic by Name(self=Get EUW_AssetPlaylist_headless(), assets=@n1, name=@n12, gender=@n5)
   > Is Valid(InputObject=@n14)
   > n15 = Cast To FortItemDefinition!(execute, Object=@n14) -> (then, CastFailed, AsFort Item Definition:Fort Item Definition Object Reference) {is=title K2Node_DynamicCast}
   > setID!(execute, self, item:Fort Item Definition Object Reference=@n15) -> (then)
   PrintString(execute=@n13.else, InString=String to short)
   > setID()
   n16 = Cast To FortItemDefinition(execute=@n11.Is Valid, Object=@n10)
   > setID(item=@n16)
   PrintString(execute=@n9.else, InString=String to short)
   > setID()
   n17 = Cast To FortItemDefinition(execute=@n7.Is Valid, Object=@n6)
   > setID(item=@n17)
  ]}
 setID {nodes [
   n1 = setID!() -> (then, item:Fort Item Definition Object Reference) {is=title K2Node_FunctionEntry}
   n2 = Get CosmeticPickerID()
   n3 = Get cosmeticType() -> (cosmeticType=NewEnumerator0)
   > PrintString(InString=Append!(
     A=GetDisplayName(Object=@n2)
     B=" "
     C=Enum to String!(Enumerator:CosmeticPicker_CosmeticTypes Enum=@n3) -> (ReturnValue:String)
     D=" "
     E:String=GetDisplayName(Object=@n1)
    ) -> (ReturnValue))
   n4 = Get cosmeticType() -> (cosmeticType=NewEnumerator0)
   n5 = Get CosmeticPickerID()
   > SetId!(
    execute
    self:EUW Cosmetic Pickerv 2 ID Object Reference=@n5
    id:Fort Item Definition Object Reference=@n1
    cosmeticType:CosmeticPicker_CosmeticTypes Enum=@n4
   ) -> (then)
   > updateVariants!(execute, self:EUW Cosmetic Pickerv 2 ID Object Reference=@n5, id:Fort Item Definition Object Reference=@n1) -> (then, Variants:EUW Cosmetic Pickerv 2 Variants Object Reference)
   > SetOverride!(
    execute
    self:EUW Cosmetic Pickerv 2 ID Object Reference=@n5
    id:Fort Item Definition Object Reference=@n1
    cosmeticType:CosmeticPicker_CosmeticTypes Enum=@n4
   ) -> (then)
  ]}
 setup {nodes [
   n1 = setup!() -> (
    then
    spawningAssetPlaylistItem:EUW Playlist Item Object Reference
    cosmeticType:CosmeticPicker_CosmeticTypes Enum=NewEnumerator0
    loadCheatValues:bool=true
   ) {is=title K2Node_FunctionEntry}
   > Set spawning_assetPlaylistItem(spawning_assetPlaylistItem:EUW Playlist Item Object Reference=@n1) -> (Output_Get:EUW Playlist Item Object Reference)
   n2 = Get SizeBox_93()
   > ClearChildren(self=@n2)
   > n3 = Create Widget!(
    execute:Exec
    WorldContextObject:Object Reference
    Class:User Widget Class Reference=/Game/NeverCook/ArtTools/Tech_Art/CosmeticPicker/EUW_CosmeticPickerv2_ID.EUW_CosmeticPickerv2_ID_C
    OwningPlayer:Player Controller Object Reference
   ) -> (then:Exec, ReturnValue:EUW Cosmetic Pickerv 2 ID Object Reference)
   > Set CosmeticPickerID(CosmeticPickerID:EUW Cosmetic Pickerv 2 ID Object Reference=@n3) -> (Output_Get:EUW Cosmetic Pickerv 2 ID Object Reference)
   > AddChild(self=@n2, Content=Get CosmeticPickerID())
   > Setup!(
    execute
    self:EUW Cosmetic Pickerv 2 ID Object Reference=Get CosmeticPickerID()
    CosmeticType:CosmeticPicker_CosmeticTypes Enum=@n1.cosmeticType
    loadCheatValues:bool=@n1.loadCheatValues
   ) -> (then)
   > guessAsset() {is=title K2Node_CallFunction}
   n4 = Enum to String(Enumerator=@n1.cosmeticType)
   n5 = Get DetailsView_ID!(self:EUW Cosmetic Pickerv 2 ID Object Reference=Get CosmeticPickerID()) -> (DetailsView_ID:Details View Object Reference)
   SetEditorProperty(
    Object=@n5
    PropertyName=CategoriesToShow
    PropertyValue:Array of Names=Make Array("[0]":Name=String To Name(InString=@n4)) -> (Array:Array of Names)
   )
   > SetObject(self=@n5, NewObject=Get CosmeticPickerID())
  ]}
]
variables [
 CosmeticPickerID {type=EUW_CosmeticPickerv2_ID_C, default=null}
 cosmeticType {type=CosmeticPicker_CosmeticTypes, default=Character}
 spawning_assetPlaylistItem {type=EUW_PlaylistItem_C, default=null}
]
widgets {
 info {rootWidgetClass=&CanvasPanel, widgetCount=14}
 widgets [
  CanvasPanel_0 {widgetClassPath=&CanvasPanel, widget=CanvasPanel_0}
  HorizontalBox_76 {
   widgetClassPath=&HorizontalBox
   parent=CanvasPanel_0
   widget=HorizontalBox_76
   slot {
    class=CanvasPanelSlot
    properties {bAutoSize=true, layoutData {alignment=<0 0>, anchors {maximum=<1 0>, minimum=<0 0>}, offsets=<ltrb 0 0 0 50>}}
   }
  }
  EUW_PlaylistItem {
   widgetClassPath=&/EUW_PlaylistItem
   bIsVariable=true
   parent=HorizontalBox_76
   widget=EUW_PlaylistItem
   slot {class=HorizontalBoxSlot}
  }
  SizeBox_219 {
   widgetClassPath=&SizeBox
   parent=HorizontalBox_76
   widget=SizeBox_219
   properties {bOverride_HeightOverride=true, bOverride_WidthOverride=true, heightOverride=51, widthOverride=100}
   slot {class=HorizontalBoxSlot, properties {verticalAlignment=VAlign_Top}}
  }
  Button_Guess {
   widgetClassPath=&Button
   parent=SizeBox_219
   widget=Button_Guess
   properties {widgetStyle {
     clickedSlateSound {resourceObject=null}
     disabled {
      bIsDynamicallyLoaded=false
      drawAs=RoundedBox
      imageType=NoImage
      mirroring=NoMirror
      resourceName=None
      resourceObject=null
      tiling=NoTile
      imageSize=<0 0>
      margin=<ltrb 0 0 0 0>
      outlineSettings {
       bUseBrushTransparency=false
       roundingType=FixedRadius
       width=1
       color {colorUseRule=UseColor_ColorTable, specifiedColor=<rgba 0 0 0 0>}
       cornerRadii=<4 4 4 4>
      }
      tintColor {colorUseRule=UseColor_Specified, specifiedColor=<rgba 0.0395 0.0395 0.0395 1>}
      uVRegion {bIsValid=false, max=<0 0>, min=<0 0>}
     }
     disabledForeground {colorUseRule=UseColor_ColorTable, specifiedColor=<rgba 0 0 0 0>}
     hovered {
      bIsDynamicallyLoaded=false
      drawAs=RoundedBox
      imageType=NoImage
      mirroring=NoMirror
      resourceName=None
      resourceObject=null
      tiling=NoTile
      imageSize=<0 0>
      margin=<ltrb 0 0 0 0>
      outlineSettings {
       bUseBrushTransparency=false
       roundingType=FixedRadius
       width=1
       color {colorUseRule=UseColor_ColorTable, specifiedColor=<rgba 0 0 0 0>}
       cornerRadii=<4 4 4 4>
      }
      tintColor {colorUseRule=UseColor_Specified, specifiedColor=<rgba 0.0953 0.0953 0.0953 1>}
      uVRegion {bIsValid=false, max=<0 0>, min=<0 0>}
     }
     hoveredForeground {colorUseRule=UseColor_ColorTable, specifiedColor=<rgba 0 0 0 0>}
     hoveredSlateSound {resourceObject=null}
     normal {
      bIsDynamicallyLoaded=false
      drawAs=RoundedBox
      imageType=NoImage
      mirroring=NoMirror
      resourceName=None
      resourceObject=null
      tiling=NoTile
      imageSize=<0 0>
      margin=<ltrb 0 0 0 0>
      outlineSettings {
       bUseBrushTransparency=false
       roundingType=FixedRadius
       width=1
       color {colorUseRule=UseColor_ColorTable, specifiedColor=<rgba 0 0 0 0>}
       cornerRadii=<4 4 4 4>
      }
      tintColor {colorUseRule=UseColor_Specified, specifiedColor=<rgba 0.0395 0.0395 0.0395 1>}
      uVRegion {bIsValid=false, max=<0 0>, min=<0 0>}
     }
     normalForeground {colorUseRule=UseColor_ColorTable, specifiedColor=<rgba 0 0 0 0>}
     normalPadding=<ltrb 12 1.5 12 1.5>
     pressed {
      bIsDynamicallyLoaded=false
      drawAs=RoundedBox
      imageType=NoImage
      mirroring=NoMirror
      resourceName=None
      resourceObject=null
      tiling=NoTile
      imageSize=<0 0>
      margin=<ltrb 0 0 0 0>
      outlineSettings {
       bUseBrushTransparency=false
       roundingType=FixedRadius
       width=1
       color {colorUseRule=UseColor_ColorTable, specifiedColor=<rgba 0 0 0 0>}
       cornerRadii=<4 4 4 4>
      }
      tintColor {colorUseRule=UseColor_Specified, specifiedColor=<rgba 0.0284 0.0284 0.0284 1>}
      uVRegion {bIsValid=false, max=<0 0>, min=<0 0>}
     }
     pressedForeground {colorUseRule=UseColor_ColorTable, specifiedColor=<rgba 0 0 0 0>}
     pressedPadding=<ltrb 12 2 12 0>
     pressedSlateSound {resourceObject=null}
    }}
   slot {class=SizeBoxSlot}
  }
  CommonTextBlock_1 {
   widgetClassPath=&CommonTextBlock
   bIsVariable=true
   parent=Button_Guess
   widget=CommonTextBlock_1
   properties {
    justification=Center
    text=Guess
    font {
     bForceMonospaced=false
     bMaterialIsStencil=false
     fontMaterial=null
     fontObject=&NotoSans
     letterSpacing=0
     monospacedWidth=1
     size=8
     skewAmount=0
     typefaceFontName=NotoSans-Regular
     outlineSettings {
      bApplyOutlineToDropShadows=false
      bMiteredCorners=false
      bSeparateFillAlpha=false
      outlineMaterial=null
      outlineSize=0
      outlineColor=<rgba 0 0 0 1>
     }
    }
    shadowOffset=<0 0>
   }
   slot {class=ButtonSlot}
  }
  SizeBox_93 {
   widgetClassPath=&SizeBox
   bIsVariable=true
   parent=HorizontalBox_76
   widget=SizeBox_93
   properties {bOverride_MinDesiredWidth=true, minDesiredWidth=600, widthOverride=600}
   slot {class=HorizontalBoxSlot, properties {size {sizeRule=Fill, value=1}}}
  }
  SizeBox_305 {
   widgetClassPath=&SizeBox
   parent=HorizontalBox_76
   widget=SizeBox_305
   properties {bOverride_HeightOverride=true, bOverride_WidthOverride=true, heightOverride=51, widthOverride=100}
   slot {class=HorizontalBoxSlot, properties {verticalAlignment=VAlign_Top}}
  }
  Button_Apply {
   widgetClassPath=&Button
   parent=SizeBox_305
   widget=Button_Apply
   properties {widgetStyle {
     clickedSlateSound {resourceObject=null}
     disabled {
      bIsDynamicallyLoaded=false
      drawAs=RoundedBox
      imageType=NoImage
      mirroring=NoMirror
      resourceName=None
      resourceObject=null
      tiling=NoTile
      imageSize=<0 0>
      margin=<ltrb 0 0 0 0>
      outlineSettings {
       bUseBrushTransparency=false
       roundingType=FixedRadius
       width=1
       color {colorUseRule=UseColor_ColorTable, specifiedColor=<rgba 0 0 0 0>}
       cornerRadii=<4 4 4 4>
      }
      tintColor {colorUseRule=UseColor_Specified, specifiedColor=<rgba 0.0395 0.0395 0.0395 1>}
      uVRegion {bIsValid=false, max=<0 0>, min=<0 0>}
     }
     disabledForeground {colorUseRule=UseColor_ColorTable, specifiedColor=<rgba 0 0 0 0>}
     hovered {
      bIsDynamicallyLoaded=false
      drawAs=RoundedBox
      imageType=NoImage
      mirroring=NoMirror
      resourceName=None
      resourceObject=null
      tiling=NoTile
      imageSize=<0 0>
      margin=<ltrb 0 0 0 0>
      outlineSettings {
       bUseBrushTransparency=false
       roundingType=FixedRadius
       width=1
       color {colorUseRule=UseColor_ColorTable, specifiedColor=<rgba 0 0 0 0>}
       cornerRadii=<4 4 4 4>
      }
      tintColor {colorUseRule=UseColor_Specified, specifiedColor=<rgba 0.0953 0.0953 0.0953 1>}
      uVRegion {bIsValid=false, max=<0 0>, min=<0 0>}
     }
     hoveredForeground {colorUseRule=UseColor_ColorTable, specifiedColor=<rgba 0 0 0 0>}
     hoveredSlateSound {resourceObject=null}
     normal {
      bIsDynamicallyLoaded=false
      drawAs=RoundedBox
      imageType=NoImage
      mirroring=NoMirror
      resourceName=None
      resourceObject=null
      tiling=NoTile
      imageSize=<0 0>
      margin=<ltrb 0 0 0 0>
      outlineSettings {
       bUseBrushTransparency=false
       roundingType=FixedRadius
       width=1
       color {colorUseRule=UseColor_ColorTable, specifiedColor=<rgba 0 0 0 0>}
       cornerRadii=<4 4 4 4>
      }
      tintColor {colorUseRule=UseColor_Specified, specifiedColor=<rgba 0.0395 0.0395 0.0395 1>}
      uVRegion {bIsValid=false, max=<0 0>, min=<0 0>}
     }
     normalForeground {colorUseRule=UseColor_ColorTable, specifiedColor=<rgba 0 0 0 0>}
     normalPadding=<ltrb 12 1.5 12 1.5>
     pressed {
      bIsDynamicallyLoaded=false
      drawAs=RoundedBox
      imageType=NoImage
      mirroring=NoMirror
      resourceName=None
      resourceObject=null
      tiling=NoTile
      imageSize=<0 0>
      margin=<ltrb 0 0 0 0>
      outlineSettings {
       bUseBrushTransparency=false
       roundingType=FixedRadius
       width=1
       color {colorUseRule=UseColor_ColorTable, specifiedColor=<rgba 0 0 0 0>}
       cornerRadii=<4 4 4 4>
      }
      tintColor {colorUseRule=UseColor_Specified, specifiedColor=<rgba 0.0284 0.0284 0.0284 1>}
      uVRegion {bIsValid=false, max=<0 0>, min=<0 0>}
     }
     pressedForeground {colorUseRule=UseColor_ColorTable, specifiedColor=<rgba 0 0 0 0>}
     pressedPadding=<ltrb 12 2 12 0>
     pressedSlateSound {resourceObject=null}
    }}
   slot {class=SizeBoxSlot}
  }
  CommonTextBlock {
   widgetClassPath=&CommonTextBlock
   bIsVariable=true
   parent=Button_Apply
   widget=CommonTextBlock
   properties {
    justification=Center
    text=Apply ShotGridId
    font {
     bForceMonospaced=false
     bMaterialIsStencil=false
     fontMaterial=null
     fontObject=&NotoSans
     letterSpacing=0
     monospacedWidth=1
     size=8
     skewAmount=0
     typefaceFontName=NotoSans-Regular
     outlineSettings {
      bApplyOutlineToDropShadows=false
      bMiteredCorners=false
      bSeparateFillAlpha=false
      outlineMaterial=null
      outlineSize=0
      outlineColor=<rgba 0 0 0 1>
     }
    }
    shadowOffset=<0 0>
   }
   slot {class=ButtonSlot}
  }
  SizeBox {
   widgetClassPath=&SizeBox
   parent=HorizontalBox_76
   widget=SizeBox
   properties {bOverride_HeightOverride=true, bOverride_WidthOverride=true, heightOverride=51, widthOverride=100}
   slot {class=HorizontalBoxSlot, properties {verticalAlignment=VAlign_Top}}
  }
  Button_Save {
   widgetClassPath=&Button
   parent=SizeBox
   widget=Button_Save
   properties {widgetStyle {
     clickedSlateSound {resourceObject=null}
     disabled {
      bIsDynamicallyLoaded=false
      drawAs=RoundedBox
      imageType=NoImage
      mirroring=NoMirror
      resourceName=None
      resourceObject=null
      tiling=NoTile
      imageSize=<0 0>
      margin=<ltrb 0 0 0 0>
      outlineSettings {
       bUseBrushTransparency=false
       roundingType=FixedRadius
       width=1
       color {colorUseRule=UseColor_ColorTable, specifiedColor=<rgba 0 0 0 0>}
       cornerRadii=<4 4 4 4>
      }
      tintColor {colorUseRule=UseColor_Specified, specifiedColor=<rgba 0.0395 0.0395 0.0395 1>}
      uVRegion {bIsValid=false, max=<0 0>, min=<0 0>}
     }
     disabledForeground {colorUseRule=UseColor_ColorTable, specifiedColor=<rgba 0 0 0 0>}
     hovered {
      bIsDynamicallyLoaded=false
      drawAs=RoundedBox
      imageType=NoImage
      mirroring=NoMirror
      resourceName=None
      resourceObject=null
      tiling=NoTile
      imageSize=<0 0>
      margin=<ltrb 0 0 0 0>
      outlineSettings {
       bUseBrushTransparency=false
       roundingType=FixedRadius
       width=1
       color {colorUseRule=UseColor_ColorTable, specifiedColor=<rgba 0 0 0 0>}
       cornerRadii=<4 4 4 4>
      }
      tintColor {colorUseRule=UseColor_Specified, specifiedColor=<rgba 0.0953 0.0953 0.0953 1>}
      uVRegion {bIsValid=false, max=<0 0>, min=<0 0>}
     }
     hoveredForeground {colorUseRule=UseColor_ColorTable, specifiedColor=<rgba 0 0 0 0>}
     hoveredSlateSound {resourceObject=null}
     normal {
      bIsDynamicallyLoaded=false
      drawAs=RoundedBox
      imageType=NoImage
      mirroring=NoMirror
      resourceName=None
      resourceObject=null
      tiling=NoTile
      imageSize=<0 0>
      margin=<ltrb 0 0 0 0>
      outlineSettings {
       bUseBrushTransparency=false
       roundingType=FixedRadius
       width=1
       color {colorUseRule=UseColor_ColorTable, specifiedColor=<rgba 0 0 0 0>}
       cornerRadii=<4 4 4 4>
      }
      tintColor {colorUseRule=UseColor_Specified, specifiedColor=<rgba 0.0395 0.0395 0.0395 1>}
      uVRegion {bIsValid=false, max=<0 0>, min=<0 0>}
     }
     normalForeground {colorUseRule=UseColor_ColorTable, specifiedColor=<rgba 0 0 0 0>}
     normalPadding=<ltrb 12 1.5 12 1.5>
     pressed {
      bIsDynamicallyLoaded=false
      drawAs=RoundedBox
      imageType=NoImage
      mirroring=NoMirror
      resourceName=None
      resourceObject=null
      tiling=NoTile
      imageSize=<0 0>
      margin=<ltrb 0 0 0 0>
      outlineSettings {
       bUseBrushTransparency=false
       roundingType=FixedRadius
       width=1
       color {colorUseRule=UseColor_ColorTable, specifiedColor=<rgba 0 0 0 0>}
       cornerRadii=<4 4 4 4>
      }
      tintColor {colorUseRule=UseColor_Specified, specifiedColor=<rgba 0.0284 0.0284 0.0284 1>}
      uVRegion {bIsValid=false, max=<0 0>, min=<0 0>}
     }
     pressedForeground {colorUseRule=UseColor_ColorTable, specifiedColor=<rgba 0 0 0 0>}
     pressedPadding=<ltrb 12 2 12 0>
     pressedSlateSound {resourceObject=null}
    }}
   slot {class=SizeBoxSlot}
  }
  CommonTextBlock_2 {
   widgetClassPath=&CommonTextBlock
   bIsVariable=true
   parent=Button_Save
   widget=CommonTextBlock_2
   properties {
    justification=Center
    text=Save Asset
    font {
     bForceMonospaced=false
     bMaterialIsStencil=false
     fontMaterial=null
     fontObject=&NotoSans
     letterSpacing=0
     monospacedWidth=1
     size=8
     skewAmount=0
     typefaceFontName=NotoSans-Regular
     outlineSettings {
      bApplyOutlineToDropShadows=false
      bMiteredCorners=false
      bSeparateFillAlpha=false
      outlineMaterial=null
      outlineSize=0
      outlineColor=<rgba 0 0 0 1>
     }
    }
    shadowOffset=<0 0>
   }
   slot {class=ButtonSlot}
  }
  EUW_AssetPlaylist_headless {
   widgetClassPath=&/EUW_AssetPlaylist_headless
   bIsVariable=true
   parent=CanvasPanel_0
   widget=EUW_AssetPlaylist_headless
   slot {class=CanvasPanelSlot, properties=~}
  }
 ]
}
