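# Editor utility widget for one cosmetic entry in the Asset Playlist tech-art tool, styled by cosmetic type with an optional thumbnail.
# Left-click toggles selection per type, applying it through CosmeticPicker Python scripts, PIE emote cheats or picker-widget wrap calls.
# Entries without items can open a find-asset tab, and right-click opens a context menu.
# A cancelled drag doubles as a viewport drop, line-tracing from the cursor to spawn the item.
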EditorUtilityWidgetBlueprint /Game/NeverCook/ArtTools/Tech_Art/AssetPlaylist/EUW_PlaylistItem
graphs [
 EventGraph {nodes [
   n1 = GetViewportWidgetGeometry!(execute, WorldContextObject:Object Reference) -> (then, ReturnValue:Geometry Structure)
   n2 = Get Screen Space Position!(Input:"Pointer Event Structure (by ref)") -> (ReturnValue:Vector 2D Structure) {is=title K2Node_CallFunction}
   n3 = ScreenToViewport!(WorldContextObject:Object Reference, ScreenPosition:Vector 2D Structure=@n2) -> (ViewportPosition:Vector 2D Structure)
   n4 = GetMousePositionOnViewport!(execute, WorldContextObject:Object Reference) -> (then, ReturnValue:Vector 2D Structure)
   n5 = GetLocalSize!(Geometry:"Geometry Structure (by ref)"=@n1) -> (ReturnValue:Vector 2D Structure)
   PrintString(
    execute=@n1.then
    InString=Append!(
     A="To String (Vector2d)"!(InVec:Vector 2D Structure=@n2) -> (ReturnValue:String)
     B="To String (Vector2d)"(InVec=@n3)
     C:String="To String (Vector2d)"(InVec=@n4)
     D:String="To String (Vector2d)"()
     E:String="To String (Vector2d)"(InVec=@n5)
    ) -> (ReturnValue)
   )
   n6 = Get Cursor World Ray!(execute) -> (then, CameraLocation:Vector, RayOrigin:Vector, RayDirection:Vector, ReturnValue:bool) {is=title K2Node_CallFunction}
   n7 = GetMousePositionOnViewport()
   PrintString(execute=[@n6.then, @n7.then, @n8.then], InString="To String (Vector2d)"(InVec=@n7))
   n8 = GetMousePositionOnViewport()
   Event OnMouseEnter()
   > SetColorAndOpacity(
    self:Image Object Reference=Get Image_button_outer() -> (Image_button_outer:Image Object Reference)
    InColorAndOpacity:Linear Color Structure=Get Hovered Outline Color() -> (Hovered Outline Color:Linear Color Structure)
   )
   > SetColorAndOpacity(
    self=Get Image_button_inner() -> (Image_button_inner:Image Object Reference)
    InColorAndOpacity=Get Hovered Fill Color() -> (Hovered Fill Color:Linear Color Structure)
   )
   Event OnMouseLeave!() -> (OutputDelegate, then, MouseEvent:"Pointer Event Structure (by ref)") {is=title K2Node_Event}
   > SetColorAndOpacity(self=Get Image_button_outer(), InColorAndOpacity=Get Normal Outline Color() -> (Normal Outline Color:Linear Color Structure))
   > SetColorAndOpacity(self=Get Image_button_inner(), InColorAndOpacity=Get Normal Fill Color() -> (Normal Fill Color:Linear Color Structure))
   Event OnDragCancelled!() -> (OutputDelegate, then, PointerEvent:"Pointer Event Structure (by ref)", Operation:Drag Drop Operation Object Reference) {is=title K2Node_Event}
   > PrintString(InString=Drag Cancelled)
   > Delay!(execute, WorldContextObject:Object Reference, Duration=0.2) -> (then)
   > PrintString(InString=GetDisplayName(Object=Get item() -> (item:Fort Item Definition Object Reference)))
   > n9 = Get Cursor World Ray()
   > n10 = Line Trace By Channel!(
    execute
    WorldContextObject:Object Reference
    Start:Vector=@n9.RayOrigin
    End:Vector=vector + vector(A=@n9.RayOrigin, B=vector * vector(A=@n9.RayDirection, B:Vector="9999,9999,9999") -> (ReturnValue=~))
    TraceChannel:ETraceTypeQuery Enum
    bTraceComplex:bool=false
    ActorsToIgnore:Array of Actor Object References
    DrawDebugType:EDrawDebugTrace Enum
    bIgnoreSelf:bool
    TraceColor:Linear Color Structure
    TraceHitColor:Linear Color Structure
    DrawTime:float
   ) -> (
    then
    OutHit_bBlockingHit:bool
    OutHit_bInitialOverlap:bool
    OutHit_Time:float
    OutHit_Distance:float
    OutHit_Location:Vector
    OutHit_ImpactPoint:Vector
    OutHit_Normal:Vector
    OutHit_ImpactNormal:Vector
    OutHit_PhysMat:Physical Material Object Reference
    OutHit_HitActor:Actor Object Reference
    OutHit_HitComponent:Primitive Component Object Reference
    OutHit_HitBoneName:Name=None
    OutHit_BoneName:Name=None
    OutHit_HitItem:int=0
    OutHit_ElementIndex:int=0
    OutHit_FaceIndex:int=0
    OutHit_TraceStart:Vector
    OutHit_TraceEnd:Vector
    ReturnValue:bool
   )
   > spawnItem!(execute, self, position:Vector=@n10.OutHit_Location) -> (then)
   GetParent!(self:Widget Object Reference) -> (ReturnValue:Panel Widget Object Reference)
   "To String (Boolean)"(InBool=@n6.ReturnValue)
   "To String (Vector2d)"(InVec="Local to Absolute (Coordinates)"!(Geometry:"Geometry Structure (by ref)"=@n1, LocalCoordinate:Vector 2D Structure=@n4) -> (ReturnValue:Vector 2D Structure))
   "To String (Vector2d)"(InVec=@n7)
   "To String (Vector2d)"(InVec=GetViewportSize!(WorldContextObject:Object Reference) -> (ReturnValue:Vector 2D Structure))
   "To String (Vector2d)"(InVec=@n8)
   GetLevelViewportCameraInfo!(WorldContextObject:Object Reference) -> (CameraLocation:Vector, CameraRotation:Rotator, ReturnValue:bool)
   Event PreConstruct()
   Event Construct()
   Event Tick()
  ]}
 Get Widget Style Colors {nodes [
   n1 = Get Widget Style Colors!() -> (then, inputColor:Linear Color Structure, outlineMult:double=1.3, hoveredMult:double=4, hoveredOutlineMult:double=5) {is=title K2Node_FunctionEntry}
   > Return Node!(
    execute
    normal:Linear Color Structure=@n1
    normalOutline:Linear Color Structure=LinearColor * LinearColor!(A:Linear Color Structure=@n1, B_R:float=@n1.outlineMult, B_G:float=@n1.outlineMult, B_B:float=@n1.outlineMult, B_A:double=1) -> (ReturnValue:Linear Color Structure) {is=title K2Node_PromotableOperator}
    hovered:Linear Color Structure=LinearColor * LinearColor(A=@n1, B_R=@n1.hoveredMult, B_G=@n1.hoveredMult, B_B=@n1.hoveredMult, B_A=1)
    hoveredOutline:Linear Color Structure=LinearColor * LinearColor(A=@n1, B_R=@n1.hoveredOutlineMult, B_G=@n1.hoveredOutlineMult, B_B=@n1.hoveredOutlineMult, B_A=1)
   )
  ]}
 OnDragDetected {nodes [
   OnDragDetected!() -> (then, MyGeometry:Geometry Structure, PointerEvent:"Pointer Event Structure (by ref)") {is=title K2Node_FunctionEntry}
   > PrintString(InString=Drag Detected)
   > n1 = Create Widget!(
    execute:Exec
    WorldContextObject:Object Reference
    Class:User Widget Class Reference=&/EUW_AssetPlaylist_DragDropWidget.EUW_AssetPlaylist_DragDropWidget_C
    OwningPlayer:Player Controller Object Reference
   ) -> (then:Exec, ReturnValue:EUW Asset Playlist Drag Drop Widget Object Reference)
   > n2 = "Create Drag && Drop Operation"!(
    execute:Exec
    WorldContextObject:Object Reference
    Class:Drag Drop Operation Class Reference=/Script/UMG.DragDropOperation
    Tag:String
    Payload:Object Reference
    DefaultDragVisual:Widget Object Reference=@n1
    Pivot:EDragPivot Enum=CenterCenter
    Offset:Vector 2D Structure=(X=0,Y=0)
   ) -> (then:Exec, ReturnValue:Drag Drop Operation Object Reference) {is=title K2Node_CreateDragDropOperation}
   > Return Node!(execute, Operation:Drag Drop Operation Object Reference=@n2)
  ]}
 OnMouseButtonDown {nodes [
   n1 = OnMouseButtonDown!() -> (then, MyGeometry:Geometry Structure, MouseEvent:"Pointer Event Structure (by ref)") {is=title K2Node_FunctionEntry}
   > n2 = Branch(Condition="Equal (Key)"(
     A:Key Structure=Get Effecting Button!(Input:"Pointer Event Structure (by ref)"=@n1.MouseEvent) -> (ReturnValue:Key Structure) {is=title K2Node_CallFunction}
     B:Key Structure=LeftMouseButton
    ) -> (ReturnValue:bool=~) {is=title K2Node_PromotableOperator})
   > PrintString(InString="Left  Down")
   > n3 = DetectDragIfPressed!(
    execute
    PointerEvent:"Pointer Event Structure (by ref)"=@n1.MouseEvent
    WidgetDetectingDrag:Widget Object Reference
    DragKey:Key Structure=LeftMouseButton
   ) -> (then, ReturnValue:Event Reply Structure)
   > Return Node!(execute, ReturnValue:Event Reply Structure=@n3)
   PrintString(execute=@n2.else, InString=Right Down)
   > add Floating Context Menu!(
    execute
    self:EUW Asset Playlist Headless Object Reference=Get EUW_AssetPlaylist_headless() -> (EUW_AssetPlaylist_headless:EUW Asset Playlist Headless Object Reference)
    widgetToAdd:User Widget Class Reference=&/EUW_PlaylistItemContextMenu.EUW_PlaylistItemContextMenu_C
    spawningWidget:User Widget Object Reference=Self-Reference()
   ) -> (then, floatingWidget:EUW Floating Context Menu Object Reference)
   > Return Node(ReturnValue=Handled())
  ]}
 OnMouseButtonUp {nodes [
   n1 = OnMouseButtonUp!() -> (then, MyGeometry:Geometry Structure, MouseEvent:"Pointer Event Structure (by ref)") {is=title K2Node_FunctionEntry}
   > n2 = Branch(Condition="Equal (Key)"(A=Get Effecting Button(Input=@n1.MouseEvent), B=LeftMouseButton) -> (ReturnValue=~))
   > PrintString(InString=Left Up)
   > leftUp()
   n3 = Handled()
   > Return Node(ReturnValue=@n3)
   PrintString(execute=@n2.else, InString=Right Up)
   > Return Node(ReturnValue=@n3)
  ]}
 Select Item {nodes [
   Select Item() {is=title K2Node_FunctionEntry}
   > Set selected(selected:bool=true) -> (Output_Get:bool)
   > Select UI()
   n1 = Get cosmeticType() -> (cosmeticType:CosmeticPicker_CosmeticTypes Enum=NewEnumerator0)
   > n2 = Switch on CosmeticPicker_CosmeticTypes!(execute, Selection:CosmeticPicker_CosmeticTypes Enum=@n1) -> (
    NewEnumerator0:Exec
    NewEnumerator1:Exec
    NewEnumerator2:Exec
    NewEnumerator3:Exec
    NewEnumerator4:Exec
    NewEnumerator5:Exec
    NewEnumerator6:Exec
    NewEnumerator7:Exec
    NewEnumerator8:Exec
    NewEnumerator9:Exec
    NewEnumerator10:Exec
    NewEnumerator11:Exec
    NewEnumerator12:Exec
    NewEnumerator13:Exec
    NewEnumerator14:Exec
    NewEnumerator15:Exec
   ) {is=title K2Node_SwitchEnum}
   PrintString(execute=@n2.NewEnumerator4, InString=Emote)
   n3 = Get item()
   > Execute Python Script!(
    execute
    PythonScript="import TechArt.CosmeticPicker.CosmeticPickerModule as CosmeticPicker\r\nCosmeticPicker.addEmote(item)"
    item:Fort Item Definition Object Reference=@n3
   ) -> (then, ReturnValue) {inputs [item]}
   > n4 = GetPIEWorlds!(execute, bIncludeDedicatedServer:bool=false) -> (then, ReturnValue:Array of World Object References)
   > n5 = For Each Loop(Array:Array of World Object References=@n4) -> (Array Element:World Object Reference)
   n6 = GetPlayerController!(WorldContextObject:Object Reference=@n5, PlayerIndex:int=0) -> (ReturnValue:Player Controller Object Reference)
   > Is Valid(InputObject=@n6)
   > ExecuteConsoleCommand(Command=Append!(A="cheat PlayEmoteItem ", B=GetDisplayName(Object=@n3)) -> (ReturnValue), SpecificPlayer=@n6)
   PrintString(execute=@n2.NewEnumerator5, InString=Wrap)
   > n7 = Cast To AthenaItemWrapDefinition!(execute, Object=Get item()) -> (then, CastFailed, AsAthena Item Wrap Definition:Athena Item Wrap Definition Object Reference) {is=title K2Node_DynamicCast}
   > n8 = Sequence()
   > n9 = GetAllWidgetsOfClass!(
    execute
    WorldContextObject:Object Reference
    WidgetClass:User Widget Class Reference=/Game/NeverCook/ArtTools/Tech_Art/CosmeticPicker/EUW_CosmeticPickerv2_Advanced.EUW_CosmeticPickerv2_Advanced_C
    TopLevelOnly:bool=false
   ) -> (then, FoundWidgets:Array of EUW Cosmetic Pickerv 2 Advanced Object References)
   > n10 = For Each Loop(Array:Array of EUW Cosmetic Pickerv 2 Advanced Object References=@n9) -> (Array Element:EUW Cosmetic Pickerv 2 Advanced Object Reference)
   > setWrap!(execute, self:EUW Cosmetic Pickerv 2 Advanced Object Reference=@n10, inWrap:Athena Item Wrap Definition Object Reference=@n7) -> (then)
   n11 = GetAllWidgetsOfClass(
    execute=@n8.then_1
    WidgetClass=/Game/NeverCook/ArtTools/Tech_Art/CosmeticPicker/EUW_CosmeticPickerv2.EUW_CosmeticPickerv2_C
    TopLevelOnly=false
   ) -> (FoundWidgets:Array of EUW Cosmetic Pickerv 2 Object References)
   > n12 = For Each Loop(Array:Array of EUW Cosmetic Pickerv 2 Object References=@n11) -> (Array Element:EUW Cosmetic Pickerv 2 Object Reference)
   > setAppliedWrapOnIDs!(execute, self:EUW Cosmetic Pickerv 2 Object Reference=@n12, inWrap:Athena Item Wrap Definition Object Reference=@n7) -> (then)
   PrintString(execute=@n7.CastFailed, InString=Cast to Wrap Failed)
   Execute Python Script!(
    execute=[@n2.NewEnumerator0, @n2.NewEnumerator1, @n2.NewEnumerator2, @n2.NewEnumerator3]
    PythonScript="import TechArt.CosmeticPicker.CosmeticPickerModule as CosmeticPicker\r\nCosmeticPicker.setOverride(id_name, item)"
    idName:String=Enum to String!(Enumerator:CosmeticPicker_CosmeticTypes Enum=@n1) -> (ReturnValue:String)
    item=Get item()
   ) -> (then, ReturnValue) {inputs [idName,item]}
   > n13 = Branch(Condition=Get isVariant() -> (isVariant:bool))
   n14 = "To Integer (Byte)"!(InByte:Byte=Get cosmeticType() -> (cosmeticType=NewEnumerator0)) -> (ReturnValue:int=0)
   > Execute Python Script!(
    execute
    PythonScript="import TechArt.CosmeticPicker.CosmeticPickerModule as CosmeticPicker\r\nCosmeticPicker.deleteCharacterVariantChannelsByType(type)"
    type:int=@n14
   ) -> (then, ReturnValue) {inputs [type]}
   > n15 = Execute Python Script!(
    execute
    PythonScript="import TechArt.CosmeticPicker.CosmeticPickerModule as CosmeticPicker\r\ntype = CosmeticPicker.addCharacterVariantChannel_array(item,  variant_index, option_index)"
    item:Fort Item Definition Object Reference=Get item()
    variant_index:Array of Integers=Get variantIndex() -> (variantIndex:Array of Integers)
    option_index:Array of Integers=Get optionIndex() -> (optionIndex:Array of Integers)
   ) -> (then, ReturnValue) {inputs [item, variant_index, option_index]}
   updateCosmeticPicker(execute=[@n15.then, @n16.then])
   n16 = Execute Python Script!(
    execute=@n13.else
    PythonScript="import TechArt.CosmeticPicker.CosmeticPickerModule as CosmeticPicker\r\nCosmeticPicker.deleteCharacterVariantChannelsByType(type)"
    type:int=@n14
   ) -> (then, ReturnValue) {inputs [type]}
  ]}
 Select UI {nodes [
   Select UI() {is=title K2Node_FunctionEntry}
   > Is Valid(InputObject=Get item())
   > SetColorAndOpacity(
    self:Text Object Reference=Get CommonTextBlock_version() -> (CommonTextBlock_version:Common Text Object Reference)
    InColorAndOpacity:Slate Color Structure=Make Slate Color(SpecifiedColor=(R=1,G=1,B=1,A=1))
   )
   n1 = Get selectedMultiplier() -> (selectedMultiplier:double)
   > n2 = Get Widget Style Colors!(
    execute
    self
    inputColor:Linear Color Structure=LinearColor * LinearColor(A=Get Color() -> (Color:Linear Color Structure), B_R=@n1, B_G=@n1, B_B=@n1, B_A=1)
    outlineMult:double=2
    hoveredMult:double=4
    hoveredOutlineMult:double=5
   ) -> (
    then
    normal:Linear Color Structure
    normalOutline:Linear Color Structure
    hovered:Linear Color Structure
    hoveredOutline:Linear Color Structure
   ) {is=title K2Node_CallFunction}
   > SetImageButtonStyle!(
    execute
    self
    NorrmalFillColor:Linear Color Structure=@n2
    NormalOutlineColor:Linear Color Structure=@n2.normalOutline
    HoveredFillColor:Linear Color Structure=@n2.hovered
    HoveredOutlineColor:Linear Color Structure=@n2.hoveredOutline
   ) -> (then)
  ]}
 SetImageButtonStyle {nodes [
   n1 = SetImageButtonStyle!() -> (
    then
    NorrmalFillColor:Linear Color Structure
    NormalOutlineColor:Linear Color Structure
    HoveredFillColor:Linear Color Structure
    HoveredOutlineColor:Linear Color Structure
   ) {is=title K2Node_FunctionEntry}
   > Set Normal Fill Color(Normal Fill Color:Linear Color Structure=@n1) -> (Output_Get:Linear Color Structure)
   > Set Normal Outline Color(Normal Outline Color:Linear Color Structure=@n1.NormalOutlineColor) -> (Output_Get:Linear Color Structure)
   > Set Hovered Fill Color(Hovered Fill Color:Linear Color Structure=@n1.HoveredFillColor) -> (Output_Get:Linear Color Structure)
   > Set Hovered Outline Color(Hovered Outline Color:Linear Color Structure=@n1.HoveredOutlineColor) -> (Output_Get:Linear Color Structure)
   > SetColorAndOpacity(self:Image Object Reference=Get Image_button_inner(), InColorAndOpacity:Linear Color Structure=Get Normal Fill Color())
   > SetColorAndOpacity(self=Get Image_button_outer(), InColorAndOpacity=Get Normal Outline Color())
  ]}
 clicked {nodes [
   clicked() {is=title K2Node_FunctionEntry}
   n1 = Self-Reference()
   > n2 = FindParentWidgetOfType!(
    execute
    StartingWidget:Widget Object Reference=@n1
    Type:Widget Class Reference=&/EUW_PlaylistGroup.EUW_PlaylistGroup_C
   ) -> (then, ReturnValue:EUW Playlist Group Object Reference)
   > n3 = Cast To EUW_PlaylistGroup!(execute, Object=@n2) -> (then, CastFailed, AsEUW Playlist Group:EUW Playlist Group Object Reference) {is=title K2Node_DynamicCast}
   > n4 = FindParentWidgetOfType(StartingWidget=@n1, Type=&/AssetPlaylist.AssetPlaylist_C) -> (ReturnValue:Asset Playlist Object Reference)
   > n5 = Cast To AssetPlaylist!(execute, Object=@n4) -> (then, CastFailed, AsAsset Playlist:Asset Playlist Object Reference) {is=title K2Node_DynamicCast}
   > n6 = deselect all Group Widgets!(execute, self:Asset Playlist Object Reference=@n5, exception:EUW Playlist Group Object Reference=@n3) -> (then)
   Select UI(execute=[@n6.then, @n5.CastFailed, @n3.CastFailed]) {is=title K2Node_CallFunction}
   n7 = Get item()
   n8 = Get selected() -> (selected:bool)
   n9 = Get cosmeticType() -> (cosmeticType=NewEnumerator0)
   > PrintString(InString=Append!(
     A="ButtonClick "
     B=Enum to String(Enumerator=@n9)
     C:String=" "
     D:String=GetDisplayName(Object=@n7)
     E:String=" "
     F:String="To String (Boolean)"(InBool=@n8)
    ) -> (ReturnValue))
   > n10 = Is Valid(InputObject=Get item())
   > n11 = Branch(Condition=Get selected())
   > PrintString(InString=Deselect)
   n12 = Get cosmeticType() -> (cosmeticType=NewEnumerator0)
   > deselect by CosmeticType!(execute, self, cosmeticType:CosmeticPicker_CosmeticTypes Enum=@n12) -> (then)
   > updateCosmeticPicker()
   PrintString(execute=@n11.else, InString=Select)
   > deselect by CosmeticType(cosmeticType=@n12)
   > Select Item() {is=title K2Node_CallFunction}
   > PrintString(InString="To String (Boolean)"(InBool=Get selected()))
   n13 = get find checkbox!(
    execute=@n10.Is Not Valid
    self:EUW Asset Playlist Headless Object Reference=Get EUW_AssetPlaylist_headless()
    widget:Widget Object Reference=Self-Reference()
   ) -> (then, isChecked:bool)
   > Branch(Condition=@n13)
   > open FindItem UI()
  ]}
 deselect Item {nodes [
   deselect Item() {is=title K2Node_FunctionEntry}
   > Branch(Condition=Get selected())
   > Set selected(selected=false)
   > Execute Python Script!(
    execute
    PythonScript="import TechArt.CosmeticPicker.CosmeticPickerModule as CosmeticPicker\r\nCosmeticPicker.clearOverridesByCosmeticType(cosmetic_type_index)"
    cosmeticTypeIndex:int="To Integer (Byte)"(InByte=Get cosmeticType() -> (cosmeticType=NewEnumerator0)) -> (ReturnValue=0)
   ) -> (then, ReturnValue) {inputs [cosmeticTypeIndex]}
   > deselect UI()
  ]}
 deselect UI {nodes [
   deselect UI() {is=title K2Node_FunctionEntry}
   > SetColorAndOpacity(
    self:Text Object Reference=Get CommonTextBlock_version()
    InColorAndOpacity:Slate Color Structure=Make Slate Color(SpecifiedColor=(R=0.3,G=0.3,B=0.3,A=1))
   )
   > n1 = Get Widget Style Colors(inputColor=Get Color(), outlineMult=1.3, hoveredMult=4, hoveredOutlineMult=5)
   > SetImageButtonStyle(NorrmalFillColor=@n1, NormalOutlineColor=@n1.normalOutline, HoveredFillColor=@n1.hovered, HoveredOutlineColor=@n1.hoveredOutline) {is=title K2Node_CallFunction}
  ]}
 deselect by CosmeticType {nodes [
   n1 = deselect by CosmeticType!() -> (then, cosmeticType:CosmeticPicker_CosmeticTypes Enum=NewEnumerator0) {is=title K2Node_FunctionEntry}
   > n2 = GetAllWidgetsOfClass(WidgetClass=&AssetPlaylist_C, TopLevelOnly=false) -> (FoundWidgets:Array of Asset Playlist Object References)
   > n3 = For Each Loop(Array:Array of Asset Playlist Object References=@n2) -> (Array Element:Asset Playlist Object Reference)
   > deselect all Widgets of CosmeticType!(execute, self:Asset Playlist Object Reference=@n3, cosmeticType:CosmeticPicker_CosmeticTypes Enum=@n1) -> (then)
  ]}
 get Color of CosmeticType {nodes [
   n1 = get Color of CosmeticType!() -> (then, AssetPlaylist_asset:"Asset Playlist Asset Structure (by ref)") {is=title K2Node_FunctionEntry}
   > n2 = Switch on CosmeticPicker_CosmeticTypes(Selection=Break Asset Playlist Asset!(AssetPlaylist_asset:"Asset Playlist Asset Structure (by ref)"=@n1) -> (
     shotGridId_27_#1:int
     cosmeticType_20_#2:CosmeticPicker_CosmeticTypes Enum
     name_22_#3:String
     gender_25_#4:String
     assetGroup_41_#5:String
     primaryAssetGroup_42_#6:String
     status_33_#7:String
     fixVersion_37_#8:String
    ) {is=title K2Node_BreakStruct}.cosmeticType_20_#2)
   > Return Node!(execute, color:Linear Color Structure=(R=0.0153,G=0.05,B=0.015,A=1))
   Return Node(execute=@n2.NewEnumerator1, color=(R=0.05,G=0.0325,B=0.015,A=1))
   Return Node(execute=@n2.NewEnumerator2, color=(R=0.02,G=0.02,B=0.1,A=1))
   Return Node(execute=@n2.NewEnumerator3, color=(R=0.0494,G=0.015,B=0.05,A=1))
   Return Node(execute=@n2.NewEnumerator4, color=(R=0.05,G=0.0442,B=0.015,A=1))
   Return Node(execute=@n2.NewEnumerator5, color=(R=0.0104,G=0.0267,B=0.05,A=1))
   Return Node(execute=@n2.NewEnumerator6, color=(R=0.15,G=0,B=0.0237,A=1))
  ]}
 get Icon {nodes [
   n1 = get Icon!() -> (then, object:Object Reference, cosmeticType:CosmeticPicker_CosmeticTypes Enum=NewEnumerator0) {is=title K2Node_FunctionEntry}
   > n2 = Switch on CosmeticPicker_CosmeticTypes(Selection=@n1.cosmeticType)
   > n3 = Cast To AthenaCharacterItemDefinition!(execute, Object=@n1) -> (then, CastFailed, AsAthena Character Item Definition:Athena Character Item Definition Object Reference) {is=title K2Node_DynamicCast}
   > n4 = Branch(Condition=Get isVariant())
   > n5 = get Variant Icon!(
    execute
    self
    variantIndex:int="Get (a copy)"(Array:Array of Integers=Get variantIndex()) -> (Output:int)
    optionIndex:int="Get (a copy)"(Array=Get optionIndex())
   ) -> (then, image:Object Reference)
   > n6 = Is Valid(InputObject=@n5)
   n7 = LoadAsset_Blocking(
    execute=[@n4.else, @n6.Is Not Valid]
    Asset=GetSmallPreviewImage!(self:Fort Item Definition Object Reference=@n3) -> (ReturnValue:Texture 2D Soft Object Reference)
   )
   > Return Node!(execute, previewImage:Object Reference=@n7)
   Return Node(execute=@n6.Is Valid, previewImage=@n5)
   n8 = Cast To AthenaPickaxeItemDefinition!(execute=@n2.NewEnumerator2, Object=@n1) -> (then, CastFailed, AsAthena Pickaxe Item Definition:Athena Pickaxe Item Definition Object Reference) {is=title K2Node_DynamicCast}
   > n9 = Branch(Condition=Get isVariant())
   n10 = get Variant Icon(execute=[@n15.then, @n9.then], variantIndex="Get (a copy)"(Array=Get variantIndex()), optionIndex="Get (a copy)"(Array=Get optionIndex()))
   > n11 = Is Valid(InputObject=@n10)
   n12 = GetEditorProperty!(execute:Exec=[@n11.Is Not Valid, @n15.else], Object:Object Reference=@n1, PropertyName:Name=smallPreviewImage) -> (then:Exec, PropertyValue:Soft Object Reference, ReturnValue:bool)
   > n13 = LoadAsset_Blocking(Asset=@n12)
   > Return Node(previewImage=@n13)
   Return Node(execute=@n11.Is Valid, previewImage=@n10)
   n14 = LoadAsset_Blocking(
    execute=@n9.else
    Asset=GetSmallPreviewImage(self=Get WeaponDefinition!(self:Athena Pickaxe Item Definition Object Reference=@n8) -> (WeaponDefinition:Fort Weapon Melee Item Definition Object Reference))
   )
   > Return Node(previewImage=@n14)
   n15 = Branch(execute=[@n2.NewEnumerator1, @n2.NewEnumerator3, @n2.NewEnumerator4, @n2.NewEnumerator5], Condition=Get isVariant())
  ]}
 get Override Variants {nodes [
   get Override Variants() {is=title K2Node_FunctionEntry}
   > PrintString(InString=Get Override Variants)
   > Clear(TargetArray:Array of Mcp Variant Channel Info Structures=Get Mcp Variant Channel Info() -> (Mcp Variant Channel Info:Array of Mcp Variant Channel Info Structures))
   n1 = Length(TargetArray:Array of Integers=Get variantIndex())
   n2 = Get isVariant()
   > PrintString(InString=Append!(A="len(variantndex) = ", B="To String (Integer)"(InInt=@n1), C=" ", D="To String (Boolean)"(InBool=@n2)) -> (ReturnValue))
   > n3 = Branch(Condition=@n2)
   > n4 = Cast To AthenaCosmeticItemDefinition!(execute, Object=Get item()) -> (then, CastFailed, AsAthena Cosmetic Item Definition:Athena Cosmetic Item Definition Object Reference) {is=title K2Node_DynamicCast}
   > n5 = For Each Loop(Array:Array of Integers=Get variantIndex()) -> (Array Element:int)
   Return Node!(execute=@n5.Completed, variants:Array of Mcp Variant Channel Info Structures=Get Mcp Variant Channel Info())
   n6 = Get EUW_AssetPlaylist_headless()
   > n7 = getVariantTag!(execute, self:EUW Asset Playlist Headless Object Reference=@n6, item:Athena Cosmetic Item Definition Object Reference=@n4, index:int=@n5) -> (then, tag:Gameplay Tag Structure, success:bool)
   n8 = "Get (a copy)"(Array=Get optionIndex(), Dimension 1=@n5.Array Index)
   > n9 = getActiveVariantTag!(
    execute
    self:EUW Asset Playlist Headless Object Reference=@n6
    item:Athena Cosmetic Item Definition Object Reference=@n4
    index:int=@n5
    option_index:int=@n8
   ) -> (then, tag:Gameplay Tag Structure)
   > n10 = MakeVariantChannelInfo!(
    execute
    ItemDefinition:Athena Cosmetic Item Definition Object Reference=@n4
    Channel:Gameplay Tag Structure=@n7
    ActiveVariant:Gameplay Tag Structure=@n9
    CustomData:String
   ) -> (then, ReturnValue:Mcp Variant Channel Info Structure)
   > PrintString(InString=Append!(
     A=Get UIName() -> (UIName:String)
     B=" "
     C="To String (Integer)"(InInt=@n5)
     D=","
     E:String="To String (Integer)"(InInt=@n8)
     F:String=" "
     G:String=GetDebugStringFromGameplayTag!(GameplayTag:Gameplay Tag Structure) -> (ReturnValue:String)
     H:String=","
     I:String=GetDebugStringFromGameplayTag(GameplayTag=@n9)
    ) -> (ReturnValue))
   > Add(
    TargetArray:Array of Mcp Variant Channel Info Structures=Get Mcp Variant Channel Info()
    NewItem:"Mcp Variant Channel Info Structure (by ref)"=@n10
   ) {is=title K2Node_CallArrayFunction}
   Return Node(execute=@n3.else, variants=Get Mcp Variant Channel Info())
   GetDisplayName(Object=@n4)
  ]}
 get Variant Icon {nodes [
   n1 = get Variant Icon!() -> (then, variantIndex:int, optionIndex:int) {is=title K2Node_FunctionEntry}
   > n2 = Execute Python Script!(
    execute
    PythonScript="import TechArt.CosmeticPlaylist.CosmeticPlaylistModule as CosmeticPlaylist\r\nimage = CosmeticPlaylist.getVariantIcon(item, variant_index, option_index)"
    item:Fort Item Definition Object Reference=Get item()
    variant_index:int=@n1
    option_index:int=@n1.optionIndex
   ) -> (then, ReturnValue, image:Object Reference) {inputs [item, variant_index, option_index], outputs [image]}
   > Return Node!(execute, image:Object Reference=@n2.image)
  ]}
 leftUp {nodes [
   leftUp() {is=title K2Node_FunctionEntry}
   > clicked() {is=title K2Node_CallFunction}
  ]}
 open FindItem UI {nodes [
   n1 = FindParentWidgetOfType(StartingWidget=Self-Reference(), Type=&AssetPlaylist_C)
   > n2 = Cast To AssetPlaylist(Object=@n1)
   n3 = SpawnAndRegisterTab(execute=[@n2.then, @n7.then], self=Get EditorUtilitySubsystem(), InBlueprint=&/EUW_findAsset)
   > SetDesiredSizeInViewport!(execute, self:User Widget Object Reference=@n3, Size:Vector 2D Structure=(X=800,Y=100)) -> (then)
   > n4 = Cast To EUW_findAsset!(execute, Object=@n3) -> (then, CastFailed, AsEUW Find Asset:EUW Find Asset Object Reference) {is=title K2Node_DynamicCast}
   > Set cosmeticType!(
    execute
    cosmeticType:CosmeticPicker_CosmeticTypes Enum=Get cosmeticType() -> (cosmeticType=NewEnumerator0)
    self:EUW Find Asset Object Reference=@n4
   ) -> (then, Output_Get:CosmeticPicker_CosmeticTypes Enum=NewEnumerator0)
   n5 = Make Array("[0]":int=0) -> (Array:Array of Integers)
   n6 = Get UIName()
   > setup!(
    execute
    self=Get EUW_PlaylistItem!(self:EUW Find Asset Object Reference=@n4) -> (EUW_PlaylistItem:EUW Playlist Item Object Reference)
    item:Fort Item Definition Object Reference
    asset:Asset Playlist Asset Structure=Make Asset Playlist Asset!(
     shotGridId_27_#1:int=Get Shot Grid ID() -> (Shot Grid ID:int=0)
     cosmeticType_20_#2:CosmeticPicker_CosmeticTypes Enum=Get cosmeticType() -> (cosmeticType=NewEnumerator0)
     name_22_#3:String=@n6
     gender_25_#4:String=Get gender() -> (gender:String)
     assetGroup_41_#5:String
     primaryAssetGroup_42_#6:String
     status_33_#7:String
     fixVersion_37_#8:String
    ) -> (AssetPlaylist_asset:Asset Playlist Asset Structure) {is=title K2Node_MakeStruct}
    UIName:String=@n6
    isVariant:bool=false
    variantIndex:Array of Integers=@n5
    optionIndex:Array of Integers=@n5
   ) -> (then)
   > setup!(
    execute
    self:EUW Find Asset Object Reference=@n4
    spawningAssetPlaylistItem:EUW Playlist Item Object Reference=Self-Reference()
    cosmeticType:CosmeticPicker_CosmeticTypes Enum=Get cosmeticType() -> (cosmeticType=NewEnumerator0)
    loadCheatValues:bool=false
   ) -> (then)
   n7 = open FindItem UI() {is=title K2Node_FunctionEntry}
  ]}
 open Shotgrid URL {nodes [
   open Shotgrid URL() {is=title K2Node_FunctionEntry}
   n1 = Append!(A=https://epicgames.shotgunstudio.com/detail/Asset/, B="To String (Integer)"(InInt=Get Shot Grid ID() -> (Shot Grid ID=0))) -> (ReturnValue)
   > PrintString(InString=@n1)
   > LaunchURL!(execute, URL:String=@n1) -> (then)
  ]}
 set Style {nodes [
   n1 = set Style!() -> (
    then
    widget:Button Object Reference
    normalTint:Linear Color Structure
    normalOutline:Linear Color Structure
    hoveredTint:Linear Color Structure
    hoveredOutline:Linear Color Structure
   ) {is=title K2Node_FunctionEntry}
   > SetStyle!(
    execute
    self:Button Object Reference=@n1
    InStyle_Normal_TintColor:Slate Color Structure=Make Slate Color(SpecifiedColor=@n1.normalTint)
    InStyle_Normal_DrawAs:ESlateBrushDrawType Enum=RoundedBox
    InStyle_Normal_Tiling:ESlateBrushTileType Enum=NoTile
    InStyle_Normal_Mirroring:ESlateBrushMirrorType Enum=NoMirror
    InStyle_Normal_ImageSize:Vector 2D Structure=(X=32,Y=32)
    InStyle_Normal_Margin:Margin Structure=(Left=0,Top=0,Right=0,Bottom=0)
    InStyle_Normal_ResourceObject:Object Reference
    InStyle_Normal_OutlineSettings:Slate Brush Outline Settings Structure=Make Slate Brush Outline Settings!(
     CornerRadii:Vector 4 Structure=(X=0,Y=0,Z=0,W=0)
     Color:Slate Color Structure=Make Slate Color(SpecifiedColor=@n1.normalOutline)
     Width:float=5
     RoundingType:ESlateBrushRoundingType Enum=HalfHeightRadius
     bUseBrushTransparency:bool=False
    ) -> (SlateBrushOutlineSettings:Slate Brush Outline Settings Structure) {is=title K2Node_MakeStruct}
    InStyle_Hovered_TintColor:Slate Color Structure=Make Slate Color(SpecifiedColor=@n1.hoveredTint)
    InStyle_Hovered_DrawAs:ESlateBrushDrawType Enum=RoundedBox
    InStyle_Hovered_Tiling:ESlateBrushTileType Enum=NoTile
    InStyle_Hovered_Mirroring:ESlateBrushMirrorType Enum=NoMirror
    InStyle_Hovered_ImageSize:Vector 2D Structure=(X=32,Y=32)
    InStyle_Hovered_Margin:Margin Structure=(Left=0,Top=0,Right=0,Bottom=0)
    InStyle_Hovered_ResourceObject:Object Reference
    InStyle_Hovered_OutlineSettings:Slate Brush Outline Settings Structure=Make Slate Brush Outline Settings(
     CornerRadii=(X=0,Y=0,Z=0,W=0)
     Color=Make Slate Color(SpecifiedColor=@n1.hoveredOutline)
     Width=5
     RoundingType=HalfHeightRadius
     bUseBrushTransparency=False
    )
    InStyle_Pressed:Slate Brush Structure=(TintColor=(SpecifiedColor=(R=1,G=1,B=1,A=1),ColorUseRule=UseColor_Specified),DrawAs=Image,Tiling=NoTile,Mirroring=NoMirror,ImageType=NoImage,ImageSize=(X=32,Y=32),Margin=(Left=0,Top=0,Right=0,Bottom=0),ResourceObject=None,OutlineSettings=(CornerRadii=(X=0,Y=0,Z=0,W=0),Color=(SpecifiedColor=(R=0,G=0,B=0,A=0),ColorUseRule=UseColor_Specified),Width=0,RoundingType=HalfHeightRadius,bUseBrushTransparency=False),UVRegion=(Min=(X=0,Y=0),Max=(X=0,Y=0),bIsValid=False),bIsDynamicallyLoaded=False,ResourceName="")
    InStyle_Disabled:Slate Brush Structure=(TintColor=(SpecifiedColor=(R=1,G=1,B=1,A=1),ColorUseRule=UseColor_Specified),DrawAs=NoDrawType,Tiling=NoTile,Mirroring=NoMirror,ImageType=NoImage,ImageSize=(X=0,Y=0),Margin=(Left=0,Top=0,Right=0,Bottom=0),ResourceObject=None,OutlineSettings=(CornerRadii=(X=0,Y=0,Z=0,W=0),Color=(SpecifiedColor=(R=0,G=0,B=0,A=0),ColorUseRule=UseColor_Specified),Width=0,RoundingType=HalfHeightRadius,bUseBrushTransparency=False),UVRegion=(Min=(X=0,Y=0),Max=(X=0,Y=0),bIsValid=False),bIsDynamicallyLoaded=False,ResourceName="")
    InStyle_NormalForeground:Slate Color Structure=(SpecifiedColor=(R=1,G=0,B=1,A=1),ColorUseRule=UseColor_Foreground)
    InStyle_HoveredForeground:Slate Color Structure=(SpecifiedColor=(R=1,G=0,B=1,A=1),ColorUseRule=UseColor_Foreground)
    InStyle_PressedForeground:Slate Color Structure=(SpecifiedColor=(R=1,G=0,B=1,A=1),ColorUseRule=UseColor_Foreground)
    InStyle_DisabledForeground:Slate Color Structure=(SpecifiedColor=(R=1,G=0,B=1,A=1),ColorUseRule=UseColor_Foreground)
    InStyle_NormalPadding:Margin Structure=(Left=0,Top=0,Right=0,Bottom=0)
    InStyle_PressedPadding:Margin Structure=(Left=0,Top=0,Right=0,Bottom=0)
    InStyle_PressedSlateSound:Slate Sound Structure=(ResourceObject=None)
    InStyle_ClickedSlateSound:Slate Sound Structure=(ResourceObject=None)
    InStyle_HoveredSlateSound:Slate Sound Structure=(ResourceObject=None)
   ) -> (then)
  ]}
 set Thumbnail {nodes [
   n1 = set Thumbnail!() -> (then, thumbnailVisibility:bool=~) {is=title K2Node_FunctionEntry}
   > PrintString(InString=Append!(A=" ", B=Get UIName(), C:String=" ", D:String="To String (Boolean)"(InBool=@n1)) -> (ReturnValue))
   > n2 = Branch(Condition=@n1)
   n3 = Get item()
   > n4 = Is Valid(InputObject=@n3)
   n5 = Set previewImage(execute=@n4.Is Not Valid, previewImage:Object Reference=&/T_assetPlaylist_questionmark) -> (Output_Get:Object Reference)
   n6 = get Icon!(
    execute=@n4.Is Valid
    self
    object:Object Reference=@n3
    cosmeticType:CosmeticPicker_CosmeticTypes Enum=Get cosmeticType() -> (cosmeticType=NewEnumerator0)
   ) -> (then, previewImage:Object Reference) {is=title K2Node_CallFunction}
   > n7 = Set previewImage(previewImage=@n6)
   n8 = Is Valid(exec=[@n7.then, @n5.then], InputObject=@n7)
   n9 = Get Image_thumbnail() -> (Image_thumbnail:Image Object Reference)
   n10 = SetBrush!(
    execute=@n8.Is Not Valid
    self=@n9
    InBrush_TintColor:Slate Color Structure=Make Slate Color(SpecifiedColor=MakeColor!(R:float=0.01, G:float=0.01, B:float=0.01, A:float=1) -> (ReturnValue:Linear Color Structure))
    InBrush_DrawAs:ESlateBrushDrawType Enum=RoundedBox
    InBrush_Tiling:ESlateBrushTileType Enum=NoTile
    InBrush_Mirroring:ESlateBrushMirrorType Enum=NoMirror
    InBrush_ImageSize:Vector 2D Structure=(X=50,Y=50)
    InBrush_Margin:Margin Structure=(Left=0,Top=0,Right=0,Bottom=0)
    InBrush_ResourceObject:Object Reference
    InBrush_OutlineSettings:Slate Brush Outline Settings Structure=(CornerRadii=(X=0,Y=0,Z=0,W=0),Color=(SpecifiedColor=(R=0,G=0,B=0,A=0),ColorUseRule=UseColor_Specified),Width=0,RoundingType=HalfHeightRadius,bUseBrushTransparency=False)
   ) -> (then)
   n11 = Get Overlay_thumbnail() -> (Overlay_thumbnail:Overlay Object Reference)
   SetVisibility(execute=[@n10.then, @n13.then], self=@n11, InVisibility=Visible)
   > InvalidateLayoutAndVolatility(self:Widget Object Reference=@n11)
   > SetVisibility(self=Get SizeBox_thumbnail() -> (SizeBox_thumbnail:Size Box Object Reference), InVisibility=Visible)
   n12 = Get SizeBox_top() -> (SizeBox_top:Size Box Object Reference)
   > SetWidthOverride(self=@n12, InWidthOverride=200)
   > SetHeightOverride(self=@n12, InHeightOverride=50)
   > "SetText (Text)"(
    self=Get CommonTextBlock_name() -> (CommonTextBlock_name:Common Text Object Reference)
    InText="To Text (String)"(InString=Get longName() -> (longName:String))
   )
   PrintString(execute=@n8.Is Valid, InString=Append!(A=" thumbnail : ", B=GetDisplayName(Object=@n7)) -> (ReturnValue))
   > n13 = SetBrush(
    self=@n9
    InBrush_TintColor=(SpecifiedColor=(R=1,G=1,B=1,A=1),ColorUseRule=UseColor_Specified)
    InBrush_DrawAs=RoundedBox
    InBrush_Tiling=NoTile
    InBrush_Mirroring=NoMirror
    InBrush_ImageSize=(X=50,Y=50)
    InBrush_Margin=(Left=0,Top=0,Right=0,Bottom=0)
    InBrush_ResourceObject=Get previewImage() -> (previewImage:Object Reference)
    InBrush_OutlineSettings=(CornerRadii=(X=0,Y=0,Z=0,W=0),Color=(SpecifiedColor=(R=0,G=0,B=0,A=0),ColorUseRule=UseColor_Specified),Width=0,RoundingType=HalfHeightRadius,bUseBrushTransparency=False)
   )
   SetVisibility(execute=@n2.else, self=Get SizeBox_thumbnail(), InVisibility=Collapsed)
   n14 = Get SizeBox_top()
   > SetWidthOverride(self=@n14, InWidthOverride=150)
   > SetHeightOverride(self=@n14, InHeightOverride=30)
   > "SetText (Text)"(self=Get CommonTextBlock_name(), InText="To Text (String)"(InString=Get shortName() -> (shortName:String)))
  ]}
 setup {nodes [
   n1 = setup!() -> (
    then
    item:Fort Item Definition Object Reference
    asset:Asset Playlist Asset Structure
    UIName:String
    isVariant:bool=~
    variantIndex:Array of Integers
    optionIndex:Array of Integers
   ) {is=title K2Node_FunctionEntry}
   > Set asset(asset:Asset Playlist Asset Structure=@n1.asset) -> (Output_Get:Asset Playlist Asset Structure)
   n2 = Break Asset Playlist Asset(AssetPlaylist_asset=@n1.asset)
   > Set Shot Grid ID(Shot Grid ID:int=@n2) -> (Output_Get:int=0)
   > Set cosmeticType!(execute, cosmeticType=@n2.cosmeticType_20_#2) -> (then, Output_Get=NewEnumerator0)
   > Set UIName(UIName:String=@n1.UIName) -> (Output_Get:String)
   > Set originalName(originalName:String=@n2.name_22_#3) -> (Output_Get:String)
   > Set gender(gender:String=@n2.gender_25_#4) -> (Output_Get:String)
   > Set isVariant(isVariant:bool=@n1.isVariant) -> (Output_Get:bool)
   > Set variantIndex(variantIndex:Array of Integers=@n1.variantIndex) -> (Output_Get:Array of Integers)
   > Set optionIndex(optionIndex:Array of Integers=@n1.optionIndex) -> (Output_Get:Array of Integers)
   > n3 = Set item(item:Fort Item Definition Object Reference=@n1) -> (Output_Get:Fort Item Definition Object Reference)
   > n4 = For Each Loop(Array=Get variantIndex())
   n5 = For Each Loop(Exec=@n4.Completed, Array=Get optionIndex())
   > Set optionIndexString(optionIndexString:String=Append!(A=Get optionIndexString() -> (optionIndexString:String), B="To String (Integer)"(InInt=@n5), C:String=" ") -> (ReturnValue)) -> (Output_Get:String)
   n6 = Is Valid(exec=@n5.Completed, InputObject=@n1)
   > n7 = Cast To AthenaCharacterItemDefinition(Object=@n3)
   > n8 = Set Character(Character:Athena Character Item Definition Object Reference=@n7) -> (Output_Get:Athena Character Item Definition Object Reference)
   > setup Widget Look!(
    execute
    self
    asset:Asset Playlist Asset Structure=Get asset() -> (asset:Asset Playlist Asset Structure)
    FortItemDefinition:Fort Item Definition Object Reference=@n8
    UIname:String=Get UIName()
    color:Linear Color Structure=(R=0.0153,G=0.05,B=0.015,A=1)
   ) -> (then)
   n9 = Cast To AthenaBackpackItemDefinition!(execute=@n7.CastFailed, Object=@n3) -> (then, CastFailed, AsAthena Backpack Item Definition:Athena Backpack Item Definition Object Reference) {is=title K2Node_DynamicCast}
   > n10 = Set Backpack(Backpack:Athena Backpack Item Definition Object Reference=@n9) -> (Output_Get:Athena Backpack Item Definition Object Reference)
   > setup Widget Look(asset=Get asset(), FortItemDefinition=@n10, UIname=Get UIName(), color=(R=0.05,G=0.0325,B=0.015,A=1))
   n11 = Cast To AthenaDanceItemDefinition!(execute=@n9.CastFailed, Object=@n3) -> (then, CastFailed, AsAthena Dance Item Definition:Athena Dance Item Definition Object Reference) {is=title K2Node_DynamicCast}
   > n12 = Set Emote(Emote:Athena Dance Item Definition Object Reference=@n11) -> (Output_Get:Athena Dance Item Definition Object Reference)
   > setup Widget Look(asset=Get asset(), FortItemDefinition=@n12, UIname=Get UIName(), color=(R=0.05,G=0.0442,B=0.015,A=1))
   n13 = Cast To AthenaPickaxeItemDefinition(execute=@n11.CastFailed, Object=@n3)
   > n14 = Set Pickaxe(Pickaxe:Athena Pickaxe Item Definition Object Reference=@n13) -> (Output_Get:Athena Pickaxe Item Definition Object Reference)
   > setup Widget Look(asset=Get asset(), FortItemDefinition=@n14, UIname=Get UIName(), color=(R=0.02,G=0.02,B=0.1,A=1))
   n15 = Cast To AthenaGliderItemDefinition!(execute=@n13.CastFailed, Object=@n3) -> (then, CastFailed, AsAthena Glider Item Definition:Athena Glider Item Definition Object Reference) {is=title K2Node_DynamicCast}
   > n16 = Set Glider(Glider:Athena Glider Item Definition Object Reference=@n15) -> (Output_Get:Athena Glider Item Definition Object Reference)
   > setup Widget Look(asset=Get asset(), FortItemDefinition=@n16, UIname=Get UIName(), color=(R=0.0494,G=0.015,B=0.05,A=1))
   n17 = Cast To AthenaItemWrapDefinition(execute=@n15.CastFailed, Object=@n3)
   > n18 = Set Wrap(Wrap:Athena Item Wrap Definition Object Reference=@n17) -> (Output_Get:Athena Item Wrap Definition Object Reference)
   > setup Widget Look(asset=Get asset(), FortItemDefinition=@n18, UIname=Get UIName(), color=(R=0.0104,G=0.0267,B=0.05,A=1))
   n19 = Get asset()
   n20 = get Color of CosmeticType!(execute=@n6.Is Not Valid, self, AssetPlaylist_asset:"Asset Playlist Asset Structure (by ref)"=@n19) -> (then, color:Linear Color Structure) {is=title K2Node_CallFunction}
   > setup Widget Look(asset=@n19, UIname=Get UIName(), color=@n20)
   Set variantIndexString(
    execute=@n4.LoopBody
    variantIndexString:String=Append(A=Get variantIndexString() -> (variantIndexString:String), B="To String (Integer)"(InInt=@n4), C=" ")
   ) -> (Output_Get:String)
  ]}
 setup Widget Look {nodes [
   n1 = setup Widget Look!() -> (
    then
    asset:Asset Playlist Asset Structure
    FortItemDefinition:Fort Item Definition Object Reference
    UIname:String
    color:Linear Color Structure
   ) {is=title K2Node_FunctionEntry}
   > Set Color(Color:Linear Color Structure=@n1.color) -> (Output_Get:Linear Color Structure)
   n2 = Get maxNameLength() -> (maxNameLength:int=0)
   > n3 = Branch(Condition="integer > integer"(A=Len(S=@n1.UIname), B=@n2))
   n4 = int - int(A=@n2, B=2)
   > n5 = Set text(text:String=Append!(A=GetSubstring!(SourceString:String=@n1.UIname, StartIndex:int=0, Length:int=@n4) -> (ReturnValue:String), B=..) -> (ReturnValue)) -> (Output_Get:String)
   n6 = Get text() -> (text:String)
   Set shortName(execute=[@n5.then, @n19.then], shortName:String=@n6) -> (Output_Get:String)
   n7 = Break Asset Playlist Asset(AssetPlaylist_asset=@n1)
   n8 = Append!(A=Enum to String(Enumerator=@n7.cosmeticType_20_#2), B="\r\n", C:String=@n6) -> (ReturnValue)
   > Set longName(longName:String=@n8) -> (Output_Get:String)
   > Set text(text=@n8)
   > "SetText (Text)"(self=Get CommonTextBlock_name(), InText="To Text (String)"(InString=Get text()))
   n9 = Get isVariant()
   n10 = Enum to String(Enumerator=@n7.cosmeticType_20_#2)
   n11 = "To Text (String)"(InString=Append!(
     A=@n1.UIname
     B="\r\n"
     C=@n7.name_22_#3
     D:String="\r\n"
     E:String=Get Object Path String(Object=@n1.FortItemDefinition)
     F:String="\r\nShotGridId: "
     G:String="To String (Integer)"(InInt=@n7)
     H:String="\r\nvariant:"
     I:String="To String (Boolean)"(InBool=@n9)
     J:String="\r\nvariantIndex:"
     K:String=Get variantIndexString()
     L:String="\r\noptionIndex:"
     M:String=Get optionIndexString()
     N:String="\r\ngender:"
     O:String=@n7.gender_25_#4
     P:String="\r\nstatus:"
     Q:String=@n7.status_33_#7
     R:String=Append!(A="version: : ", B=@n7.fixVersion_37_#8, C="\r\n") -> (ReturnValue)
     S:String=Append(A="cosmeticType: : ", B=@n10, C="\r\n")
     T:String=Append(A="assetGroup : ", B=@n7.assetGroup_41_#5, C="\r\n")
     U:String=Append!(A="assetGroups : ", B=@n7.primaryAssetGroup_42_#6) -> (ReturnValue)
    ) -> (ReturnValue))
   > SetToolTipText!(execute, self:Widget Object Reference=Get Image_button_inner(), InToolTipText:"Text (by ref)"=@n11) -> (then)
   > SetToolTipText(self=Get Image_button_outer(), InToolTipText=@n11)
   > "SetText (Text)"(self=Get CommonTextBlock_version(), InText="To Text (String)"(InString=@n7.fixVersion_37_#8))
   > n12 = Is Valid(InputObject=@n1.FortItemDefinition)
   > SetColorAndOpacity(self=Get CommonTextBlock_name(), InColorAndOpacity=Make Slate Color(SpecifiedColor=(R=0.9,G=0.9,B=0.9,A=1)))
   > n13 = Get Widget Style Colors(inputColor=Get Color(), outlineMult=1.3, hoveredMult=4, hoveredOutlineMult=5)
   > n14 = SetImageButtonStyle(NorrmalFillColor=@n13, NormalOutlineColor=@n13.normalOutline, HoveredFillColor=@n13.hovered, HoveredOutlineColor=@n13.hoveredOutline)
   n15 = Branch(execute=[@n18.then, @n14.then], Condition=Get isVariant())
   > n16 = SetVisibility(self=Get Button_isVariant() -> (Button_isVariant:Button Object Reference), InVisibility=Visible)
   set Thumbnail!(execute=[@n16.then, @n15.else], self, thumbnailVisibility:bool=true) -> (then) {is=title K2Node_CallFunction}
   SetColorAndOpacity(execute=@n12.Is Not Valid, self=Get CommonTextBlock_name(), InColorAndOpacity=Make Slate Color(SpecifiedColor=(R=0.2,G=0.2,B=0.2,A=1)))
   > n17 = Get Widget Style Colors(
    inputColor=LinearColor * LinearColor!(A=Get Color(), B:Linear Color Structure=(R=0.25,G=0.25,B=0.25,A=1)) -> (ReturnValue)
    outlineMult=1.3
    hoveredMult=4
    hoveredOutlineMult=5
   )
   > n18 = SetImageButtonStyle(NorrmalFillColor=@n17, NormalOutlineColor=@n17.normalOutline, HoveredFillColor=@n17.hovered, HoveredOutlineColor=@n17.hoveredOutline)
   n19 = Set text(execute=@n3.else, text=@n1.UIname)
  ]}
 "setup Widget as Added " {nodes [
   "setup Widget as Added "() {is=title K2Node_FunctionEntry}
   > SetVisibility(self=Get CommonTextBlock_added() -> (CommonTextBlock_added:Common Text Object Reference), InVisibility=Visible)
  ]}
 spawnItem {nodes [
   n1 = Cast To AthenaGliderItemDefinition(Object=Get item())
   > n2 = get Override Variants!(execute, self) -> (then, variants:Array of Mcp Variant Channel Info Structures) {is=title K2Node_CallFunction}
   > SpawnAthenaGlider!(
    execute
    ItemDefinition:Athena Glider Item Definition Object Reference=@n1
    SpawnTransform_Location:Vector=Get pos() -> (pos:Vector)
    SpawnTransform_Rotation:Rotator="0, 0, 0"
    SpawnTransform_Scale:Vector="1,1,1"
    VariantChannelInfos:Array of Mcp Variant Channel Info Structures=@n2
    ClearAnimations:bool=true
   ) -> (then, ReturnValue:Array of Fort Player Parachute Object References)
   n3 = spawnItem!() -> (then, position:Vector=~) {is=title K2Node_FunctionEntry}
   > PrintString(InString=Spawn Item)
   > Set pos(pos:Vector=@n3) -> (Output_Get:Vector)
   > n4 = Branch(Condition=Get isVariant())
   > n5 = get Override Variants()
   > n6 = Set variants(variants:Array of Mcp Variant Channel Info Structures=@n5) -> (Output_Get:Array of Mcp Variant Channel Info Structures)
   Is Valid(exec=[@n18.then, @n6.then], InputObject=Get item())
   > n7 = Switch on CosmeticPicker_CosmeticTypes(Selection=Get cosmeticType() -> (cosmeticType=NewEnumerator0))
   > n8 = SpawnActor Fort Player Mannequin!(
    execute:Exec
    WorldContextObject:Object Reference
    Class:Actor Class Reference=/Script/FortniteGame.FortPlayerMannequin
    SpawnTransform_Location:Vector=Get pos()
    SpawnTransform_Rotation:Rotator="0, 0, 0"
    SpawnTransform_Scale:Vector="1,1,1"
    CollisionHandlingOverride:ESpawnActorCollisionHandlingMethod Enum=Undefined
    TransformScaleMethod:ESpawnActorScaleMethod Enum=MultiplyWithRoot
    Owner:Actor Object Reference
    Instigator:Pawn Object Reference
   ) -> (then:Exec, ReturnValue:Fort Player Mannequin Object Reference) {is=title K2Node_SpawnActorFromClass}
   > Set mannequin(mannequin:Fort Player Mannequin Object Reference=@n8) -> (Output_Get:Fort Player Mannequin Object Reference)
   > n9 = SetEditorProperty(
    Object=Get mannequin() -> (mannequin:Fort Player Mannequin Object Reference)
    PropertyName=athenaCharacter
    PropertyValue:Fort Item Definition Object Reference=Get item()
   )
   SetEditorProperty(
    execute=[@n17.then, @n9.then, @n14.then]
    Object=Get mannequin()
    PropertyName=overrideVariants
    PropertyValue:Array of Mcp Variant Channel Info Structures=Get variants() -> (variants:Array of Mcp Variant Channel Info Structures)
   )
   n10 = Cast To AthenaBackpackItemDefinition(execute=@n7.NewEnumerator1, Object=Get item())
   > n11 = get Override Variants()
   > n12 = SpawnActor Fort Player Mannequin(
    Class=&FortPlayerMannequin
    SpawnTransform_Location=vector + vector(A=Get pos(), B="0,0,75")
    SpawnTransform_Rotation="0,180, 0"
    SpawnTransform_Scale="1,1,1"
    CollisionHandlingOverride=Undefined
    TransformScaleMethod=MultiplyWithRoot
   )
   > n13 = Set mannequin(mannequin=@n12)
   > Set AthenaBackBling!(execute, AthenaBackBling:Athena Backpack Item Definition Object Reference=@n10, self:Fort Player Mannequin Object Reference=@n13) -> (then, Output_Get:Athena Backpack Item Definition Object Reference)
   > n14 = SetPart!(
    execute
    self:Fort Player Mannequin Object Reference=@n13
    Part:EFortCustomPartType Enum=Backpack
    PartAsset:Custom Character Part Object Reference="Get (a copy)"(Array:Array of Custom Character Part Object References=GetCharacterParts!(self:Athena Character Part Item Definition Object Reference=@n10) -> (ReturnValue:Array of Custom Character Part Object References)) -> (Output:Custom Character Part Object Reference)
   ) -> (then, ReturnValue:bool)
   n15 = Cast To AthenaPickaxeItemDefinition(execute=@n7.NewEnumerator2, Object=Get item())
   > n16 = SpawnActor Fort Player Mannequin(
    Class=&FortPlayerMannequin
    SpawnTransform_Location=vector + vector(A=Get pos(), B="0,0,75")
    SpawnTransform_Rotation="0, 0, 0"
    SpawnTransform_Scale="1,1,1"
    CollisionHandlingOverride=Undefined
    TransformScaleMethod=MultiplyWithRoot
   )
   > Set mannequin(mannequin=@n16)
   > Set AthenaBackBling(self=Get mannequin())
   > n17 = SetWeaponSkeletalMesh!(
    execute
    self:Fort Player Mannequin Object Reference=Get mannequin()
    SkeletalMesh:Skeletal Mesh Object Reference=GetWeaponMeshOverride!(self:Fort Weapon Item Definition Object Reference=Get WeaponDefinition(self=@n15)) -> (ReturnValue:Skeletal Mesh Object Reference)
   ) -> (then, ReturnValue:bool)
   n18 = Set variants(execute=@n4.else, variants=Make Array("[0]":Mcp Variant Channel Info Structure) -> (Array:Array of Mcp Variant Channel Info Structures))
   SpawnAthenaBackpackPawn!(
    execute
    ItemDefinition:Athena Backpack Item Definition Object Reference=@n10
    SpawnTransform_Location:Vector=vector + vector(A=Get pos(), B="0,0,50")
    SpawnTransform_Rotation:Rotator="0, 0, 0"
    SpawnTransform_Scale:Vector="1,1,1"
    VariantChannelInfos:Array of Mcp Variant Channel Info Structures=@n11
    ClearAnimations:bool=true
   ) -> (then, ReturnValue:Array of Actor Object References)
  ]}
 updateCosmeticPicker {nodes [
   updateCosmeticPicker() {is=title K2Node_FunctionEntry}
   > n1 = GetAllWidgetsOfClass(WidgetClass=&EUW_CosmeticPickerv2_C, TopLevelOnly=false) -> (FoundWidgets:Array of EUW Cosmetic Pickerv 2 Object References)
   > PrintString(InString=Updating CosmeticPicker ---------------------------------------------------)
   > n2 = For Each Loop(Array:Array of EUW Cosmetic Pickerv 2 Object References=@n1) -> (Array Element:EUW Cosmetic Pickerv 2 Object Reference)
   > PrintString(InString=Append!(A="To String (Integer)"(InInt=@n2.Array Index), B=" ", C:String=GetDisplayName(Object=@n2)) -> (ReturnValue))
   > n3 = Branch(Condition=Get selected())
   > Playlist Update!(
    execute
    self:EUW Cosmetic Pickerv 2 Object Reference=@n2
    cosmeticType:CosmeticPicker_CosmeticTypes Enum=Get cosmeticType() -> (cosmeticType=NewEnumerator0)
    item:Fort Item Definition Object Reference=Get item()
   ) -> (then)
   Playlist Update(execute=@n3.else, self=@n2, cosmeticType=Get cosmeticType() -> (cosmeticType=NewEnumerator0))
  ]}
]
variables [
 asset {
  type=AssetPlaylist_asset
  default {name="", assetGroup="", cosmeticType=Character, fixVersion="", gender="", primaryAssetGroup="", shotGridId=0, status=""}
 }
 Backpack {type=AthenaBackpackItemDefinition, default=null}
 Character {type=AthenaCharacterItemDefinition, default=null}
 Color {type=LinearColor, default=<rgba 0 0 0 0>}
 cosmeticType {type=CosmeticPicker_CosmeticTypes, default=Character}
 Emote {type=AthenaDanceItemDefinition, default=null}
 gender {type=string, default=""}
 Glider {type=AthenaGliderItemDefinition, default=null}
 Hovered Fill Color {type=LinearColor, default=<rgba 0 0 0 0>}
 Hovered Outline Color {type=LinearColor, default=<rgba 0 0 0 0>}
 isVariant {type=boolean}
 item {type=FortItemDefinition, default=null}
 longName {type=string, default=""}
 maxNameLength {type=integer, default=23}
 Normal Fill Color {type=LinearColor, default=<rgba 0 0 0 0>}
 Normal Outline Color {type=LinearColor, default=<rgba 0 0 0 0>}
 optionIndex {type="array<integer>"}
 optionIndexString {type=string, default=""}
 originalName {type=string, default=""}
 Pickaxe {type=AthenaPickaxeItemDefinition, default=null}
 selected {type=boolean}
 shortName {type=string, default=""}
 Shot Grid ID {type=integer}
 UIName {type=string, default=""}
 variantIndex {type="array<integer>"}
 variantIndexString {type=string, default=""}
 Wrap {type=AthenaItemWrapDefinition, default=null}
]
widgets {
 info {rootWidgetClass=&CanvasPanel, widgetCount=18}
 widgets [
  CanvasPanel_0 {widgetClassPath=&CanvasPanel, widget=CanvasPanel_0}
  SizeBox_top {
   widgetClassPath=&SizeBox
   bIsVariable=true
   parent=CanvasPanel_0
   widget=SizeBox_top
   properties {bOverride_HeightOverride=true, bOverride_WidthOverride=true, heightOverride=50, minDesiredWidth=150, widthOverride=180}
   slot {
    class=CanvasPanelSlot
    properties {bAutoSize=true, layoutData {alignment=<0 0>, anchors {maximum=<0 0>, minimum=<0 0>}, offsets=<ltrb 0 0 10 50>}}
   }
  }
  Overlay_684 {widgetClassPath=&Overlay, parent=SizeBox_top, widget=Overlay_684, slot {class=SizeBoxSlot}}
  Image_button_outer {
   widgetClassPath=&Image
   bIsVariable=true
   parent=Overlay_684
   widget=Image_button_outer
   properties {
    brush {
     bIsDynamicallyLoaded=false
     drawAs=RoundedBox
     imageType=NoImage
     mirroring=NoMirror
     resourceName=None
     resourceObject=null
     tiling=NoTile
     imageSize=<32 32>
     margin=<ltrb 0 0 0 0>
     outlineSettings {
      bUseBrushTransparency=false
      roundingType=HalfHeightRadius
      width=0
      color {colorUseRule=UseColor_Specified, specifiedColor=<rgba 0 0 0 0>}
      cornerRadii=<0 0 0 0>
     }
     tintColor {colorUseRule=UseColor_Specified, specifiedColor=<rgba 1 1 1 1>}
     uVRegion {bIsValid=false, max=<0 0>, min=<0 0>}
    }
    colorAndOpacity=<rgba 0.05 0.05 0.05 1>
   }
   slot {class=OverlaySlot, properties {horizontalAlignment=HAlign_Fill, verticalAlignment=VAlign_Fill}}
  }
  Image_button_inner {
   widgetClassPath=&Image
   bIsVariable=true
   parent=Overlay_684
   widget=Image_button_inner
   properties {
    brush {
     bIsDynamicallyLoaded=false
     drawAs=RoundedBox
     imageType=NoImage
     mirroring=NoMirror
     resourceName=None
     resourceObject=null
     tiling=NoTile
     imageSize=<32 32>
     margin=<ltrb 0 0 0 0>
     outlineSettings {
      bUseBrushTransparency=false
      roundingType=HalfHeightRadius
      width=0
      color {colorUseRule=UseColor_Specified, specifiedColor=<rgba 0 0 0 0>}
      cornerRadii=<0 0 0 0>
     }
     tintColor {colorUseRule=UseColor_Specified, specifiedColor=<rgba 1 1 1 1>}
     uVRegion {bIsValid=false, max=<0 0>, min=<0 0>}
    }
    colorAndOpacity=<rgba 0.03 0.03 0.03 1>
   }
   slot {class=OverlaySlot, properties {horizontalAlignment=HAlign_Fill, verticalAlignment=VAlign_Fill, padding=<ltrb 5 5 5 5>}}
  }
  HorizontalBox_344 {
   widgetClassPath=&HorizontalBox
   parent=Overlay_684
   widget=HorizontalBox_344
   slot {class=OverlaySlot, properties {horizontalAlignment=HAlign_Fill, verticalAlignment=VAlign_Fill}}
  }
  SizeBox_382 {
   widgetClassPath=&SizeBox
   parent=HorizontalBox_344
   widget=SizeBox_382
   properties {bOverride_WidthOverride=true, widthOverride=20}
   slot {class=HorizontalBoxSlot}
  }
  Button_isVariant {
   widgetClassPath=&Button
   parent=SizeBox_382
   widget=Button_isVariant
   properties {
    bIsVolatile=true
    visibility=Collapsed
    widgetStyle {
     clickedSlateSound {resourceObject=null}
     disabled {
      bIsDynamicallyLoaded=false
      drawAs=RoundedBox
      imageType=NoImage
      mirroring=NoMirror
      resourceName=None
      resourceObject=null
      tiling=NoTile
      imageSize=<0 0>
      margin=<ltrb 0 0 0 0>
      outlineSettings {
       bUseBrushTransparency=false
       roundingType=FixedRadius
       width=1
       color {colorUseRule=UseColor_ColorTable, specifiedColor=<rgba 0 0 0 0>}
       cornerRadii=<4 4 4 4>
      }
      tintColor {colorUseRule=UseColor_Specified, specifiedColor=<rgba 0.0395 0.0395 0.0395 1>}
      uVRegion {bIsValid=false, max=<0 0>, min=<0 0>}
     }
     disabledForeground {colorUseRule=UseColor_ColorTable, specifiedColor=<rgba 0 0 0 0>}
     hovered {
      bIsDynamicallyLoaded=false
      drawAs=RoundedBox
      imageType=NoImage
      mirroring=NoMirror
      resourceName=None
      resourceObject=null
      tiling=NoTile
      imageSize=<0 0>
      margin=<ltrb 0 0 0 0>
      outlineSettings {
       bUseBrushTransparency=false
       roundingType=FixedRadius
       width=1
       color {colorUseRule=UseColor_ColorTable, specifiedColor=<rgba 0 0 0 0>}
       cornerRadii=<4 4 4 4>
      }
      tintColor {colorUseRule=UseColor_Specified, specifiedColor=<rgba 0.0953 0.0953 0.0953 1>}
      uVRegion {bIsValid=false, max=<0 0>, min=<0 0>}
     }
     hoveredForeground {colorUseRule=UseColor_ColorTable, specifiedColor=<rgba 0 0 0 0>}
     hoveredSlateSound {resourceObject=null}
     normal {
      bIsDynamicallyLoaded=false
      drawAs=RoundedBox
      imageType=NoImage
      mirroring=NoMirror
      resourceName=None
      resourceObject=null
      tiling=NoTile
      imageSize=<0 0>
      margin=<ltrb 0 0 0 0>
      outlineSettings {
       bUseBrushTransparency=false
       roundingType=FixedRadius
       width=1
       color {colorUseRule=UseColor_ColorTable, specifiedColor=<rgba 0 0 0 0>}
       cornerRadii=<4 4 4 4>
      }
      tintColor {colorUseRule=UseColor_Specified, specifiedColor=<rgba 0.01 0.01 0.01 0.8>}
      uVRegion {bIsValid=false, max=<0 0>, min=<0 0>}
     }
     normalForeground {colorUseRule=UseColor_ColorTable, specifiedColor=<rgba 0 0 0 0>}
     normalPadding=<ltrb 12 1.5 12 1.5>
     pressed {
      bIsDynamicallyLoaded=false
      drawAs=RoundedBox
      imageType=NoImage
      mirroring=NoMirror
      resourceName=None
      resourceObject=null
      tiling=NoTile
      imageSize=<0 0>
      margin=<ltrb 0 0 0 0>
      outlineSettings {
       bUseBrushTransparency=false
       roundingType=FixedRadius
       width=1
       color {colorUseRule=UseColor_ColorTable, specifiedColor=<rgba 0 0 0 0>}
       cornerRadii=<4 4 4 4>
      }
      tintColor {colorUseRule=UseColor_Specified, specifiedColor=<rgba 0.0284 0.0284 0.0284 1>}
      uVRegion {bIsValid=false, max=<0 0>, min=<0 0>}
     }
     pressedForeground {colorUseRule=UseColor_ColorTable, specifiedColor=<rgba 0 0 0 0>}
     pressedPadding=<ltrb 12 2 12 0>
     pressedSlateSound {resourceObject=null}
    }
   }
   slot {class=SizeBoxSlot}
  }
  CommonTextBlock {
   widgetClassPath=&CommonTextBlock
   bIsVariable=true
   parent=Button_isVariant
   widget=CommonTextBlock
   properties {
    justification=Center
    text=V
    font {
     bForceMonospaced=false
     bMaterialIsStencil=false
     fontMaterial=null
     fontObject=&NotoSans
     letterSpacing=0
     monospacedWidth=1
     size=10
     skewAmount=0
     typefaceFontName=NotoSans-Regular
     outlineSettings {
      bApplyOutlineToDropShadows=false
      bMiteredCorners=false
      bSeparateFillAlpha=false
      outlineMaterial=null
      outlineSize=0
      outlineColor=<rgba 0 0 0 1>
     }
    }
    shadowOffset=<0 0>
   }
   slot {class=ButtonSlot}
  }
  SizeBox_thumbnail {
   widgetClassPath=&SizeBox
   bIsVariable=true
   parent=HorizontalBox_344
   widget=SizeBox_thumbnail
   properties {bOverride_HeightOverride=true, bOverride_WidthOverride=true, heightOverride=50, widthOverride=50}
   slot {class=HorizontalBoxSlot}
  }
  Overlay_thumbnail {
   widgetClassPath=&Overlay
   bIsVariable=true
   parent=SizeBox_thumbnail
   widget=Overlay_thumbnail
   properties {bIsVolatile=true, visibility=Collapsed}
   slot {class=SizeBoxSlot}
  }
  Image_thumbnail_bg {
   widgetClassPath=&Image
   bIsVariable=true
   parent=Overlay_thumbnail
   widget=Image_thumbnail_bg
   properties {brush {
     bIsDynamicallyLoaded=false
     drawAs=RoundedBox
     imageType=NoImage
     mirroring=NoMirror
     resourceName=None
     resourceObject=null
     tiling=NoTile
     imageSize=<44 44>
     margin=<ltrb 0 0 0 0>
     outlineSettings {
      bUseBrushTransparency=false
      roundingType=HalfHeightRadius
      width=0
      color {colorUseRule=UseColor_Specified, specifiedColor=<rgba 0 0 0 0>}
      cornerRadii=<0 0 0 0>
     }
     tintColor {colorUseRule=UseColor_Specified, specifiedColor=<rgba 0.01 0.01 0.01 1>}
     uVRegion {bIsValid=false, max=<0 0>, min=<0 0>}
    }}
   slot {class=OverlaySlot, properties {horizontalAlignment=HAlign_Center, verticalAlignment=VAlign_Center, padding=<ltrb 5 5 5 5>}}
  }
  Image_thumbnail {
   widgetClassPath=&Image
   bIsVariable=true
   parent=Overlay_thumbnail
   widget=Image_thumbnail
   properties {brush {
     bIsDynamicallyLoaded=false
     drawAs=Box
     imageType=NoImage
     mirroring=NoMirror
     resourceName=None
     resourceObject=null
     tiling=NoTile
     imageSize=<50 50>
     margin=<ltrb 0 0 0 0>
     outlineSettings {
      bUseBrushTransparency=false
      roundingType=HalfHeightRadius
      width=0
      color {colorUseRule=UseColor_Specified, specifiedColor=<rgba 0 0 0 0>}
      cornerRadii=<0 0 0 0>
     }
     tintColor {colorUseRule=UseColor_Specified, specifiedColor=<rgba 1 1 1 1>}
     uVRegion {bIsValid=false, max=<0 0>, min=<0 0>}
    }}
   slot {class=OverlaySlot}
  }
  Overlay_1013 {
   widgetClassPath=&Overlay
   parent=HorizontalBox_344
   widget=Overlay_1013
   slot {class=HorizontalBoxSlot, properties {size {sizeRule=Fill, value=1}}}
  }
  CommonTextBlock_name {
   widgetClassPath=&CommonTextBlock
   bIsVariable=true
   parent=Overlay_1013
   widget=CommonTextBlock_name
   properties {
    justification=Center
    text=???
    font {
     bForceMonospaced=false
     bMaterialIsStencil=false
     fontMaterial=null
     fontObject=&NotoSans
     letterSpacing=0
     monospacedWidth=1
     size=8
     skewAmount=0
     typefaceFontName=NotoSans-Regular
     outlineSettings {
      bApplyOutlineToDropShadows=false
      bMiteredCorners=false
      bSeparateFillAlpha=false
      outlineMaterial=null
      outlineSize=0
      outlineColor=<rgba 0 0 0 1>
     }
    }
    shadowColorAndOpacity=<rgba 0 0 0 1>
    shadowOffset=<1.5 1.5>
   }
   slot {class=OverlaySlot, properties {horizontalAlignment=HAlign_Fill, verticalAlignment=VAlign_Center, padding=<ltrb -4 2 4 2>}}
  }
  CommonTextBlock_added {
   widgetClassPath=&CommonTextBlock
   bIsVariable=true
   parent=Overlay_1013
   widget=CommonTextBlock_added
   properties {
    justification=Center
    text=+
    visibility=Collapsed
    colorAndOpacity {colorUseRule=UseColor_Specified, specifiedColor=<rgba 0.5 0.5 0.5 1>}
    font {
     bForceMonospaced=false
     bMaterialIsStencil=false
     fontMaterial=null
     fontObject=&NotoSans
     letterSpacing=0
     monospacedWidth=1
     size=15
     skewAmount=0
     typefaceFontName=NotoSans-Bold
     outlineSettings {
      bApplyOutlineToDropShadows=false
      bMiteredCorners=false
      bSeparateFillAlpha=false
      outlineMaterial=null
      outlineSize=0
      outlineColor=<rgba 0 0 0 1>
     }
    }
    shadowOffset=<0 0>
   }
   slot {class=OverlaySlot, properties {horizontalAlignment=HAlign_Right, padding=<ltrb 0 -6 8 0>}}
  }
  CommonTextBlock_version {
   widgetClassPath=&CommonTextBlock
   bIsVariable=true
   parent=Overlay_1013
   widget=CommonTextBlock_version
   properties {
    justification=Center
    text=00.00
    visibility=Visible
    colorAndOpacity {colorUseRule=UseColor_Specified, specifiedColor=<rgba 0.4 0.4 0.4 1>}
    font {
     bForceMonospaced=false
     bMaterialIsStencil=false
     fontMaterial=null
     fontObject=&NotoSans
     letterSpacing=0
     monospacedWidth=1
     size=6
     skewAmount=0
     typefaceFontName=NotoSans-Regular
     outlineSettings {
      bApplyOutlineToDropShadows=false
      bMiteredCorners=false
      bSeparateFillAlpha=false
      outlineMaterial=null
      outlineSize=0
      outlineColor=<rgba 0 0 0 1>
     }
    }
    shadowOffset=<0 0>
   }
   slot {class=OverlaySlot, properties {horizontalAlignment=HAlign_Right, verticalAlignment=VAlign_Bottom, padding=<ltrb 0 0 8 1>}}
  }
  EUW_AssetPlaylist_headless {
   widgetClassPath=&/EUW_AssetPlaylist_headless
   bIsVariable=true
   parent=CanvasPanel_0
   widget=EUW_AssetPlaylist_headless
   slot {class=CanvasPanelSlot, properties=~}
  }
 ]
}
