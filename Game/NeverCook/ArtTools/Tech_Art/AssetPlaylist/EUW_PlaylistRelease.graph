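# Editor utility widget for one release in the Asset Playlist tool, holding its playlist groups under a release button.
# setReleaseText pulls lock dates from an embedded headless playlist widget into a day-count tooltip, then colors the button by release scope.
# Groups fill a vertical or horizontal box; resize sizes the column from visible horizontal groups, collapsing it when none show.
# Clicking reveals all groups and prints their items; another function re-sorts groups alphabetically.
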EditorUtilityWidgetBlueprint /Game/NeverCook/ArtTools/Tech_Art/AssetPlaylist/EUW_PlaylistRelease
graphs [
 EventGraph {nodes [
   "On Clicked (Button_Group)"() {is=title K2Node_ComponentBoundEvent}
   > PrintString(InString=Get release() -> (release:String))
   > n1 = getGroups!(execute, self) -> (then, groups:Array of EUW Playlist Group Object References)
   > n2 = For Each Loop(Array:Array of EUW Playlist Group Object References=@n1) -> (Array Element:EUW Playlist Group Object Reference)
   n3 = GetVisibility!(self:Widget Object Reference=@n2) -> (ReturnValue:ESlateVisibility Enum=Visible)
   > PrintString(InString=Append!(
     A=Get groupName!(self:EUW Playlist Group Object Reference=@n2) -> (groupName:String)
     B=" "
     C:String=Enum to String!(Enumerator:ESlateVisibility Enum=@n3) -> (ReturnValue:String)
    ) -> (ReturnValue))
   > SetVisibility(self=@n2, InVisibility=Visible)
   > n4 = get Items!(execute, self:EUW Playlist Group Object Reference=@n2) -> (then, result:Array of EUW Playlist Item Object References)
   n5 = Length(TargetArray:Array of EUW Playlist Item Object References=@n4)
   > PrintString(InString=Append(A="    ", B="To String (Integer)"(InInt=@n5), C=" items"))
   > n6 = For Each Loop(Array:Array of EUW Playlist Item Object References=@n4) -> (Array Element:EUW Playlist Item Object Reference)
   > PrintString(InString=Append!(A="  ", B=Get UIName!(self:EUW Playlist Item Object Reference=@n6) -> (UIName:String)) -> (ReturnValue))
   Event PreConstruct()
   Event Construct()
   Event Tick()
  ]}
 Reorder Group Alphabetically {nodes [
   Reorder Group Alphabetically() {is=title K2Node_FunctionEntry}
   > PrintString(InString=Reorder Alphabetically)
   > n1 = getGroups()
   > n2 = Set Groups(Groups:Array of EUW Playlist Group Object References=@n1) -> (Output_Get:Array of EUW Playlist Group Object References)
   > ClearChildren(self=Get HorizontalBox_Groups() -> (HorizontalBox_Groups:Horizontal Box Object Reference))
   > n3 = reorderGroupsArray!(
    execute
    self:EUW Asset Playlist Headless Object Reference=Get EUW_AssetPlaylist_headless() -> (EUW_AssetPlaylist_headless:EUW Asset Playlist Headless Object Reference)
    Array:Array of EUW Playlist Group Object References=@n2
   ) -> (then, OutputArray:Array of EUW Playlist Group Object References)
   > n4 = For Each Loop(Array:Array of EUW Playlist Group Object References=@n3) -> (Array Element:EUW Playlist Group Object Reference)
   > AddChild(self=Get HorizontalBox_Groups(), Content=@n4)
  ]}
 addGroup {nodes [
   n1 = addGroup!() -> (then, group:EUW Playlist Group Object Reference) {is=title K2Node_FunctionEntry}
   > n2 = Branch(Condition=Get vertical() -> (vertical:bool))
   > AddChild(self=Get VerticalBox_Groups() -> (VerticalBox_Groups:Vertical Box Object Reference), Content=@n1)
   AddChild(execute=@n2.else, self=Get HorizontalBox_Groups(), Content=@n1)
   > resize!(execute, self, thumbnails:bool=true) -> (then)
  ]}
 dateInfo {nodes [
   n1 = dateInfo!() -> (then, MDYstring:String) {is=title K2Node_FunctionEntry}
   n2 = ParseIntoArray(SourceString=@n1, Delimiter=/)
   n3 = "Get (a copy)"(Array=@n2, Dimension 1=2)
   n4 = Get Date!(A:Date Time Structure=Now!() -> (ReturnValue:Date Time Structure)) -> (
    ReturnValue_Year:int=0
    ReturnValue_Month:int=0
    ReturnValue_Day:int=0
    ReturnValue_Hour:int=0
    ReturnValue_Minute:int=0
    ReturnValue_Second:int=0
    ReturnValue_Millisecond:int=0
   ) {is=title K2Node_CallFunction}
   > n5 = DaysBetweenDates!(
    execute
    self:EUW Asset Playlist Headless Object Reference=Get EUW_AssetPlaylist_headless()
    YearA:int=@n4
    MonthA:int=@n4.ReturnValue_Month
    DayA:int=@n4.ReturnValue_Day
    YearB:int=String To Integer(InString=Append(A=20, B=@n3))
    MonthB:int=String To Integer(InString="Get (a copy)"(Array=@n2))
    DayB:int=String To Integer(InString="Get (a copy)"(Array=@n2, Dimension 1=1))
   ) -> (then, days:int=0)
   > Return Node!(execute, ReturnValue:String=Append!(A=@n1, B=" in ", C:String="To String (Integer)"(InInt=@n5), D:String=" Days") -> (ReturnValue))
  ]}
 getGroups {nodes [
   getGroups() {is=title K2Node_FunctionEntry}
   > n1 = Branch(Condition=Get vertical())
   > n2 = Set children(children:Array of Widget Object References=GetAllChildren(self=Get VerticalBox_Groups())) -> (Output_Get:Array of Widget Object References)
   n3 = For Each Loop(Exec=[@n5.then, @n2.then], Array:Array of Widget Object References=Get children() -> (children:Array of Widget Object References)) -> (Array Element:Widget Object Reference)
   Return Node!(
    execute=@n3.Completed
    groups:Array of EUW Playlist Group Object References=Get localGroups() -> (localGroups:Array of EUW Playlist Group Object References)
   )
   > n4 = Cast To EUW_PlaylistGroup!(execute, Object=@n3) -> (then, CastFailed, AsEUW Playlist Group:EUW Playlist Group Object Reference) {is=title K2Node_DynamicCast}
   > Add(TargetArray:Array of EUW Playlist Group Object References=Get localGroups(), NewItem:"EUW Playlist Group Object Reference (by ref)"=@n4) {is=title K2Node_CallArrayFunction}
   n5 = Set children(execute=@n1.else, children=GetAllChildren(self=Get HorizontalBox_Groups()))
  ]}
 inScope {nodes [
   n1 = inScope!() -> (then, release:String) {is=title K2Node_FunctionEntry}
   > n2 = get Release!(execute, self:EUW Asset Playlist Headless Object Reference=Get EUW_AssetPlaylist_headless()) -> (then, release:String)
   > n3 = Branch(Condition="Equal Exactly (String)"(A=@n2, B=Main))
   > Return Node!(execute, inScope:bool=true)
   Return Node(execute=@n3.else, inScope="float <= float"(A=String To Float(InString=@n1), B=String To Float(InString=@n2)) -> (ReturnValue=~))
  ]}
 resize {nodes [
   n1 = resize!() -> (then, thumbnails:bool=true) {is=title K2Node_FunctionEntry}
   > n2 = Branch(Condition=@n1)
   > n3 = Set width(width:int=200) -> (Output_Get:int=0)
   Set count(execute=[@n9.then, @n3.then], count:int=0) -> (Output_Get:int=0)
   > n4 = For Each Loop(Array=GetAllChildren(self=Get HorizontalBox_Groups()))
   n5 = Get count() -> (count:int=0)
   n6 = Get width() -> (width:int=0)
   SetWidthOverride(
    execute=@n4.Completed
    self=Get SizeBox_top() -> (SizeBox_top:Size Box Object Reference)
    InWidthOverride="To Float (Integer)"(InInt=int * int(A=@n5, B=@n6))
   )
   > n7 = Branch(Condition="Equal (Integer)"(A=Get count() -> (count=0)))
   > SetVisibility(InVisibility=Collapsed)
   SetVisibility(execute=@n7.else, InVisibility=Visible)
   n8 = Cast To EUW_PlaylistGroup(execute=@n4.LoopBody, Object=@n4)
   > Branch(Condition=IsVisible(self=@n8))
   > Increment Int(Value=Get count() -> (count=0))
   n9 = Set width(execute=@n2.else, width=150) -> (Output_Get=0)
   Get VerticalBox_Groups()
  ]}
 setReleaseText {nodes [
   n1 = setReleaseText!() -> (then, release:String) {is=title K2Node_FunctionEntry}
   > n2 = Set release(release:String=@n1) -> (Output_Get:String)
   > "SetText (Text)"(self=Get CommonTextBlock_release() -> (CommonTextBlock_release:Common Text Object Reference), InText="To Text (String)"(InString=@n2))
   > n3 = Get Lock Dates!(execute, self:EUW Asset Playlist Headless Object Reference=Get EUW_AssetPlaylist_headless(), release:String=@n2) -> (
    then
    releaseStructure_branchFrom_47_#1:String
    releaseStructure_BranchCreation_62_#2:String
    releaseStructure_BranchOpen_60_#3:String
    releaseStructure_ArtLock_49_#4:String
    releaseStructure_HardLock_51_#5:String
    releaseStructure_PencilsDown_58_#6:String
    releaseStructure_Cert_53_#7:String
    releaseStructure_Live_55_#8:String
   ) {is=title K2Node_CallFunction}
   > n4 = dateInfo!(execute, self, MDYstring:String=@n3.releaseStructure_BranchCreation_62_#2) -> (then, ReturnValue:String) {is=title K2Node_CallFunction}
   > n5 = dateInfo(MDYstring=@n3.releaseStructure_BranchOpen_60_#3)
   > n6 = dateInfo(MDYstring=@n3.releaseStructure_ArtLock_49_#4)
   > n7 = dateInfo(MDYstring=@n3.releaseStructure_HardLock_51_#5)
   > n8 = dateInfo(MDYstring=@n3.releaseStructure_PencilsDown_58_#6)
   > SetToolTipText!(
    execute
    self:Widget Object Reference
    InToolTipText:"Text (by ref)"="To Text (String)"(InString=Append!(
      A=@n2
      B="\r\nBranchFrom         "
      C=@n3
      D="\r\nBranch Creation  "
      E:String=@n4
      F:String="\r\nBranchOpen         "
      G:String=@n5
      H:String="\r\nArtLock                 "
      I:String=@n6
      J:String="\r\nHardLock              "
      K:String=@n7
      L:String="\r\nPencilsDown        "
      M:String=@n8
      N:String="\r\nCert                            "
      O:String=@n3.releaseStructure_Cert_53_#7
      P:String="\r\nLive                           "
      Q:String=@n3.releaseStructure_Live_55_#8
     ) -> (ReturnValue))
   ) -> (then)
   > n9 = inScope!(execute, self, release:String=Get release()) -> (then, inScope:bool) {is=title K2Node_CallFunction}
   > n10 = Branch(Condition=@n9)
   > SetBackgroundColor!(
    execute
    self:Button Object Reference=Get Button_Group() -> (Button_Group:Button Object Reference)
    InBackgroundColor:Linear Color Structure=(R=0.0153,G=0.05,B=0.015,A=1)
   ) -> (then)
   SetBackgroundColor(execute=@n10.else, self=Get Button_Group(), InBackgroundColor=(R=0.15,G=0,B=0.0237,A=1))
  ]}
]
variables [
 release {type=string, default=""}
 vertical {type=boolean}
]
widgets {
 info {rootWidgetClass=&CanvasPanel, widgetCount=10}
 widgets [
  CanvasPanel_0 {widgetClassPath=&CanvasPanel, widget=CanvasPanel_0}
  SizeBox_top {
   widgetClassPath=&SizeBox
   bIsVariable=true
   parent=CanvasPanel_0
   widget=SizeBox_top
   properties {bOverride_WidthOverride=true, widthOverride=200}
   slot {
    class=CanvasPanelSlot
    properties {bAutoSize=true, layoutData {alignment=<0 0>, anchors {maximum=<0 1>, minimum=<0 0>}, offsets=<ltrb 0 0 180 0>}}
   }
  }
  VerticalBox {widgetClassPath=&VerticalBox, parent=SizeBox_top, widget=VerticalBox, slot {class=SizeBoxSlot}}
  SizeBox_167 {
   widgetClassPath=&SizeBox
   parent=VerticalBox
   widget=SizeBox_167
   properties {bOverride_HeightOverride=true, heightOverride=20, widthOverride=80}
   slot {class=VerticalBoxSlot}
  }
  Button_Group {
   widgetClassPath=&Button
   parent=SizeBox_167
   widget=Button_Group
   properties {
    backgroundColor=<rgba 0.1 0.1 0.1 1>
    widgetStyle {
     clickedSlateSound {resourceObject=null}
     disabled {
      bIsDynamicallyLoaded=false
      drawAs=RoundedBox
      imageType=NoImage
      mirroring=NoMirror
      resourceName=None
      resourceObject=null
      tiling=NoTile
      imageSize=<0 0>
      margin=<ltrb 0 0 0 0>
      outlineSettings {
       bUseBrushTransparency=false
       roundingType=FixedRadius
       width=1
       color {colorUseRule=UseColor_ColorTable, specifiedColor=<rgba 0 0 0 0>}
       cornerRadii=<4 4 4 4>
      }
      tintColor {colorUseRule=UseColor_Specified, specifiedColor=<rgba 0.0395 0.0395 0.0395 1>}
      uVRegion {bIsValid=false, max=<0 0>, min=<0 0>}
     }
     disabledForeground {colorUseRule=UseColor_ColorTable, specifiedColor=<rgba 0 0 0 0>}
     hovered {
      bIsDynamicallyLoaded=false
      drawAs=RoundedBox
      imageType=NoImage
      mirroring=NoMirror
      resourceName=None
      resourceObject=null
      tiling=NoTile
      imageSize=<0 0>
      margin=<ltrb 0 0 0 0>
      outlineSettings {
       bUseBrushTransparency=false
       roundingType=FixedRadius
       width=1
       color {colorUseRule=UseColor_ColorTable, specifiedColor=<rgba 0 0 0 0>}
       cornerRadii=<4 4 4 4>
      }
      tintColor {colorUseRule=UseColor_Specified, specifiedColor=<rgba 2 2 2 1>}
      uVRegion {bIsValid=false, max=<0 0>, min=<0 0>}
     }
     hoveredForeground {colorUseRule=UseColor_ColorTable, specifiedColor=<rgba 0 0 0 0>}
     hoveredSlateSound {resourceObject=null}
     normal {
      bIsDynamicallyLoaded=false
      drawAs=RoundedBox
      imageType=NoImage
      mirroring=NoMirror
      resourceName=None
      resourceObject=null
      tiling=NoTile
      imageSize=<0 0>
      margin=<ltrb 0 0 0 0>
      outlineSettings {
       bUseBrushTransparency=false
       roundingType=FixedRadius
       width=1
       color {colorUseRule=UseColor_ColorTable, specifiedColor=<rgba 0 0 0 0>}
       cornerRadii=<4 4 4 4>
      }
      tintColor {colorUseRule=UseColor_Specified, specifiedColor=<rgba 1 1 1 1>}
      uVRegion {bIsValid=false, max=<0 0>, min=<0 0>}
     }
     normalForeground {colorUseRule=UseColor_ColorTable, specifiedColor=<rgba 0 0 0 0>}
     normalPadding=<ltrb 12 1.5 12 1.5>
     pressed {
      bIsDynamicallyLoaded=false
      drawAs=RoundedBox
      imageType=NoImage
      mirroring=NoMirror
      resourceName=None
      resourceObject=null
      tiling=NoTile
      imageSize=<0 0>
      margin=<ltrb 0 0 0 0>
      outlineSettings {
       bUseBrushTransparency=false
       roundingType=FixedRadius
       width=1
       color {colorUseRule=UseColor_ColorTable, specifiedColor=<rgba 0 0 0 0>}
       cornerRadii=<4 4 4 4>
      }
      tintColor {colorUseRule=UseColor_Specified, specifiedColor=<rgba 0.0284 0.0284 0.0284 1>}
      uVRegion {bIsValid=false, max=<0 0>, min=<0 0>}
     }
     pressedForeground {colorUseRule=UseColor_ColorTable, specifiedColor=<rgba 0 0 0 0>}
     pressedPadding=<ltrb 12 2 12 0>
     pressedSlateSound {resourceObject=null}
    }
   }
   slot {class=SizeBoxSlot}
  }
  CommonTextBlock_release {
   widgetClassPath=&CommonTextBlock
   bIsVariable=true
   parent=Button_Group
   widget=CommonTextBlock_release
   properties {
    justification=Center
    text=Release
    font {
     bForceMonospaced=false
     bMaterialIsStencil=false
     fontMaterial=null
     fontObject=&NotoSans
     letterSpacing=0
     monospacedWidth=1
     size=8
     skewAmount=0
     typefaceFontName=NotoSans-Regular
     outlineSettings {
      bApplyOutlineToDropShadows=false
      bMiteredCorners=false
      bSeparateFillAlpha=false
      outlineMaterial=null
      outlineSize=0
      outlineColor=<rgba 0 0 0 1>
     }
    }
    shadowOffset=<0 0>
   }
   slot {class=ButtonSlot}
  }
  Overlay_197 {
   widgetClassPath=&Overlay
   parent=VerticalBox
   widget=Overlay_197
   slot {class=VerticalBoxSlot, properties {size {sizeRule=Fill, value=1}}}
  }
  HorizontalBox_Groups {
   widgetClassPath=&HorizontalBox
   bIsVariable=true
   parent=Overlay_197
   widget=HorizontalBox_Groups
   slot {class=OverlaySlot, properties {horizontalAlignment=HAlign_Fill, verticalAlignment=VAlign_Fill}}
  }
  VerticalBox_Groups {
   widgetClassPath=&VerticalBox
   bIsVariable=true
   parent=Overlay_197
   widget=VerticalBox_Groups
   slot {class=OverlaySlot, properties {horizontalAlignment=HAlign_Fill, verticalAlignment=VAlign_Fill}}
  }
  EUW_AssetPlaylist_headless {
   widgetClassPath=&/EUW_AssetPlaylist_headless
   bIsVariable=true
   parent=CanvasPanel_0
   widget=EUW_AssetPlaylist_headless
   slot {class=CanvasPanelSlot, properties=~}
  }
 ]
}
