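# AssetPlaylist is an editor utility widget that loads cosmetic assets from ShotGrid by release, playlist or user.
# A headless helper runs queries and matches item definitions, filling tabbed columns with release, group and item widgets.
# Group names derive from asset group, gender and cosmetic name, and an optional cosmetic picker stack mirrors the layout.
# It also sorts and filters entries, steps through item selection, and spawns assets at an editor-camera trace hit.
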EditorUtilityWidgetBlueprint /Game/NeverCook/ArtTools/Tech_Art/AssetPlaylist/AssetPlaylist
graphs [
 EventGraph {nodes [
   n1 = get Groups!(execute, self) -> (then, out:Array of EUW Playlist Group Object References)
   n2 = Branch(execute=[@n1.then, @n5.else], Condition="integer > integer"(A=Length(TargetArray:Array of EUW Playlist Group Object References=@n1)))
   > n3 = get Items!(
    execute
    self:EUW Playlist Group Object Reference="Get (a copy)"(Array:Array of EUW Playlist Group Object References=@n1) -> (Output:EUW Playlist Group Object Reference)
   ) -> (then, result:Array of EUW Playlist Item Object References)
   > n4 = Remove Index!(execute, TargetArray:Array of EUW Playlist Group Object References=@n1, IndexToRemove:int=0) -> (then)
   n5 = Branch(execute=[@n4.then, @n6.then], Condition="integer > integer"(A=Length(TargetArray:Array of EUW Playlist Item Object References=@n3)))
   > Select Item(self:EUW Playlist Item Object Reference="Get (a copy)"(Array:Array of EUW Playlist Item Object References=@n3) -> (Output:EUW Playlist Item Object Reference))
   > Remove Index(TargetArray:Array of EUW Playlist Item Object References=@n3, IndexToRemove=0)
   > n6 = Delay!(execute, WorldContextObject:Object Reference, Duration=0.01) -> (then)
   deselect all Widgets(execute=@n2.else)
   "On Clicked (Button_reload)"() {is=title K2Node_ComponentBoundEvent}
   > safe Reload()
   n7 = "On Tab Selected (Generic01_TabContainer)"!() -> (OutputDelegate, then, tabIndex:int=0, tabWidget:Widget Object Reference) {is=title K2Node_ComponentBoundEvent}
   > PrintString(InString="To String (Integer)"(InInt=@n7.tabIndex))
   > Set Tab Widget(Tab Widget:Widget Object Reference=@n7.tabWidget) -> (Output_Get:Widget Object Reference)
   > n8 = Switch on Int!(execute:Exec, Selection:int=@n7.tabIndex) -> (Default:Exec, 0:Exec, 1:Exec, 2:Exec, 3:Exec)
   Set HorizontalBox_current(
    execute=@n8.0
    HorizontalBox_current:Horizontal Box Object Reference=Get HorizontalBox_releaseColumns() -> (HorizontalBox_releaseColumns:Horizontal Box Object Reference)
   ) -> (Output_Get:Horizontal Box Object Reference)
   > SetVisibility(self=Get HorizontalBox_releaseColumns(), InVisibility=Visible)
   > SetVisibility(self=Get HorizontalBox_playlistColumns() -> (HorizontalBox_playlistColumns:Horizontal Box Object Reference), InVisibility=Collapsed)
   > SetVisibility(self=Get HorizontalBox_userColumns() -> (HorizontalBox_userColumns:Horizontal Box Object Reference), InVisibility=Collapsed)
   Set HorizontalBox_current(execute=@n8.1, HorizontalBox_current=Get HorizontalBox_playlistColumns())
   > SetVisibility(self=Get HorizontalBox_releaseColumns(), InVisibility=Collapsed)
   > SetVisibility(self=Get HorizontalBox_playlistColumns(), InVisibility=Visible)
   > SetVisibility(self=Get HorizontalBox_userColumns(), InVisibility=Collapsed)
   Set HorizontalBox_current(execute=@n8.2, HorizontalBox_current=Get HorizontalBox_userColumns())
   > SetVisibility(self=Get HorizontalBox_releaseColumns(), InVisibility=Collapsed)
   > SetVisibility(self=Get HorizontalBox_playlistColumns(), InVisibility=Collapsed)
   > SetVisibility(self=Get HorizontalBox_userColumns(), InVisibility=Visible)
   PrintString(execute=@n8.3, InString=Options)
   n9 = "On Check State Changed (CheckBox_cosmeticPicker)"!() -> (OutputDelegate, then, bIsChecked:bool) {is=title K2Node_ComponentBoundEvent}
   > n10 = Branch(Condition=@n9.bIsChecked)
   n11 = getAssetPlaylistStack!(execute=@n10.else, self) -> (then, assetPlaylist_stack:EUW Asset Playlist Vertical Group List Object Reference)
   > clearWidgets(self:EUW Asset Playlist Vertical Group List Object Reference=@n11)
   n12 = "On Tabs Initialized (Generic01_TabContainer)"!() -> (OutputDelegate, then, TabContainer:Base Tab Container Widget Object Reference) {is=title K2Node_ComponentBoundEvent}
   > n13 = For Each Loop(Array:Array of TCW Tab Data Struct Structures=getCurrentTabData!(self:Base Tab Container Widget Object Reference=@n12.TabContainer) -> (currentTabData:Array of TCW Tab Data Struct Structures)) -> (Array Element:TCW Tab Data Struct Structure)
   > n14 = Cast To TA_BaseEditorButtonWidget!(
    execute
    Object=Break TCW Tab Data Struct!(TCW_TabData_Struct:"TCW Tab Data Struct Structure (by ref)"=@n13) -> (
     tabName_2_#1:String
     tabIndex_5_#2:int
     tabWidgetClass_11_#3:User Widget Class Reference
     tabContentWidget_14_#4:User Widget Object Reference
     tabButtonWidget_17_#5:User Widget Object Reference
     tabButtonStyle_20_#6:TCW Button Visuals Struct Structure
    ) {is=title K2Node_BreakStruct}.tabContentWidget_14_#4
   ) -> (then, CastFailed, AsTA Base Editor Button Widget:TA Base Editor Button Widget Object Reference) {is=title K2Node_DynamicCast}
   > Branch(Condition="Equal (Integer)"(A=@n13.Array Index, B=3))
   > Set!(
    execute
    iconStyle_TintColor:Slate Color Structure=(SpecifiedColor=(R=1,G=1,B=1,A=1),ColorUseRule=UseColor_Specified)
    iconStyle_DrawAs:ESlateBrushDrawType Enum=Image
    iconStyle_Tiling:ESlateBrushTileType Enum=NoTile
    iconStyle_Mirroring:ESlateBrushMirrorType Enum=NoMirror
    iconStyle_ImageSize:Vector 2D Structure=(X=32,Y=32)
    iconStyle_Margin:Margin Structure=(Left=0,Top=0,Right=0,Bottom=0)
    iconStyle_ResourceObject:Object Reference=&/T_assetplaylist_options
    iconStyle_OutlineSettings:Slate Brush Outline Settings Structure=(CornerRadii=(X=0,Y=0,Z=0,W=0),Color=(SpecifiedColor=(R=0,G=0,B=0,A=0),ColorUseRule=UseColor_Specified),Width=0,RoundingType=HalfHeightRadius,bUseBrushTransparency=False)
    self:TA Base Editor Button Widget Object Reference=@n14
   ) -> (then, Output_Get:Slate Brush Structure)
   > validateIconVisuals!(execute, self:TA Base Editor Button Widget Object Reference=@n14) -> (then, handled:bool)
   "On Clicked (Button_spawn)"() {is=title K2Node_ComponentBoundEvent}
   > n15 = Get Editor Camera Info!(execute) -> (then, CameraLocation:Vector, ForwardDirection:Vector, FOV:float, ReturnValue:bool) {is=title K2Node_CallFunction}
   n16 = vector + vector(A=@n15, B=vector * vector(A=@n15.ForwardDirection, B:Vector="10000,10000,10000") -> (ReturnValue=~))
   > n17 = Line Trace For Objects!(
    execute
    WorldContextObject:Object Reference
    Start:Vector=@n15
    End:Vector=@n16
    ObjectTypes:Array of EObjectTypeQuery Enums=Make Array("[0]":EObjectTypeQuery Enum=ObjectTypeQuery1) -> (Array:Array of EObjectTypeQuery Enums)
    bTraceComplex:bool=true
    ActorsToIgnore:Array of Actor Object References
    DrawDebugType:EDrawDebugTrace Enum
    bIgnoreSelf:bool=true
    TraceColor:Linear Color Structure=(R=1,G=0,B=0,A=1)
    TraceHitColor:Linear Color Structure=(R=0,G=1,B=0,A=1)
    DrawTime:float=5
   ) -> (
    then
    OutHit_bBlockingHit:bool
    OutHit_bInitialOverlap:bool
    OutHit_Time:float
    OutHit_Distance:float
    OutHit_Location:Vector
    OutHit_ImpactPoint:Vector
    OutHit_Normal:Vector
    OutHit_ImpactNormal:Vector
    OutHit_PhysMat:Physical Material Object Reference
    OutHit_HitActor:Actor Object Reference
    OutHit_HitComponent:Primitive Component Object Reference
    OutHit_HitBoneName:Name=None
    OutHit_BoneName:Name=None
    OutHit_HitItem:int=0
    OutHit_ElementIndex:int=0
    OutHit_FaceIndex:int=0
    OutHit_TraceStart:Vector
    OutHit_TraceEnd:Vector
    ReturnValue:bool
   )
   > n18 = Branch(Condition=@n17.ReturnValue)
   > spawnAssets!(execute, self, position:Vector=@n17.OutHit_Location) -> (then)
   PrintString(execute=@n18.else, InString=Trace Failed)
   > PrintString(InString=Append!(A="To String (Vector)"!(InVec:Vector=@n15) -> (ReturnValue:String), B=" ", C:String="To String (Vector)"(InVec=@n16)) -> (ReturnValue))
   n19 = Event Load Release!() -> (OutputDelegate, then, release:String) {is=title K2Node_CustomEvent}
   > Macro Init Playlist!(execute:Exec, ProcessingText:Text="To Text (String)"(InString=Append!(A="Querying ShotGrid for Release ", B=@n19.release) -> (ReturnValue))) -> (Completed:Exec) {is=title K2Node_MacroInstance}
   n20 = "Equal (Enum)"(
    A:ECheckBoxState Enum=Get CheckedState!(self:Check Box Object Reference=Get CheckBox_assetGroups() -> (CheckBox_assetGroups:Check Box Object Reference)) -> (CheckedState:ECheckBoxState Enum=Unchecked)
    B:ECheckBoxState Enum=Checked
   )
   > n21 = get Assets from ShotGrid release!(
    execute
    self:EUW Asset Playlist Headless Object Reference=Get EUW_AssetPlaylist_headless() -> (EUW_AssetPlaylist_headless:EUW Asset Playlist Headless Object Reference)
    release:String=@n19.release
    useAssetGroups:bool=@n20
   ) -> (then, Assets:Array of Asset Playlist Asset Structures)
   > Set shotGridAssetList(shotGridAssetList:Array of Asset Playlist Asset Structures=@n21) -> (Output_Get:Array of Asset Playlist Asset Structures)
   > Macro Load Asset List!(execute:Exec) -> (then:Exec) {is=title K2Node_MacroInstance}
   > filter Items()
   > filter Releases()
   n22 = Event Load Playlist!() -> (OutputDelegate, then, playlist:String, playlistid:int) {is=title K2Node_CustomEvent}
   > Branch(Condition="Not Equal, Case Insenstive (String)"!(A:String=@n22.playlist, B:String) -> (ReturnValue:bool))
   > Macro Init Playlist(ProcessingText="To Text (String)"(InString=Append(A="Querying ShotGrid for Playlist ", B=@n22.playlist)))
   > n23 = get Assets from ShotGrid playlist!(
    execute
    self:EUW Asset Playlist Headless Object Reference=Get EUW_AssetPlaylist_headless()
    playlist_id:int=@n22.playlistid
    useAssetGroups:bool=@n20
   ) -> (then, Assets:Array of Asset Playlist Asset Structures)
   > Set shotGridAssetList(shotGridAssetList=@n23)
   > Macro Load Asset List()
   > filter Items()
   > filter Releases()
   n24 = Event Load User Assets!() -> (OutputDelegate, then, UserName:String) {is=title K2Node_CustomEvent}
   > Macro Init Playlist(ProcessingText="To Text (String)"(InString=Append(A="Querying ShotGrid for User  ", B=@n24.UserName)))
   > n25 = get Assets from ShotGrid user!(
    execute
    self:EUW Asset Playlist Headless Object Reference=Get EUW_AssetPlaylist_headless()
    username:String=@n24.UserName
    useAssetGroups:bool=@n20
   ) -> (then, Assets:Array of Asset Playlist Asset Structures)
   > Set shotGridAssetList(shotGridAssetList=@n25)
   > Macro Load Asset List()
   Event Construct()
   > Set HorizontalBox_current(HorizontalBox_current=Get HorizontalBox_releaseColumns())
   Event OnMouseEnter()
   > n26 = Branch(Condition=Get mouseEntered() -> (mouseEntered:bool))
   Set mouseEntered(execute=@n26.else, mouseEntered:bool=true) -> (Output_Get:bool)
   > pipeTest()
   Set buttonColor_Default!(execute, buttonColor_Default:Linear Color Structure=(R=0.1,G=0.1,B=0.1,A=1), self:TA Base Editor Button Widget Object Reference=@n14) -> (then, Output_Get:Linear Color Structure)
   > Set buttonColor_Selected!(execute, buttonColor_Selected:Linear Color Structure=(R=0.25,G=0.25,B=0.25,A=1), self:TA Base Editor Button Widget Object Reference=@n14) -> (then, Output_Get:Linear Color Structure)
   > validateVisuals!(execute, self:TA Base Editor Button Widget Object Reference=@n14) -> (then, " ":bool)
   > validateLabelColor!(execute, self:TA Base Editor Button Widget Object Reference=@n14) -> (then, handled:bool)
   "On Check State Changed (CheckBox_assetGroups)"!() -> (OutputDelegate, then, bIsChecked:bool) {is=title K2Node_ComponentBoundEvent}
   Event PreConstruct()
   Event Tick()
  ]}
 Reorder Groups Alphabetically {nodes [
   n1 = getAssetPlaylistStack()
   > Is Valid(InputObject=@n1)
   > n2 = getReleases!(execute, self:EUW Asset Playlist Vertical Group List Object Reference=@n1) -> (then, groups:Array of EUW Playlist Release Object References)
   > n3 = Set releases(releases:Array of EUW Playlist Release Object References=@n2) -> (Output_Get:Array of EUW Playlist Release Object References)
   > clearWidgets(self=@n1)
   > n4 = Reorder Releases Array!(execute, self, inputReleases:Array of EUW Playlist Release Object References=@n3) -> (then, orderedReleases:Array of EUW Playlist Release Object References)
   > n5 = For Each Loop(Array:Array of EUW Playlist Release Object References=@n4) -> (Array Element:EUW Playlist Release Object Reference)
   > AddChild(
    self=Get ScrollBox!(self:EUW Asset Playlist Vertical Group List Object Reference=@n1) -> (ScrollBox:Scroll Box Object Reference)
    Content=@n5
   )
   Reorder Groups Alphabetically() {is=title K2Node_FunctionEntry}
   > PrintString(InString="ReorderGroups Alphabetically <----")
   > Sequence()
   > n6 = get Releases!(execute, self) -> (then, out:Array of EUW Playlist Release Object References)
   > n7 = Set releases(releases=@n6)
   > n8 = For Each Loop(Array=@n7)
   > Reorder Group Alphabetically(self:EUW Playlist Release Object Reference=@n8)
  ]}
 Reorder Releases Array {nodes [
   n1 = Reorder Releases Array!() -> (then, inputReleases:Array of EUW Playlist Release Object References) {is=title K2Node_FunctionEntry}
   > n2 = Set releases(releases=@n1)
   n3 = Get releases() -> (releases:Array of EUW Playlist Release Object References)
   n4 = Branch(
    execute=[@n2.then, @n8.then]
    Condition=Is Valid Index!(TargetArray:Array of EUW Playlist Release Object References=@n3, IndexToTest:int) -> (ReturnValue:bool)
   )
   > n5 = Set lowest(lowest:EUW Playlist Release Object Reference="Get (a copy)"(Array:Array of EUW Playlist Release Object References=@n3) -> (Output:EUW Playlist Release Object Reference)) -> (Output_Get:EUW Playlist Release Object Reference)
   > n6 = For Each Loop with Break!(Exec:Exec, Array:Array of EUW Playlist Release Object References=@n3, Break:Exec) -> (LoopBody:Exec, Array Element:EUW Playlist Release Object Reference, Array Index:int, Completed:Exec)
   n7 = Get lowest() -> (lowest:EUW Playlist Release Object Reference)
   Add(
    execute=@n6.Completed
    TargetArray:Array of EUW Playlist Release Object References=Get orderedReleases2() -> (orderedReleases2:Array of EUW Playlist Release Object References)
    NewItem:"EUW Playlist Release Object Reference (by ref)"=@n7
   ) {is=title K2Node_CallArrayFunction}
   > n8 = Remove Item!(
    execute
    TargetArray:Array of EUW Playlist Release Object References=Get releases()
    Item:"EUW Playlist Release Object Reference (by ref)"=@n7
   ) -> (then, ReturnValue:bool)
   n9 = Get release!(self:EUW Playlist Release Object Reference=@n6) -> (release:String)
   n10 = Get release(self=@n5)
   Branch(execute=@n6.LoopBody, Condition="float < float"(A=String To Float(InString=@n9), B=String To Float(InString=@n10)))
   > Set lowest(lowest=@n6)
   Return Node!(execute=@n4.else, orderedReleases:Array of EUW Playlist Release Object References=Get orderedReleases2())
  ]}
 add Asset To Playlist {nodes [
   n1 = PrintString(InString=----------- aaaa)
   n2 = Get!() -> (
    asset_shotGridId_27_#7:int
    asset_cosmeticType_20_#8:CosmeticPicker_CosmeticTypes Enum
    asset_name_22_#9:String
    asset_gender_25_#10:String
    asset_assetGroup_41_#11:String
    asset_primaryAssetGroup_42_#12:String
    asset_status_33_#13:String
    asset_fixVersion_37_#14:String
   )
   n3 = find Assets Matching Shot Grid Id!(
    execute=@n11.then
    self:EUW Asset Playlist Headless Object Reference=Get EUW_AssetPlaylist_headless()
    id:int=@n2
    cosmeticType:CosmeticPicker_CosmeticTypes Enum=@n2.asset_cosmeticType_20_#8
   ) -> (then, foundAssets:Array of Asset Playlist Asset Return Structures)
   > n4 = For Each Loop!(Exec, Array:Array of Asset Playlist Asset Return Structures=@n3) -> (
    LoopBody
    Array Element_object_2_#15:Fort Item Definition Object Reference
    Array Element_variant_5_#16:bool=~
    Array Element_variant_index_9_#17:Array of Integers
    Array Element_option_index_11_#18:Array of Integers
    Array Index
    Completed
   )
   n5 = Is Valid(exec=[@n1.then, @n4.LoopBody], InputObject=@n4)
   n6 = Make Array("[0]":int=0) -> (Array:Array of Integers)
   n7 = Get asset() -> (asset:Asset Playlist Asset Structure)
   n8 = add Release To Playlist!(execute=@n20.then, self, asset:Asset Playlist Asset Structure=Get asset()) -> (then, release:EUW Playlist Release Object Reference)
   n9 = Get()
   n10 = get Group Name Wrapper!(
    execute=[@n16.else, @n18.then]
    self
    cosmeticName:String=@n9.asset_name_22_#9
    gender:String=@n9.asset_gender_25_#10
    assetGroup:String=@n9.asset_assetGroup_41_#11
    assetGroups:String=@n9.asset_primaryAssetGroup_42_#12
   ) -> (then, ReturnValue:String, genderOut:String)
   n11 = create Group!(
    execute=@n8.then
    self
    groupName:String=@n10
    gender:String=Get().asset_gender_25_#10
    release:EUW Playlist Release Object Reference=@n8
   ) -> (then, group:EUW Playlist Group Object Reference)
   add Cosmetic to Playlist!(
    execute=@n5.Is Not Valid
    self
    asset:Asset Playlist Asset Structure=@n7
    id:Fort Item Definition Object Reference=@n4
    group:EUW Playlist Group Object Reference=@n11
    isVariant:bool=false
    variantIndex:Array of Integers=@n6
    optionIndex:Array of Integers=@n6
   ) -> (then, widget:EUW Playlist Item Object Reference)
   add Cosmetic to Playlist(
    execute=@n5.Is Valid
    asset=@n7
    id=@n4
    group=@n11
    isVariant=@n4.Array Element_variant_5_#16
    variantIndex=@n4.Array Element_variant_index_9_#17
    optionIndex=@n4.Array Element_option_index_11_#18
   )
   n12 = add Asset To Playlist!() -> (then, in_asset:Asset Playlist Asset Structure, filterByType:bool=true, cosmeticType:CosmeticPicker_CosmeticTypes Enum=NewEnumerator0) {is=title K2Node_FunctionEntry}
   > PrintString(InString="\r\n------------------------------------")
   > Set asset(asset:Asset Playlist Asset Structure=@n12) -> (Output_Get:Asset Playlist Asset Structure)
   > n13 = Branch(Condition=@n12.filterByType)
   > n14 = Branch(Condition="Equal (Enum)"(
     A:CosmeticPicker_CosmeticTypes Enum=Get().asset_cosmeticType_20_#8
     B:CosmeticPicker_CosmeticTypes Enum=@n12.cosmeticType
    ))
   n15 = Get()
   PrintString(execute=[@n13.else, @n14.then], InString=Append(A="AddAsset ---> processing ", B=@n15.asset_name_22_#9))
   > Set sg_asset_group(sg_asset_group:String=@n15.asset_assetGroup_41_#11) -> (Output_Get:String)
   > n16 = Branch(Condition="Equal Exactly (String)"(A=Get().asset_gender_25_#10))
   n17 = Get()
   > n18 = Set!(
    execute
    asset_shotGridId_27_#7:int=@n17
    asset_cosmeticType_20_#8:CosmeticPicker_CosmeticTypes Enum=@n17.asset_cosmeticType_20_#8
    asset_name_22_#9:String=@n17.asset_name_22_#9
    asset_gender_25_#10:String=M
    asset_assetGroup_41_#11:String=@n17.asset_assetGroup_41_#11
    asset_primaryAssetGroup_42_#12:String
    asset_status_33_#13:String=@n17.asset_status_33_#13
    asset_fixVersion_37_#14:String=@n17.asset_fixVersion_37_#14
   ) -> (then, Output_Get:Asset Playlist Asset Structure)
   n19 = Get()
   n20 = Set(
    execute=@n10.then
    asset_shotGridId_27_#7=@n19
    asset_cosmeticType_20_#8=@n19.asset_cosmeticType_20_#8
    asset_name_22_#9=@n19.asset_name_22_#9
    asset_gender_25_#10=@n10.genderOut
    asset_assetGroup_41_#11=@n10
    asset_status_33_#13=@n19.asset_status_33_#13
    asset_fixVersion_37_#14=@n19.asset_fixVersion_37_#14
   )
  ]}
 add Assets of Group {nodes [
   n1 = add Assets of Group!() -> (then, sg_asset_group:String, gender:String, exclude:Array of Strings) {is=title K2Node_FunctionEntry}
   > PrintString(InString=Append(A="addAssetsOfGroup ", B=@n1))
   > n2 = get Assets of Group!(
    execute
    self:EUW Asset Playlist Headless Object Reference=Get EUW_AssetPlaylist_headless()
    sg_asset_group:String=@n1
    matchGender:bool=true
    gender:String=@n1.gender
   ) -> (then, Assets:Array of Asset Playlist Asset Structures)
   n3 = Length(TargetArray:Array of Asset Playlist Asset Structures=@n2)
   > PrintString(InString=Append!(A="found ", B="To String (Integer)"(InInt=@n3), C:String=" assets") -> (ReturnValue))
   > n4 = For Each Loop!(Exec, Array:Array of Asset Playlist Asset Structures=@n2) -> (LoopBody, Array Element:Asset Playlist Asset Structure, Array Index, Completed)
   n5 = For Each Loop(Exec=@n4.Completed, Array=Get filteredAssets() -> (filteredAssets:Array of Asset Playlist Asset Structures))
   > add Asset To Playlist!(
    execute
    self
    in_asset:Asset Playlist Asset Structure=@n5
    filterByType:bool=false
    cosmeticType:CosmeticPicker_CosmeticTypes Enum=NewEnumerator0
   ) -> (then) {is=title K2Node_CallFunction}
   Set match(execute=@n4.LoopBody, match:bool=false) -> (Output_Get:bool)
   > n6 = For Each Loop(Array:Array of Strings=@n1.exclude) -> (Array Element:String)
   n7 = Branch(
    execute=@n6.Completed
    Condition="Equal (Boolean)"(A:bool=Get match() -> (match:bool), B:bool=true) -> (ReturnValue:bool=~) {is=title K2Node_PromotableOperator}
   )
   Add(
    execute=@n7.else
    TargetArray:Array of Asset Playlist Asset Structures=Get filteredAssets()
    NewItem:"Asset Playlist Asset Structure (by ref)"=@n4
   )
   n8 = Break Asset Playlist Asset!(AssetPlaylist_asset:"Asset Playlist Asset Structure (by ref)"=@n4) -> (
    shotGridId_27_#7:int
    cosmeticType_20_#8:CosmeticPicker_CosmeticTypes Enum
    name_22_#9:String
    gender_25_#10:String
    assetGroup_41_#11:String
    primaryAssetGroup_42_#12:String
    status_33_#13:String
    fixVersion_37_#14:String
   ) {is=title K2Node_BreakStruct}
   PrintString(execute=@n6.LoopBody, InString=Append(A=@n6, B=" vs ", C=@n8.name_22_#9))
   > Branch(Condition="Equal, Case Insensitive (String)"(A=@n6, B=@n8.name_22_#9))
   > Set match(match=true)
   > PrintString(InString=Match)
  ]}
 add Cosmetic to Playlist {nodes [
   n1 = add Cosmetic to Playlist!() -> (
    then
    asset:Asset Playlist Asset Structure
    id:Fort Item Definition Object Reference
    group:EUW Playlist Group Object Reference
    isVariant:bool=~
    variantIndex:Array of Integers
    optionIndex:Array of Integers
   ) {is=title K2Node_FunctionEntry}
   > PrintString(InString=add cosmetic to playlist)
   n2 = Break Asset Playlist Asset(AssetPlaylist_asset=@n1)
   > n3 = name to UI Name!(execute, self, SourceString:String=@n2.name_22_#9) -> (then, ReturnValue:String)
   > n4 = Create Widget!(
    execute:Exec
    WorldContextObject:Object Reference
    Class:User Widget Class Reference=&/EUW_PlaylistItem.EUW_PlaylistItem_C
    OwningPlayer:Player Controller Object Reference
   ) -> (then:Exec, ReturnValue:EUW Playlist Item Object Reference)
   > setup!(
    execute
    self:EUW Playlist Item Object Reference=@n4
    item:Fort Item Definition Object Reference=@n1.id
    asset:Asset Playlist Asset Structure=@n1
    UIName:String=@n3
    isVariant:bool=@n1.isVariant
    variantIndex:Array of Integers=@n1.variantIndex
    optionIndex:Array of Integers=@n1.optionIndex
   ) -> (then)
   > AddChild(
    self=Get VerticalBox_itemGroup!(self:EUW Playlist Group Object Reference=@n1.group) -> (VerticalBox_itemGroup:Vertical Box Object Reference)
    Content=@n4
   )
   > PrintString(InString=Append!(
     A="** adding "
     B=Enum to String!(Enumerator:CosmeticPicker_CosmeticTypes Enum=@n2.cosmeticType_20_#8) -> (ReturnValue:String)
     C=" '"
     D:String=@n3
     E:String=' to '
     F:String=@n2.assetGroup_41_#11
     G:String="' ,"
     H:String=GetDisplayName(Object=@n4)
     I:String=" to "
     J:String=GetDisplayName(Object=@n1.group)
    ) -> (ReturnValue))
   > Return Node!(execute, widget:EUW Playlist Item Object Reference=@n4)
  ]}
 add Group to Playlist {nodes [
   n1 = add Group to Playlist!() -> (
    then
    cosmeticType:CosmeticPicker_CosmeticTypes Enum=NewEnumerator0
    asset:Asset Playlist Asset Structure
    release:EUW Playlist Release Object Reference
   ) {is=title K2Node_FunctionEntry}
   n2 = Break Asset Playlist Asset(AssetPlaylist_asset=@n1.asset)
   > PrintString(InString=Append!(A="AddGroup for ", B=@n2.name_22_#9) -> (ReturnValue))
   > n3 = Set gender(gender:String=@n2.gender_25_#10) -> (Output_Get:String)
   > n4 = Branch(Condition="Equal Exactly (String)"(A=@n3))
   > n5 = Set gender(gender=M)
   n6 = get Group Name Wrapper(
    execute=[@n4.else, @n5.then]
    cosmeticName=@n2.name_22_#9
    gender=Get gender() -> (gender:String)
    assetGroup=@n2.assetGroup_41_#11
    assetGroups=@n2.primaryAssetGroup_42_#12
   )
   > Set gender(gender=@n6.genderOut)
   > create Group(groupName=@n6, gender=Get gender(), release=@n1.release)
  ]}
 add Release To Playlist {nodes [
   n1 = add Release To Playlist!() -> (then, asset:Asset Playlist Asset Structure) {is=title K2Node_FunctionEntry}
   > n2 = get Releases()
   > n3 = For Each Loop(Array:Array of EUW Playlist Release Object References=@n2) -> (Array Element:EUW Playlist Release Object Reference)
   n4 = Break Asset Playlist Asset(AssetPlaylist_asset=@n1)
   n5 = create Release!(execute=@n3.Completed, self, release:String=@n4.fixVersion_37_#14) -> (then, releaseObject:EUW Playlist Release Object Reference)
   > Return Node!(execute, release:EUW Playlist Release Object Reference=@n5)
   Branch(
    execute=@n3.LoopBody
    Condition="Equal, Case Insensitive (String)"(A=Get release(self=@n3), B=@n4.fixVersion_37_#14)
   )
   > Return Node(release=@n3)
  ]}
 clear Playlist {nodes [
   clear Playlist() {is=title K2Node_FunctionEntry}
   > ClearChildren(self=Get HorizontalBox_current() -> (HorizontalBox_current:Horizontal Box Object Reference))
   > n1 = getAssetPlaylistStack()
   > clearWidgets(self=@n1)
  ]}
 compile Shaders {nodes [
   compile Shaders() {is=title K2Node_FunctionEntry}
   > n1 = get Groups()
   > n2 = For Each Loop(Array:Array of EUW Playlist Group Object References=@n1) -> (Array Element:EUW Playlist Group Object Reference)
   > n3 = get Items(self=@n2)
   > n4 = For Each Loop(Array:Array of EUW Playlist Item Object References=@n3) -> (Array Element:EUW Playlist Item Object Reference)
   > Select Item(self=@n4)
  ]}
 create Group {nodes [
   n1 = create Group!() -> (then, groupName:String, gender:String, release:EUW Playlist Release Object Reference) {is=title K2Node_FunctionEntry}
   > PrintString(InString=Append!(A=Create Group, B=" ", C:String=@n1, D:String=', E:String=@n1.gender, F:String="' ", G:String=GetDisplayName(Object=@n1.release)) -> (ReturnValue))
   > n2 = getGroups!(execute, self:EUW Playlist Release Object Reference=@n1.release) -> (then, groups:Array of EUW Playlist Group Object References)
   > n3 = For Each Loop(Array:Array of Widget Object References=@n2) -> (Array Element:Widget Object Reference)
   n4 = Create Widget(execute=@n3.Completed, Class=&/EUW_PlaylistGroup.EUW_PlaylistGroup_C) -> (ReturnValue:EUW Playlist Group Object Reference)
   > setupGroup!(execute, self:EUW Playlist Group Object Reference=@n4, group:String=@n1, gender:String=@n1.gender) -> (then)
   > addGroup!(execute, self:EUW Playlist Release Object Reference=@n1.release, group:EUW Playlist Group Object Reference=@n4) -> (then)
   > n5 = Switch on ECheckBoxState!(
    execute
    Selection:ECheckBoxState Enum=GetCheckedState!(self:Check Box Object Reference=Get CheckBox_cosmeticPicker() -> (CheckBox_cosmeticPicker:Check Box Object Reference)) -> (ReturnValue:ECheckBoxState Enum=Unchecked)
   ) -> (Unchecked:Exec, Checked:Exec, Undetermined:Exec) {is=title K2Node_SwitchEnum}
   n6 = getAssetPlaylistStack(execute=@n5.Checked)
   > n7 = Is Valid(InputObject=@n6)
   > n8 = Create Widget(Class=&EUW_PlaylistGroup_C)
   > setupGroup(self=@n8, group=@n1, gender=@n1.gender)
   > n9 = getRelease!(execute, self:EUW Asset Playlist Vertical Group List Object Reference=[@n6, @n6], release:String=Get release(self=@n1.release)) -> (then, playlistRelease:EUW Playlist Release Object Reference)
   > n10 = Is Valid(InputObject=@n9)
   PrintString(execute=@n10.Is Not Valid, InString=could not find release)
   > n11 = addWidget!(execute, self:EUW Asset Playlist Vertical Group List Object Reference=@n6, group:EUW Playlist Group Object Reference=@n8) -> (then)
   Set assetPlaylistGroup!(execute=[@n13.then, @n11.then], assetPlaylistGroup:EUW Playlist Group Object Reference=@n4, self:EUW Playlist Group Object Reference=@n8) -> (then, Output_Get:EUW Playlist Group Object Reference)
   > n12 = Set cosmeticPickerGroup!(execute, cosmeticPickerGroup:EUW Playlist Group Object Reference=@n8, self:EUW Playlist Group Object Reference=@n4) -> (then, Output_Get:EUW Playlist Group Object Reference)
   Return Node!(execute=[@n12.then, @n7.Is Not Valid, @n5.Unchecked, @n5.Undetermined], group:EUW Playlist Group Object Reference=@n4)
   n13 = addGroup(execute=@n10.Is Valid, self=@n9, group=@n8)
   n14 = Cast To EUW_PlaylistGroup!(execute=@n3.LoopBody, Object=@n3) -> (then, CastFailed, AsEUW Playlist Group:EUW Playlist Group Object Reference) {is=title K2Node_DynamicCast}
   n15 = Get groupName!(self:EUW Playlist Group Object Reference=@n14) -> (groupName:String)
   > Branch(Condition=AND Boolean(A="Equal Exactly (String)"(A=@n1, B=@n15), B="Not Equal Exactly (String)"(A=@n15, B="(M) -")))
   > Return Node(group=@n14)
  ]}
 create Release {nodes [
   n1 = create Release!() -> (then, release:String) {is=title K2Node_FunctionEntry}
   > n2 = Create Widget(Class=&/EUW_PlaylistRelease.EUW_PlaylistRelease_C) -> (ReturnValue:EUW Playlist Release Object Reference)
   > setReleaseText!(execute, self:EUW Playlist Release Object Reference=@n2, release:String=@n1) -> (then)
   > AddChild(self=Get HorizontalBox_current(), Content=@n2)
   > PrintString(InString=Append!(A="Created Release ", B=@n1, C=" ", D=GetDisplayName(Object=@n2)) -> (ReturnValue))
   > n3 = Switch on ECheckBoxState(Selection=GetCheckedState(self=Get CheckBox_cosmeticPicker()) -> (ReturnValue=Unchecked))
   n4 = getAssetPlaylistStack(execute=@n3.Checked)
   > n5 = Is Valid(InputObject=@n4)
   > n6 = Create Widget(Class=&EUW_PlaylistRelease_C)
   > setReleaseText(self=@n6, release=@n1)
   > Set vertical!(execute, vertical:bool=true, self:EUW Playlist Release Object Reference=@n6) -> (then, Output_Get:bool)
   > Set release!(execute, release:String=@n1, self:EUW Playlist Release Object Reference=@n6) -> (then, Output_Get:String)
   > n7 = AddChild(self=Get ScrollBox(self=@n4), Content=@n6)
   Return Node!(execute=[@n7.then, @n5.Is Not Valid, @n3.Unchecked, @n3.Undetermined], releaseObject:EUW Playlist Release Object Reference=@n2)
  ]}
 deselect all Group Widgets {nodes [
   n1 = deselect all Group Widgets!() -> (then, exception:EUW Playlist Group Object Reference) {is=title K2Node_FunctionEntry}
   > n2 = get Groups()
   > n3 = For Each Loop(Array=@n2)
   > n4 = Cast To EUW_PlaylistGroup(Object=@n3)
   > Branch(Condition="Not Equal (Object)"(A:Object Reference=@n4, B:Object Reference=@n1) -> (ReturnValue:bool=~) {is=title K2Node_PromotableOperator})
   > deselect Group(self:EUW Playlist Group Object Reference=@n4)
  ]}
 deselect all Widgets {nodes [
   deselect all Widgets() {is=title K2Node_FunctionEntry}
   > n1 = Sequence()
   > n2 = get Groups()
   > n3 = For Each Loop(Array=@n2)
   > n4 = Cast To EUW_PlaylistGroup(Object=@n3)
   > deselect All Widgets(self:EUW Playlist Group Object Reference=@n4)
   n5 = getAssetPlaylistStack(execute=@n1.then_1)
   > Is Valid(InputObject=@n5)
   > deselectAllGroups(self:EUW Asset Playlist Vertical Group List Object Reference=@n5)
  ]}
 deselect all Widgets of CosmeticType {nodes [
   n1 = deselect all Widgets of CosmeticType!() -> (then, cosmeticType:CosmeticPicker_CosmeticTypes Enum=NewEnumerator0) {is=title K2Node_FunctionEntry}
   > n2 = get Groups()
   > n3 = For Each Loop(Array=@n2)
   > n4 = Cast To EUW_PlaylistGroup(Object=@n3)
   > deselect Widgets of CosmeticType!(execute, self:EUW Playlist Group Object Reference=@n4, cosmeticType:CosmeticPicker_CosmeticTypes Enum=@n1) -> (then)
  ]}
 filter Items {nodes [
   filter Items() {is=title K2Node_FunctionEntry}
   > PrintString(InString=FilterItems)
   > n1 = getOptions!(
    execute
    self:EUW Asset Playlist Options Widget Object Reference=Get EUW_AssetPlaylistOptions() -> (EUW_AssetPlaylistOptions:EUW Asset Playlist Options Widget Object Reference)
   ) -> (then, cosmeticType:CosmeticPicker_CosmeticTypes Enum=NewEnumerator0, items:bool, unknown:bool, variants:bool, builtin:bool, thumbnails:bool)
   > reorderReleases()
   > Reorder Groups Alphabetically() {is=title K2Node_CallFunction}
   > n2 = get Groups()
   > n3 = For Each Loop(Array:Array of EUW Playlist Group Object References=@n2) -> (Array Element:EUW Playlist Group Object Reference)
   > filter Items!(
    execute
    self:EUW Playlist Group Object Reference=@n3
    cosmeticType:CosmeticPicker_CosmeticTypes Enum=@n1
    knownItems:bool=@n1.items
    unknown:bool=@n1.unknown
    variants:bool=@n1.variants
    builtIn:bool=@n1.builtin
   ) -> (then) {is=title K2Node_CallFunction}
   set Thumbnail!(execute, self:EUW Playlist Group Object Reference=@n3, thumbnail:bool=@n1.thumbnails) -> (then)
  ]}
 filter Playlist {nodes [
   n1 = filter Playlist!() -> (then, cosmeticType:CosmeticPicker_CosmeticTypes Enum=NewEnumerator0) {is=title K2Node_FunctionEntry}
   > PrintString(InString=Enum to String(Enumerator=@n1))
   > n2 = get Groups()
   > n3 = For Each Loop(Array=@n2)
   > Set containsVisibleItem(containsVisibleItem:bool=false) -> (Output_Get:bool)
   > n4 = get Items(self=@n3)
   > n5 = For Each Loop(Array:Array of EUW Playlist Item Object References=@n4) -> (Array Element:EUW Playlist Item Object Reference)
   n6 = Branch(execute=@n5.Completed, Condition=Get containsVisibleItem() -> (containsVisibleItem:bool))
   > SetVisibility(self=@n3, InVisibility=Visible)
   SetVisibility(execute=@n6.else, self=@n3, InVisibility=Collapsed)
   n7 = Get cosmeticType!(self:EUW Playlist Item Object Reference=@n5) -> (cosmeticType:CosmeticPicker_CosmeticTypes Enum=NewEnumerator0)
   n8 = Branch(execute=@n5.LoopBody, Condition=OR Boolean(A="Equal (Enum)"(A=@n7, B=@n1), B="Equal (Enum)"(A=@n1, B=NewEnumerator6)))
   > SetVisibility(self=@n5, InVisibility=Visible)
   > Set containsVisibleItem(containsVisibleItem=true)
   SetVisibility(execute=@n8.else, self=@n5, InVisibility=Collapsed)
  ]}
 filter Releases {nodes [
   filter Releases() {is=title K2Node_FunctionEntry}
   > n1 = get Releases()
   > n2 = For Each Loop(Array:Array of EUW Playlist Release Object References=@n1) -> (Array Element:EUW Playlist Release Object Reference)
   > n3 = getGroups(self=@n2)
   > n4 = Branch(Condition=Is Valid Index(TargetArray:Array of EUW Playlist Group Object References=@n3))
   > SetVisibility(self=@n2, InVisibility=Visible)
   SetVisibility(execute=@n4.else, self=@n2, InVisibility=Collapsed)
  ]}
 get Group Name {nodes [
   n1 = get Group Name!() -> (then, gender:String, cosmeticName:String, assetGroup:String, assetGroups:String) {is=title K2Node_FunctionEntry}
   > Set input_gender(input_gender:String=@n1) -> (Output_Get:String)
   > Set input_groupName(input_groupName:String=@n1.assetGroup) -> (Output_Get:String)
   > Set input_cosmeticName(input_cosmeticName:String=@n1.cosmeticName) -> (Output_Get:String)
   > n2 = Set inputAssetGroups(inputAssetGroups:String=@n1.assetGroups) -> (Output_Get:String)
   > PrintString(InString=Append!(A="assetGroupsName: ", B=@n1.assetGroups) -> (ReturnValue))
   > n3 = Branch(Condition="Equal Exactly (String)"(A=@n2))
   > n4 = getShortCosmeticName!(execute, self, SourceString:String=Get input_cosmeticName() -> (input_cosmeticName:String)) -> (then, ReturnValue:String)
   > n5 = Set short_cosmeticName(short_cosmeticName:String=@n4) -> (Output_Get:String)
   > n6 = Branch(Condition="Equal Exactly (String)"(A=Get input_groupName() -> (input_groupName:String), B=zzz))
   > Return Node!(execute, ReturnValue:String=-, genderOut:String=-, message:String)
   PrintString(execute=@n6.else, InString=Append!(A="shortname : `", B=@n5, C:String="`") -> (ReturnValue))
   > n7 = Branch(Condition=Contains(SearchIn=Get input_groupName(), Substring=@n5))
   n8 = Len(S=Get input_groupName())
   n9 = Branch(execute=[@n7.then, @n29.then, @n30.then], Condition="integer > integer"(A=@n8, B=25))
   > Set final_cosmeticName(final_cosmeticName:String=Get short_cosmeticName() -> (short_cosmeticName:String)) -> (Output_Get:String)
   n10 = "To String (Integer)"(InInt=@n8)
   > n11 = PrintString(InString=Append!(A=@n10, B=" Using Short Cosmetic Name") -> (ReturnValue))
   n12 = Branch(execute=[@n26.then, @n11.then], Condition="Equal, Case Insensitive (String)"(A=Get input_gender() -> (input_gender:String), B=None))
   > n13 = Branch(Condition=Contains(SearchIn=Get input_cosmeticName(), Substring=Female))
   > Return Node(ReturnValue=Append(A="(F) ", B=Get final_cosmeticName() -> (final_cosmeticName:String)), genderOut=F, message=b)
   n14 = Branch(execute=@n13.else, Condition=Contains(SearchIn=Get input_cosmeticName(), Substring=male))
   > Return Node(ReturnValue=Append(A="(M) ", B=Get final_cosmeticName()), genderOut=M, message=c)
   n15 = name to UI Name(execute=@n14.else, SourceString=Get input_cosmeticName())
   > PrintString(InString=Append!(A=Searching existing groups for ', B=@n15, C:String=') -> (ReturnValue))
   > n16 = get Groups()
   > n17 = For Each Loop(Array:Array of Widget Object References=@n16) -> (Array Element:Widget Object Reference)
   PrintString(execute=@n17.Completed, InString=nope - now searching for a group that contains the group name)
   > n18 = For Each Loop(Array=@n16)
   PrintString(execute=@n18.Completed, InString=could not find an exact match for the group)
   n19 = Get final_cosmeticName()
   > Return Node(ReturnValue=@n19, genderOut=Get input_gender(), message=f)
   n20 = Cast To EUW_PlaylistGroup(execute=@n18.LoopBody, Object=@n18)
   n21 = Get groupName(self=@n20)
   > n22 = Branch(Condition=Contains(SearchIn=@n21, Substring=@n19))
   > PrintString(InString=Append!(A=@n21, B=" <---- match") -> (ReturnValue))
   > Return Node(ReturnValue=@n21, genderOut=Get gender!(self:EUW Playlist Group Object Reference=@n20) -> (gender), message=e)
   PrintString(execute=@n22.else, InString=@n21)
   n23 = Cast To EUW_PlaylistGroup(execute=@n17.LoopBody, Object=@n17)
   n24 = Get groupName(self=@n23)
   > Branch(Condition=Contains(SearchIn=@n24, Substring=@n15))
   > PrintString(InString=Append(A=@n24, B=" <---- match"))
   > Return Node(ReturnValue=@n24, genderOut=Get gender(self=@n23), message=d)
   Return Node(
    execute=@n12.else
    ReturnValue=Append!(A="(", B=Get input_gender(), C:String=") ", D:String=Get final_cosmeticName()) -> (ReturnValue)
    genderOut=Get input_gender()
    message=a
   )
   n25 = Get input_groupName()
   Set final_cosmeticName(execute=@n9.else, final_cosmeticName=@n25)
   > n26 = PrintString(InString=Append(A=@n10, B=" Using input group Name", C=" ", D=@n25))
   n27 = Get input_groupName()
   n28 = Branch(execute=@n7.else, Condition="Equal, Case Insensitive (String)"(A=@n27, B=none))
   > n29 = Set input_groupName(input_groupName=@n5)
   n30 = Set input_groupName(execute=@n28.else, input_groupName=Append!(A=@n27, B=" : ", C=@n5) -> (ReturnValue))
   Return Node(execute=@n3.else, ReturnValue=Get inputAssetGroups() -> (inputAssetGroups:String), genderOut=Get input_gender())
   PrintString(InString=@n24)
  ]}
 get Group Name Wrapper {nodes [
   n1 = get Group Name Wrapper!() -> (then, cosmeticName:String, gender:String, assetGroup:String, assetGroups:String) {is=title K2Node_FunctionEntry}
   > PrintString(InString=Append!(A="Group Name Wrapper : ", B=@n1.gender, C=",", D:String=@n1.assetGroup, E:String=",", F:String=@n1, G:String=",", H:String=@n1.assetGroups) -> (ReturnValue))
   > n2 = get Group Name!(execute, self, gender:String=@n1.gender, cosmeticName:String=@n1, assetGroup:String=@n1.assetGroup, assetGroups:String=@n1.assetGroups) -> (then, ReturnValue:String, genderOut:String, message:String) {is=title K2Node_CallFunction}
   > PrintString(InString=Append!(A=GetGroupName returned ', B=@n2, C=",", D=@n2.genderOut, E=', F=" ", G=@n2.message) -> (ReturnValue))
   > Return Node!(execute, ReturnValue=@n2, genderOut=@n2.genderOut)
  ]}
 get Groups {nodes [
   get Groups() {is=title K2Node_FunctionEntry}
   > n1 = get Releases()
   > n2 = For Each Loop(Array:Array of EUW Playlist Release Object References=@n1) -> (Array Element:EUW Playlist Release Object Reference)
   Return Node!(
    execute=@n2.Completed
    out:Array of EUW Playlist Group Object References=Get groups() -> (groups:Array of EUW Playlist Group Object References)
   )
   > n3 = getGroups(self=@n2)
   > Append Array!(
    execute
    TargetArray:Array of EUW Playlist Group Object References=Get groups()
    SourceArray:Array of EUW Playlist Group Object References=@n3
   ) -> (then)
  ]}
 get Releases {nodes [
   get Releases() {is=title K2Node_FunctionEntry}
   > n1 = For Each Loop(Array:Array of Widget Object References=GetAllChildren(self=Get HorizontalBox_current())) -> (Array Element:Widget Object Reference)
   Return Node(execute=@n1.Completed, out:Array of EUW Playlist Release Object References=Get releases())
   > n2 = Cast To EUW_PlaylistRelease!(execute, Object=@n1) -> (then, CastFailed, AsEUW Playlist Release:EUW Playlist Release Object Reference) {is=title K2Node_DynamicCast}
   > Add(TargetArray:Array of EUW Playlist Release Object References=Get releases(), NewItem:"EUW Playlist Release Object Reference (by ref)"=@n2)
  ]}
 getAssetPlaylistStack {nodes [
   getAssetPlaylistStack() {is=title K2Node_FunctionEntry}
   > n1 = GetAllWidgetsOfClass!(
    execute
    WorldContextObject:Object Reference
    WidgetClass:User Widget Class Reference=&/EUW_AssetPlaylist_verticalGroupList.EUW_AssetPlaylist_verticalGroupList_C
    TopLevelOnly:bool=false
   ) -> (then, FoundWidgets:Array of EUW Asset Playlist Vertical Group List Object References)
   > n2 = Branch(Condition=Is Valid Index(TargetArray:Array of EUW Asset Playlist Vertical Group List Object References=@n1))
   n3 = "Get (a copy)"(Array:Array of EUW Asset Playlist Vertical Group List Object References=@n1) -> (Output:EUW Asset Playlist Vertical Group List Object Reference)
   > PrintString(InString=Append!(A="found stack ", B=GetDisplayName(Object=@n3)) -> (ReturnValue))
   > Return Node!(execute, assetPlaylist_stack:EUW Asset Playlist Vertical Group List Object Reference=@n3)
   PrintString(execute=@n2.else, InString=no stack)
   > Return Node()
  ]}
 getRelease {nodes [
   n1 = getRelease!() -> (then, fixVersion:String) {is=title K2Node_FunctionEntry}
   > n2 = get Releases() {is=title K2Node_CallFunction}
   > n3 = For Each Loop(Array:Array of EUW Playlist Release Object References=@n2) -> (Array Element:EUW Playlist Release Object Reference)
   > Branch(Condition="Equal, Case Insensitive (String)"(A=Get release(self=@n3), B=@n1))
   > Return Node!(execute, release:EUW Playlist Release Object Reference=@n3)
  ]}
 getShortCosmeticName {nodes [
   n1 = getShortCosmeticName!() -> (then, SourceString:String) {is=title K2Node_FunctionEntry}
   n2 = Replace(SourceString=Replace(SourceString=Split(SourceString=@n1, InStr=" - ", SearchDir=FromEnd).RightS, From=female), From=male)
   > Set name(name:String=@n2) -> (Output_Get:String)
   n3 = Split(SourceString=@n2, InStr=" ", SearchDir=FromStart)
   > n4 = Branch(Condition=@n3.ReturnValue)
   > n5 = Set name(name=@n3)
   Return Node!(execute=[@n4.else, @n5.then], ReturnValue:String=Get name() -> (name:String))
  ]}
 macro_initPlaylist {nodes [
   n1 = Inputs!() -> (execute, ProcessingText:Text="LOCTEXT(\"Querying ShotGrid...\")")
   > SetVisibility(self=Get ScrollBox_playlist() -> (ScrollBox_playlist:Scroll Box Object Reference), InVisibility=Collapsed)
   > SetVisibility(self=Get Overlay_processing() -> (Overlay_processing:Overlay Object Reference), InVisibility=Visible)
   > clear Playlist() {is=title K2Node_CallFunction}
   > SetPercent!(execute, self:Progress Bar Object Reference=Get ProgressBar() -> (ProgressBar:Progress Bar Object Reference), InPercent:float=0) -> (then)
   > "SetText (Text)"(self=Get CommonTextBlock_processing() -> (CommonTextBlock_processing:Common Text Object Reference), InText=@n1)
   > Delay(Duration=0.05)
   > Outputs!(Completed:Exec)
  ]}
 macro_loadAssetList {nodes [
   Inputs!() -> (execute)
   n1 = Length(TargetArray=Get shotGridAssetList() -> (shotGridAssetList:Array of Asset Playlist Asset Structures))
   > Set numAssets(numAssets:int=@n1) -> (Output_Get:int=0)
   > Set ProgressPercent(ProgressPercent:double=0) -> (Output_Get:double)
   n2 = float + float(A="To Float (Integer)"(InInt=@n1), B=1)
   > Set ProgressPercent_step(ProgressPercent_step:double=float * float(A=float / float(A=1, B=@n2), B=10) {is=title K2Node_PromotableOperator}) -> (Output_Get:double)
   > n3 = For Each Loop(Array:Array of Asset Playlist Asset Structures=Get shotGridAssetList()) -> (Array Element:Asset Playlist Asset Structure)
   n4 = For Each Loop(Exec=@n3.Completed, Array=Get shotGridAssetList())
   n5 = Set assetProgress(execute=@n4.Completed, assetProgress:int=0) -> (Output_Get:int=0)
   n6 = Get shotGridAssetList()
   n7 = Branch(execute=[@n5.then, @n9.else, @n13.then], Condition="integer > integer"(A=Length(TargetArray=@n6)))
   > add Asset To Playlist(
    in_asset="Get (a copy)"(Array:Array of Asset Playlist Asset Structures=@n6) -> (Output:Asset Playlist Asset Structure)
    filterByType=false
    cosmeticType=NewEnumerator0
   )
   > Remove Index(TargetArray:Array of Asset Playlist Asset Structures=@n6, IndexToRemove=0)
   > n8 = Increment Int(Value=Get assetProgress() -> (assetProgress:int=0))
   > n9 = Branch(Condition="Equal (Integer)"(A="% (Integer)"(A=@n8, B=10) -> (ReturnValue=0)))
   n10 = "To String (Integer)"(InInt=Get numAssets() -> (numAssets:int=0))
   n11 = "To String (Integer)"(InInt=Get assetProgress() -> (assetProgress=0))
   > "SetText (Text)"(
    self=Get CommonTextBlock_processing()
    InText="To Text (String)"(InString=Append!(A="Searching for ", B=@n11, C:String=" of ", D:String=@n10, E:String=" ShotGrid assets") -> (ReturnValue))
   )
   > n12 = Set ProgressPercent(ProgressPercent=float + float(A=Get ProgressPercent() -> (ProgressPercent:double), B=Get ProgressPercent_step() -> (ProgressPercent_step:double)))
   > SetPercent(self=Get ProgressBar(), InPercent=@n12)
   > n13 = Delay(Duration=0.001)
   PrintString(execute=@n7.else, InString="\r\n\r\n--- Sorting Groups ---\r\n")
   > n14 = get Groups() {is=title K2Node_CallFunction}
   > n15 = For Each Loop(Array:Array of EUW Playlist Group Object References=@n14) -> (Array Element:EUW Playlist Group Object Reference)
   SetVisibility(execute=@n15.Completed, self=Get Overlay_processing(), InVisibility=Collapsed)
   > SetVisibility(self=Get ScrollBox_playlist(), InVisibility=Visible)
   > Outputs!(then:Exec)
   reorder Group(execute=@n15.LoopBody, self:EUW Playlist Group Object Reference=@n15)
   > hide Empty Group(self:EUW Playlist Group Object Reference=@n15)
   n16 = getRelease!(
    execute=@n4.LoopBody
    self:Self Object Reference
    fixVersion:String=Break Asset Playlist Asset(AssetPlaylist_asset=@n4).fixVersion_37_#14
   ) -> (then, release:EUW Playlist Release Object Reference) {is=title K2Node_CallFunction}
   > add Group to Playlist!(
    execute
    self
    cosmeticType:CosmeticPicker_CosmeticTypes Enum=NewEnumerator0
    asset:Asset Playlist Asset Structure=@n4
    release:EUW Playlist Release Object Reference=@n16
   ) -> (then) {is=title K2Node_CallFunction}
   add Release To Playlist(execute=@n3.LoopBody, asset=@n3) {is=title K2Node_CallFunction}
  ]}
 name to UI Name {nodes [
   n1 = name to UI Name!() -> (then, SourceString:String) {is=title K2Node_FunctionEntry}
   n2 = Replace(SourceString=@n1, From="Backpack - ")
   n3 = Replace(
    SourceString=Replace(SourceString=Replace(SourceString=@n2, From="Female Commando - "), From="Female  Commando - ")
    From="Female Commando  - "
   )
   n4 = Replace(SourceString=Replace(SourceString=Replace(SourceString=@n3, From="Male Commando - "), From="Pickaxe - "), From="Emote - ")
   > Return Node(ReturnValue=Replace(SourceString=Replace(SourceString=@n4, From="Glider - "), From="Wrap - "))
  ]}
 pipeTest {nodes [
   pipeTest() {is=title K2Node_FunctionEntry}
   > n1 = Execute Python Script!(
    execute
    PythonScript="test =True\r\n\r\ntry:\r\n   import unpipe\r\nexcept ImportError:\r\n   print(\"UnPipe not loaded!\")\r\n   test = False\r\n\r\nunpipe = test\r\n\r\n\r\n"
   ) -> (then, ReturnValue, unpipe:bool=~) {outputs [unpipe]}
   > n2 = Branch(Condition=@n1.unpipe)
   > PrintString(InString=UnPipe True)
   PrintString(execute=@n2.else, InString=UnPipe False)
   > SetVisibility(self=Get Overlay_processing(), InVisibility=Visible)
   > "SetText (Text)"(
    self=Get CommonTextBlock_processing()
    InText="LOCTEXT(\"\\r\\n\\r\\n!!!! Unpipe not loaded !!!!\\r\\nAssetPlaylist depends on UnPipe being loaded.\\r\\nPlease launch UGS via Launchbox to get the correct environment.\")"
   )
   > SetVisibility(self=Get Generic01_TabContainer() -> (Generic01_TabContainer:Generic 01 Tab Container Object Reference), InVisibility=Collapsed)
   > SetFillColorAndOpacity!(execute, self:Progress Bar Object Reference=Get ProgressBar(), InColor:Linear Color Structure=(R=1,G=0.1,B=0,A=1)) -> (then)
  ]}
 reorderReleases {nodes [
   reorderReleases() {is=title K2Node_FunctionEntry}
   > n1 = Sequence()
   > n2 = get Releases()
   > n3 = Set releases(releases=@n2)
   > ClearChildren(self=Get HorizontalBox_current())
   > n4 = Reorder Releases Array(inputReleases=@n3) {is=title K2Node_CallFunction}
   > n5 = For Each Loop(Array:Array of EUW Playlist Release Object References=@n4) -> (Array Element:EUW Playlist Release Object Reference)
   > AddChild(self=Get HorizontalBox_current(), Content=@n5)
   n6 = getAssetPlaylistStack(execute=@n1.then_1) {is=title K2Node_CallFunction}
   > Is Valid(InputObject=@n6)
   > n7 = getReleases(self=@n6)
   > n8 = Set releases(releases=@n7)
   > clearWidgets(self=@n6)
   > n9 = Reorder Releases Array(inputReleases=@n8)
   > n10 = For Each Loop(Array=@n9)
   > AddChild(self=Get ScrollBox(self=@n6), Content=@n10)
  ]}
 safe Reload {nodes [
   safe Reload() {is=title K2Node_FunctionEntry}
   n1 = Get Tab Widget() -> (Tab Widget:Widget Object Reference)
   > n2 = Cast To EUW_AssetPlaylist_releaseWidget!(execute, Object=@n1) -> (then, CastFailed, AsEUW Asset Playlist Release Widget:EUW Asset Playlist Release Widget Object Reference) {is=title K2Node_DynamicCast}
   > PrintString(InString=Release)
   > Reload(self:EUW Asset Playlist Release Widget Object Reference=@n2)
   n3 = Cast To EUW_AssetPlaylist_playlistWidget!(execute=@n2.CastFailed, Object=@n1) -> (then, CastFailed, AsEUW Asset Playlist Playlist Widget:EUW Asset Playlist Playlist Widget Object Reference) {is=title K2Node_DynamicCast}
   > PrintString(InString=Playlist)
   > reload(self:EUW Asset Playlist Playlist Widget Object Reference=@n3)
   n4 = Cast To EUW_AssetPlaylist_userWidget!(execute=@n3.CastFailed, Object=@n1) -> (then, CastFailed, AsEUW Asset Playlist User Widget:EUW Asset Playlist User Widget Object Reference) {is=title K2Node_DynamicCast}
   > PrintString(InString=User)
   > Reload(self:EUW Asset Playlist User Widget Object Reference=@n4)
  ]}
 spawnAssets {nodes [
   n1 = spawnAssets!() -> (then, position:Vector=~) {is=title K2Node_FunctionEntry}
   > Set currentPosition(currentPosition:Vector=@n1) -> (Output_Get:Vector)
   > n2 = get Groups()
   > n3 = For Each Loop(Array:Array of EUW Playlist Group Object References=@n2) -> (Array Element:EUW Playlist Group Object Reference)
   > n4 = get Items(self=@n3)
   > n5 = For Each Loop(Array:Array of EUW Playlist Item Object References=@n4) -> (Array Element:EUW Playlist Item Object Reference)
   n6 = Get item!(self:EUW Playlist Item Object Reference=@n5) -> (item:Fort Item Definition Object Reference)
   > Is Valid(InputObject=@n6)
   n7 = Get!(self:EUW Playlist Item Object Reference=@n5) -> (
    asset_shotGridId_27_#7
    asset_cosmeticType_20_#8
    asset_name_22_#9
    asset_gender_25_#10
    asset_assetGroup_41_#11
    asset_primaryAssetGroup_42_#12
    asset_status_33_#13
    asset_fixVersion_37_#14
   )
   > PrintString(InString=Append!(
     A=Enum to String(Enumerator=@n7.asset_cosmeticType_20_#8)
     B=" '"
     C=@n7.asset_assetGroup_41_#11
     D="' "
     E=@n7.asset_name_22_#9
     F:String=" "
     G:String=GetDisplayName(Object=@n6)
    ) -> (ReturnValue))
   > Branch(Condition="Equal (Enum)"(A=@n7.asset_cosmeticType_20_#8, B=NewEnumerator0))
   > n8 = SpawnActor Fort Player Mannequin!(
    execute:Exec
    WorldContextObject:Object Reference
    Class:Actor Class Reference=/Script/FortniteGame.FortPlayerMannequin
    SpawnTransform_Location:Vector=Get currentPosition() -> (currentPosition:Vector)
    SpawnTransform_Rotation:Rotator="0, 0, 0"
    SpawnTransform_Scale:Vector="1,1,1"
    CollisionHandlingOverride:ESpawnActorCollisionHandlingMethod Enum=Undefined
    TransformScaleMethod:ESpawnActorScaleMethod Enum=MultiplyWithRoot
    Owner:Actor Object Reference
    Instigator:Pawn Object Reference
   ) -> (then:Exec, ReturnValue:Fort Player Mannequin Object Reference) {is=title K2Node_SpawnActorFromClass}
   > Cast To AthenaCharacterItemDefinition!(execute, Object=@n6) -> (then, CastFailed, AsAthena Character Item Definition:Athena Character Item Definition Object Reference) {is=title K2Node_DynamicCast}
   > SetEditorProperty(Object=@n8, PropertyName=athenaCharacter, PropertyValue:Fort Item Definition Object Reference=@n6)
   > n9 = Branch(Condition=Get isVariant!(self:EUW Playlist Item Object Reference=@n5) -> (isVariant:bool))
   > n10 = get Override Variants!(execute, self:EUW Playlist Item Object Reference=@n5) -> (then, variants:Array of Mcp Variant Channel Info Structures)
   > n11 = SetEditorProperty(Object=@n8, PropertyName=overrideVariants, PropertyValue:Array of Mcp Variant Channel Info Structures=@n10)
   Set currentPosition(execute=[@n9.else, @n11.then], currentPosition=vector + vector(A=Get currentPosition(), B=Get spacing() -> (spacing:Vector)))
  ]}
]
variables [
 assetProgress {type=integer}
 csvFile {type=FilePath, default {filePath=C:/Users/MikeMulholland/Downloads/Asset.csv}}
 EUW_AssetPlaylistOptions {type=EUW_AssetPlaylistOptionsWidget_C, default=null}
 HorizontalBox_current {type=HorizontalBox, default=null}
 mouseEntered {type=boolean}
 numAssets {type=integer}
 Out Hit {
  type=HitResult
  default {
   bBlockingHit=false
   bStartPenetrating=false
   boneName=None
   component=null
   distance=0
   elementIndex=0
   faceIndex=0
   item=0
   myBoneName=None
   myItem=-1
   penetrationDepth=0
   physMaterial=null
   time=1
   hitObjectHandle {referenceObject=null}
   impactNormal=<0 0 0>
   impactPoint=<0 0 0>
   location=<0 0 0>
   normal=<0 0 0>
   traceEnd=<0 0 0>
   traceStart=<0 0 0>
  }
 }
 Out Hit_0 {
  type=HitResult
  default {
   bBlockingHit=false
   bStartPenetrating=false
   boneName=None
   component=null
   distance=0
   elementIndex=0
   faceIndex=0
   item=0
   myBoneName=None
   myItem=-1
   penetrationDepth=0
   physMaterial=null
   time=1
   hitObjectHandle {referenceObject=null}
   impactNormal=<0 0 0>
   impactPoint=<0 0 0>
   location=<0 0 0>
   normal=<0 0 0>
   traceEnd=<0 0 0>
   traceStart=<0 0 0>
  }
 }
 ProgressPercent {type=number, default=0.1}
 ProgressPercent_step {type=number, default=0.1}
 shotGridAssetList {type="array<AssetPlaylist_asset>", default []}
 Tab Widget {type=Widget, default=null}
]
widgets {
 info {rootWidgetClass=&CanvasPanel, widgetCount=30}
 widgets [
  CanvasPanel_0 {widgetClassPath=&CanvasPanel, widget=CanvasPanel_0}
  Overlay_122 {
   widgetClassPath=&Overlay
   parent=CanvasPanel_0
   widget=Overlay_122
   slot {class=CanvasPanelSlot, properties {layoutData {alignment=<0 0>, anchors {maximum=<1 1>, minimum=<0 0>}, offsets=<ltrb 0 0 0 0>}}}
  }
  HorizontalBox_184 {
   widgetClassPath=&HorizontalBox
   parent=Overlay_122
   widget=HorizontalBox_184
   slot {class=OverlaySlot, properties {horizontalAlignment=HAlign_Fill, verticalAlignment=VAlign_Fill}}
  }
  SizeBox_leftColumn {
   widgetClassPath=&SizeBox
   parent=HorizontalBox_184
   widget=SizeBox_leftColumn
   properties {bOverride_MaxDesiredWidth=true, bOverride_WidthOverride=true, maxDesiredWidth=200, widthOverride=260}
   slot {class=HorizontalBoxSlot}
  }
  VerticalBox_293 {widgetClassPath=&VerticalBox, parent=SizeBox_leftColumn, widget=VerticalBox_293, slot {class=SizeBoxSlot}}
  Generic01_TabContainer {
   widgetClassPath=/Game/Blueprints/Blutilities/ForTechArt/Widgets/TabContainer/Generic01_TabContainer
   bIsVariable=true
   parent=VerticalBox_293
   widget=Generic01_TabContainer
   properties {initialTabs {
     Options=&/EUW_AssetPlaylistOptionsWidget.EUW_AssetPlaylistOptionsWidget_C
     Playlist=&/EUW_AssetPlaylist_playlistWidget.EUW_AssetPlaylist_playlistWidget_C
     Release=&/EUW_AssetPlaylist_releaseWidget.EUW_AssetPlaylist_releaseWidget_C
     User=&/EUW_AssetPlaylist_userWidget.EUW_AssetPlaylist_userWidget_C
    }}
   slot {class=VerticalBoxSlot, properties {size {sizeRule=Fill, value=1}}}
  }
  CheckBox_assetGroups {
   widgetClassPath=/Script/UMG.CheckBox
   bIsVariable=true
   parent=VerticalBox_293
   widget=CheckBox_assetGroups
   properties {
    checkedState=Checked
    widgetStyle {
     checkBoxType=CheckBox
     backgroundHoveredImage {
      bIsDynamicallyLoaded=false
      drawAs=RoundedBox
      imageType=NoImage
      mirroring=NoMirror
      resourceName=None
      resourceObject=null
      tiling=NoTile
      imageSize=<18 18>
      margin=<ltrb 0 0 0 0>
      outlineSettings {
       bUseBrushTransparency=false
       roundingType=FixedRadius
       width=1
       color {colorUseRule=UseColor_ColorTable, specifiedColor=<rgba 0 0 0 0>}
       cornerRadii=<3 3 3 3>
      }
      tintColor {colorUseRule=UseColor_Specified, specifiedColor=<rgba 0.0048 0.0048 0.0048 1>}
      uVRegion {bIsValid=false, max=<0 0>, min=<0 0>}
     }
     backgroundImage {
      bIsDynamicallyLoaded=false
      drawAs=RoundedBox
      imageType=NoImage
      mirroring=NoMirror
      resourceName=None
      resourceObject=null
      tiling=NoTile
      imageSize=<18 18>
      margin=<ltrb 0 0 0 0>
      outlineSettings {
       bUseBrushTransparency=false
       roundingType=FixedRadius
       width=1
       color {colorUseRule=UseColor_ColorTable, specifiedColor=<rgba 0 0 0 0>}
       cornerRadii=<3 3 3 3>
      }
      tintColor {colorUseRule=UseColor_Specified, specifiedColor=<rgba 0.0048 0.0048 0.0048 1>}
      uVRegion {bIsValid=false, max=<0 0>, min=<0 0>}
     }
     backgroundPressedImage {
      bIsDynamicallyLoaded=false
      drawAs=RoundedBox
      imageType=NoImage
      mirroring=NoMirror
      resourceName=None
      resourceObject=null
      tiling=NoTile
      imageSize=<18 18>
      margin=<ltrb 0 0 0 0>
      outlineSettings {
       bUseBrushTransparency=false
       roundingType=FixedRadius
       width=1
       color {colorUseRule=UseColor_ColorTable, specifiedColor=<rgba 0 0 0 0>}
       cornerRadii=<3 3 3 3>
      }
      tintColor {colorUseRule=UseColor_Specified, specifiedColor=<rgba 0.0048 0.0048 0.0048 1>}
      uVRegion {bIsValid=false, max=<0 0>, min=<0 0>}
     }
     borderBackgroundColor {colorUseRule=UseColor_Specified, specifiedColor=<rgba 1 1 1 1>}
     checkedForeground {colorUseRule=UseColor_Specified, specifiedColor=<rgba 1 1 1 1>}
     checkedHoveredForeground {colorUseRule=UseColor_Specified, specifiedColor=<rgba 1 1 1 1>}
     checkedHoveredImage {
      bIsDynamicallyLoaded=false
      drawAs=Image
      imageType=Vector
      mirroring=NoMirror
      resourceName=../../../Engine/Content/Slate/Starship/CoreWidgets/CheckBox/check.svg
      resourceObject=null
      tiling=NoTile
      imageSize=<16 16>
      margin=<ltrb 0 0 0 0>
      outlineSettings {
       bUseBrushTransparency=false
       roundingType=HalfHeightRadius
       width=0
       color {colorUseRule=UseColor_Specified, specifiedColor=<rgba 0 0 0 0>}
       cornerRadii=<0 0 0 0>
      }
      tintColor {colorUseRule=UseColor_Specified, specifiedColor=<rgba 0.0044 0.2384 1 1>}
      uVRegion {bIsValid=false, max=<0 0>, min=<0 0>}
     }
     checkedImage {
      bIsDynamicallyLoaded=false
      drawAs=Image
      imageType=Vector
      mirroring=NoMirror
      resourceName=../../../Engine/Content/Slate/Starship/CoreWidgets/CheckBox/check.svg
      resourceObject=null
      tiling=NoTile
      imageSize=<16 16>
      margin=<ltrb 0 0 0 0>
      outlineSettings {
       bUseBrushTransparency=false
       roundingType=HalfHeightRadius
       width=0
       color {colorUseRule=UseColor_Specified, specifiedColor=<rgba 0 0 0 0>}
       cornerRadii=<0 0 0 0>
      }
      tintColor {colorUseRule=UseColor_Specified, specifiedColor=<rgba 0 0.162 0.7454 1>}
      uVRegion {bIsValid=false, max=<0 0>, min=<0 0>}
     }
     checkedPressedForeground {colorUseRule=UseColor_Specified, specifiedColor=<rgba 1 1 1 1>}
     checkedPressedImage {
      bIsDynamicallyLoaded=false
      drawAs=Image
      imageType=Vector
      mirroring=NoMirror
      resourceName=../../../Engine/Content/Slate/Starship/CoreWidgets/CheckBox/check.svg
      resourceObject=null
      tiling=NoTile
      imageSize=<16 16>
      margin=<ltrb 0 0 0 0>
      outlineSettings {
       bUseBrushTransparency=false
       roundingType=HalfHeightRadius
       width=0
       color {colorUseRule=UseColor_Specified, specifiedColor=<rgba 0 0 0 0>}
       cornerRadii=<0 0 0 0>
      }
      tintColor {colorUseRule=UseColor_Specified, specifiedColor=<rgba 0 0.0802 0.3515 1>}
      uVRegion {bIsValid=false, max=<0 0>, min=<0 0>}
     }
     checkedSlateSound {resourceObject=null}
     foregroundColor {colorUseRule=UseColor_Specified, specifiedColor=<rgba 1 1 1 1>}
     hoveredForeground {colorUseRule=UseColor_Specified, specifiedColor=<rgba 1 1 1 1>}
     hoveredSlateSound {resourceObject=null}
     padding=<ltrb 2 0 0 0>
     pressedForeground {colorUseRule=UseColor_Specified, specifiedColor=<rgba 1 1 1 1>}
     uncheckedHoveredImage {
      bIsDynamicallyLoaded=false
      drawAs=NoDrawType
      imageType=NoImage
      mirroring=NoMirror
      resourceName=None
      resourceObject=null
      tiling=NoTile
      imageSize=<0 0>
      margin=<ltrb 0 0 0 0>
      outlineSettings {
       bUseBrushTransparency=false
       roundingType=HalfHeightRadius
       width=0
       color {colorUseRule=UseColor_Specified, specifiedColor=<rgba 0 0 0 0>}
       cornerRadii=<0 0 0 0>
      }
      tintColor {colorUseRule=UseColor_Specified, specifiedColor=<rgba 1 1 1 1>}
      uVRegion {bIsValid=false, max=<0 0>, min=<0 0>}
     }
     uncheckedImage {
      bIsDynamicallyLoaded=false
      drawAs=NoDrawType
      imageType=NoImage
      mirroring=NoMirror
      resourceName=None
      resourceObject=null
      tiling=NoTile
      imageSize=<0 0>
      margin=<ltrb 0 0 0 0>
      outlineSettings {
       bUseBrushTransparency=false
       roundingType=HalfHeightRadius
       width=0
       color {colorUseRule=UseColor_Specified, specifiedColor=<rgba 0 0 0 0>}
       cornerRadii=<0 0 0 0>
      }
      tintColor {colorUseRule=UseColor_Specified, specifiedColor=<rgba 1 1 1 1>}
      uVRegion {bIsValid=false, max=<0 0>, min=<0 0>}
     }
     uncheckedPressedImage {
      bIsDynamicallyLoaded=false
      drawAs=NoDrawType
      imageType=NoImage
      mirroring=NoMirror
      resourceName=None
      resourceObject=null
      tiling=NoTile
      imageSize=<0 0>
      margin=<ltrb 0 0 0 0>
      outlineSettings {
       bUseBrushTransparency=false
       roundingType=HalfHeightRadius
       width=0
       color {colorUseRule=UseColor_Specified, specifiedColor=<rgba 0 0 0 0>}
       cornerRadii=<0 0 0 0>
      }
      tintColor {colorUseRule=UseColor_Specified, specifiedColor=<rgba 1 1 1 1>}
      uVRegion {bIsValid=false, max=<0 0>, min=<0 0>}
     }
     uncheckedSlateSound {resourceObject=null}
     undeterminedForeground {colorUseRule=UseColor_Specified, specifiedColor=<rgba 1 1 1 1>}
     undeterminedHoveredImage {
      bIsDynamicallyLoaded=false
      drawAs=Image
      imageType=Vector
      mirroring=NoMirror
      resourceName=../../../Engine/Content/Slate/Starship/CoreWidgets/CheckBox/indeterminate.svg
      resourceObject=null
      tiling=NoTile
      imageSize=<16 16>
      margin=<ltrb 0 0 0 0>
      outlineSettings {
       bUseBrushTransparency=false
       roundingType=HalfHeightRadius
       width=0
       color {colorUseRule=UseColor_Specified, specifiedColor=<rgba 0 0 0 0>}
       cornerRadii=<0 0 0 0>
      }
      tintColor {colorUseRule=UseColor_Specified, specifiedColor=<rgba 0.0044 0.2384 1 1>}
      uVRegion {bIsValid=false, max=<0 0>, min=<0 0>}
     }
     undeterminedImage {
      bIsDynamicallyLoaded=false
      drawAs=Image
      imageType=Vector
      mirroring=NoMirror
      resourceName=../../../Engine/Content/Slate/Starship/CoreWidgets/CheckBox/indeterminate.svg
      resourceObject=null
      tiling=NoTile
      imageSize=<16 16>
      margin=<ltrb 0 0 0 0>
      outlineSettings {
       bUseBrushTransparency=false
       roundingType=HalfHeightRadius
       width=0
       color {colorUseRule=UseColor_Specified, specifiedColor=<rgba 0 0 0 0>}
       cornerRadii=<0 0 0 0>
      }
      tintColor {colorUseRule=UseColor_Specified, specifiedColor=<rgba 0 0.162 0.7454 1>}
      uVRegion {bIsValid=false, max=<0 0>, min=<0 0>}
     }
     undeterminedPressedImage {
      bIsDynamicallyLoaded=false
      drawAs=Image
      imageType=Vector
      mirroring=NoMirror
      resourceName=../../../Engine/Content/Slate/Starship/CoreWidgets/CheckBox/indeterminate.svg
      resourceObject=null
      tiling=NoTile
      imageSize=<16 16>
      margin=<ltrb 0 0 0 0>
      outlineSettings {
       bUseBrushTransparency=false
       roundingType=HalfHeightRadius
       width=0
       color {colorUseRule=UseColor_Specified, specifiedColor=<rgba 0 0 0 0>}
       cornerRadii=<0 0 0 0>
      }
      tintColor {colorUseRule=UseColor_Specified, specifiedColor=<rgba 0 0.0802 0.3515 1>}
      uVRegion {bIsValid=false, max=<0 0>, min=<0 0>}
     }
    }
   }
   slot {class=VerticalBoxSlot}
  }
  CommonTextBlock_useAssetGroups {
   widgetClassPath=&CommonTextBlock
   bIsVariable=true
   parent=CheckBox_assetGroups
   widget=CommonTextBlock_useAssetGroups
   properties {
    justification=Right
    text=Use AssetGroups
    font {
     bForceMonospaced=false
     bMaterialIsStencil=false
     fontMaterial=null
     fontObject=&NotoSans
     letterSpacing=0
     monospacedWidth=1
     size=11
     skewAmount=0
     typefaceFontName=NotoSans-Regular
     outlineSettings {
      bApplyOutlineToDropShadows=false
      bMiteredCorners=false
      bSeparateFillAlpha=false
      outlineMaterial=null
      outlineSize=0
      outlineColor=<rgba 0 0 0 1>
     }
    }
    renderTransform {angle=0, scale=<1 1>, shear=<0 0>, translation=<-3 0>}
    shadowOffset=<0 0>
   }
   slot {class=PanelSlot}
  }
  CheckBox_cosmeticPicker {
   widgetClassPath=&CheckBox
   bIsVariable=true
   parent=VerticalBox_293
   widget=CheckBox_cosmeticPicker
   properties {
    checkedState=Checked
    widgetStyle {
     checkBoxType=CheckBox
     backgroundHoveredImage {
      bIsDynamicallyLoaded=false
      drawAs=RoundedBox
      imageType=NoImage
      mirroring=NoMirror
      resourceName=None
      resourceObject=null
      tiling=NoTile
      imageSize=<18 18>
      margin=<ltrb 0 0 0 0>
      outlineSettings {
       bUseBrushTransparency=false
       roundingType=FixedRadius
       width=1
       color {colorUseRule=UseColor_ColorTable, specifiedColor=<rgba 0 0 0 0>}
       cornerRadii=<3 3 3 3>
      }
      tintColor {colorUseRule=UseColor_Specified, specifiedColor=<rgba 0.0048 0.0048 0.0048 1>}
      uVRegion {bIsValid=false, max=<0 0>, min=<0 0>}
     }
     backgroundImage {
      bIsDynamicallyLoaded=false
      drawAs=RoundedBox
      imageType=NoImage
      mirroring=NoMirror
      resourceName=None
      resourceObject=null
      tiling=NoTile
      imageSize=<18 18>
      margin=<ltrb 0 0 0 0>
      outlineSettings {
       bUseBrushTransparency=false
       roundingType=FixedRadius
       width=1
       color {colorUseRule=UseColor_ColorTable, specifiedColor=<rgba 0 0 0 0>}
       cornerRadii=<3 3 3 3>
      }
      tintColor {colorUseRule=UseColor_Specified, specifiedColor=<rgba 0.0048 0.0048 0.0048 1>}
      uVRegion {bIsValid=false, max=<0 0>, min=<0 0>}
     }
     backgroundPressedImage {
      bIsDynamicallyLoaded=false
      drawAs=RoundedBox
      imageType=NoImage
      mirroring=NoMirror
      resourceName=None
      resourceObject=null
      tiling=NoTile
      imageSize=<18 18>
      margin=<ltrb 0 0 0 0>
      outlineSettings {
       bUseBrushTransparency=false
       roundingType=FixedRadius
       width=1
       color {colorUseRule=UseColor_ColorTable, specifiedColor=<rgba 0 0 0 0>}
       cornerRadii=<3 3 3 3>
      }
      tintColor {colorUseRule=UseColor_Specified, specifiedColor=<rgba 0.0048 0.0048 0.0048 1>}
      uVRegion {bIsValid=false, max=<0 0>, min=<0 0>}
     }
     borderBackgroundColor {colorUseRule=UseColor_Specified, specifiedColor=<rgba 1 1 1 1>}
     checkedForeground {colorUseRule=UseColor_Specified, specifiedColor=<rgba 1 1 1 1>}
     checkedHoveredForeground {colorUseRule=UseColor_Specified, specifiedColor=<rgba 1 1 1 1>}
     checkedHoveredImage {
      bIsDynamicallyLoaded=false
      drawAs=Image
      imageType=Vector
      mirroring=NoMirror
      resourceName=../../../Engine/Content/Slate/Starship/CoreWidgets/CheckBox/check.svg
      resourceObject=null
      tiling=NoTile
      imageSize=<16 16>
      margin=<ltrb 0 0 0 0>
      outlineSettings {
       bUseBrushTransparency=false
       roundingType=HalfHeightRadius
       width=0
       color {colorUseRule=UseColor_Specified, specifiedColor=<rgba 0 0 0 0>}
       cornerRadii=<0 0 0 0>
      }
      tintColor {colorUseRule=UseColor_Specified, specifiedColor=<rgba 0.0044 0.2384 1 1>}
      uVRegion {bIsValid=false, max=<0 0>, min=<0 0>}
     }
     checkedImage {
      bIsDynamicallyLoaded=false
      drawAs=Image
      imageType=Vector
      mirroring=NoMirror
      resourceName=../../../Engine/Content/Slate/Starship/CoreWidgets/CheckBox/check.svg
      resourceObject=null
      tiling=NoTile
      imageSize=<16 16>
      margin=<ltrb 0 0 0 0>
      outlineSettings {
       bUseBrushTransparency=false
       roundingType=HalfHeightRadius
       width=0
       color {colorUseRule=UseColor_Specified, specifiedColor=<rgba 0 0 0 0>}
       cornerRadii=<0 0 0 0>
      }
      tintColor {colorUseRule=UseColor_Specified, specifiedColor=<rgba 0 0.162 0.7454 1>}
      uVRegion {bIsValid=false, max=<0 0>, min=<0 0>}
     }
     checkedPressedForeground {colorUseRule=UseColor_Specified, specifiedColor=<rgba 1 1 1 1>}
     checkedPressedImage {
      bIsDynamicallyLoaded=false
      drawAs=Image
      imageType=Vector
      mirroring=NoMirror
      resourceName=../../../Engine/Content/Slate/Starship/CoreWidgets/CheckBox/check.svg
      resourceObject=null
      tiling=NoTile
      imageSize=<16 16>
      margin=<ltrb 0 0 0 0>
      outlineSettings {
       bUseBrushTransparency=false
       roundingType=HalfHeightRadius
       width=0
       color {colorUseRule=UseColor_Specified, specifiedColor=<rgba 0 0 0 0>}
       cornerRadii=<0 0 0 0>
      }
      tintColor {colorUseRule=UseColor_Specified, specifiedColor=<rgba 0 0.0802 0.3515 1>}
      uVRegion {bIsValid=false, max=<0 0>, min=<0 0>}
     }
     checkedSlateSound {resourceObject=null}
     foregroundColor {colorUseRule=UseColor_Specified, specifiedColor=<rgba 1 1 1 1>}
     hoveredForeground {colorUseRule=UseColor_Specified, specifiedColor=<rgba 1 1 1 1>}
     hoveredSlateSound {resourceObject=null}
     padding=<ltrb 2 0 0 0>
     pressedForeground {colorUseRule=UseColor_Specified, specifiedColor=<rgba 1 1 1 1>}
     uncheckedHoveredImage {
      bIsDynamicallyLoaded=false
      drawAs=NoDrawType
      imageType=NoImage
      mirroring=NoMirror
      resourceName=None
      resourceObject=null
      tiling=NoTile
      imageSize=<0 0>
      margin=<ltrb 0 0 0 0>
      outlineSettings {
       bUseBrushTransparency=false
       roundingType=HalfHeightRadius
       width=0
       color {colorUseRule=UseColor_Specified, specifiedColor=<rgba 0 0 0 0>}
       cornerRadii=<0 0 0 0>
      }
      tintColor {colorUseRule=UseColor_Specified, specifiedColor=<rgba 1 1 1 1>}
      uVRegion {bIsValid=false, max=<0 0>, min=<0 0>}
     }
     uncheckedImage {
      bIsDynamicallyLoaded=false
      drawAs=NoDrawType
      imageType=NoImage
      mirroring=NoMirror
      resourceName=None
      resourceObject=null
      tiling=NoTile
      imageSize=<0 0>
      margin=<ltrb 0 0 0 0>
      outlineSettings {
       bUseBrushTransparency=false
       roundingType=HalfHeightRadius
       width=0
       color {colorUseRule=UseColor_Specified, specifiedColor=<rgba 0 0 0 0>}
       cornerRadii=<0 0 0 0>
      }
      tintColor {colorUseRule=UseColor_Specified, specifiedColor=<rgba 1 1 1 1>}
      uVRegion {bIsValid=false, max=<0 0>, min=<0 0>}
     }
     uncheckedPressedImage {
      bIsDynamicallyLoaded=false
      drawAs=NoDrawType
      imageType=NoImage
      mirroring=NoMirror
      resourceName=None
      resourceObject=null
      tiling=NoTile
      imageSize=<0 0>
      margin=<ltrb 0 0 0 0>
      outlineSettings {
       bUseBrushTransparency=false
       roundingType=HalfHeightRadius
       width=0
       color {colorUseRule=UseColor_Specified, specifiedColor=<rgba 0 0 0 0>}
       cornerRadii=<0 0 0 0>
      }
      tintColor {colorUseRule=UseColor_Specified, specifiedColor=<rgba 1 1 1 1>}
      uVRegion {bIsValid=false, max=<0 0>, min=<0 0>}
     }
     uncheckedSlateSound {resourceObject=null}
     undeterminedForeground {colorUseRule=UseColor_Specified, specifiedColor=<rgba 1 1 1 1>}
     undeterminedHoveredImage {
      bIsDynamicallyLoaded=false
      drawAs=Image
      imageType=Vector
      mirroring=NoMirror
      resourceName=../../../Engine/Content/Slate/Starship/CoreWidgets/CheckBox/indeterminate.svg
      resourceObject=null
      tiling=NoTile
      imageSize=<16 16>
      margin=<ltrb 0 0 0 0>
      outlineSettings {
       bUseBrushTransparency=false
       roundingType=HalfHeightRadius
       width=0
       color {colorUseRule=UseColor_Specified, specifiedColor=<rgba 0 0 0 0>}
       cornerRadii=<0 0 0 0>
      }
      tintColor {colorUseRule=UseColor_Specified, specifiedColor=<rgba 0.0044 0.2384 1 1>}
      uVRegion {bIsValid=false, max=<0 0>, min=<0 0>}
     }
     undeterminedImage {
      bIsDynamicallyLoaded=false
      drawAs=Image
      imageType=Vector
      mirroring=NoMirror
      resourceName=../../../Engine/Content/Slate/Starship/CoreWidgets/CheckBox/indeterminate.svg
      resourceObject=null
      tiling=NoTile
      imageSize=<16 16>
      margin=<ltrb 0 0 0 0>
      outlineSettings {
       bUseBrushTransparency=false
       roundingType=HalfHeightRadius
       width=0
       color {colorUseRule=UseColor_Specified, specifiedColor=<rgba 0 0 0 0>}
       cornerRadii=<0 0 0 0>
      }
      tintColor {colorUseRule=UseColor_Specified, specifiedColor=<rgba 0 0.162 0.7454 1>}
      uVRegion {bIsValid=false, max=<0 0>, min=<0 0>}
     }
     undeterminedPressedImage {
      bIsDynamicallyLoaded=false
      drawAs=Image
      imageType=Vector
      mirroring=NoMirror
      resourceName=../../../Engine/Content/Slate/Starship/CoreWidgets/CheckBox/indeterminate.svg
      resourceObject=null
      tiling=NoTile
      imageSize=<16 16>
      margin=<ltrb 0 0 0 0>
      outlineSettings {
       bUseBrushTransparency=false
       roundingType=HalfHeightRadius
       width=0
       color {colorUseRule=UseColor_Specified, specifiedColor=<rgba 0 0 0 0>}
       cornerRadii=<0 0 0 0>
      }
      tintColor {colorUseRule=UseColor_Specified, specifiedColor=<rgba 0 0.0802 0.3515 1>}
      uVRegion {bIsValid=false, max=<0 0>, min=<0 0>}
     }
    }
   }
   slot {class=VerticalBoxSlot}
  }
  CommonTextBlock_cosmeticPicker {
   widgetClassPath=&CommonTextBlock
   bIsVariable=true
   parent=CheckBox_cosmeticPicker
   widget=CommonTextBlock_cosmeticPicker
   properties {
    justification=Right
    text=Send to CosmeticPicker
    font {
     bForceMonospaced=false
     bMaterialIsStencil=false
     fontMaterial=null
     fontObject=&NotoSans
     letterSpacing=0
     monospacedWidth=1
     size=11
     skewAmount=0
     typefaceFontName=NotoSans-Regular
     outlineSettings {
      bApplyOutlineToDropShadows=false
      bMiteredCorners=false
      bSeparateFillAlpha=false
      outlineMaterial=null
      outlineSize=0
      outlineColor=<rgba 0 0 0 1>
     }
    }
    renderTransform {angle=0, scale=<1 1>, shear=<0 0>, translation=<-3 0>}
    shadowOffset=<0 0>
   }
   slot {class=PanelSlot}
  }
  Overlay_46 {
   widgetClassPath=&Overlay
   parent=HorizontalBox_184
   widget=Overlay_46
   slot {class=HorizontalBoxSlot, properties {size {sizeRule=Fill, value=1}}}
  }
  Image_95 {
   widgetClassPath=&Image
   parent=Overlay_46
   widget=Image_95
   properties {brush {
     bIsDynamicallyLoaded=false
     drawAs=Box
     imageType=NoImage
     mirroring=NoMirror
     resourceName=None
     resourceObject=null
     tiling=Both
     imageSize=<100 100>
     margin=<ltrb 0 0 0 0>
     outlineSettings {
      bUseBrushTransparency=false
      roundingType=HalfHeightRadius
      width=0
      color {colorUseRule=UseColor_Specified, specifiedColor=<rgba 0 0 0 0>}
      cornerRadii=<0 0 0 0>
     }
     tintColor {colorUseRule=UseColor_Specified, specifiedColor=<rgba 0.005 0.005 0.005 1>}
     uVRegion {bIsValid=false, max=<0 0>, min=<0 0>}
    }}
   slot {class=OverlaySlot, properties {horizontalAlignment=HAlign_Fill, verticalAlignment=VAlign_Fill}}
  }
  ScrollBox_playlist {
   widgetClassPath=&ScrollBox
   bIsVariable=true
   parent=Overlay_46
   widget=ScrollBox_playlist
   properties {
    allowOverscroll=false
    consumeMouseWheel=Never
    orientation=Orient_Horizontal
    widgetBarStyle {
     thickness=8
     draggedThumbImage {
      bIsDynamicallyLoaded=false
      drawAs=RoundedBox
      imageType=NoImage
      mirroring=NoMirror
      resourceName=None
      resourceObject=null
      tiling=NoTile
      imageSize=<0 0>
      margin=<ltrb 0 0 0 0>
      outlineSettings {
       bUseBrushTransparency=false
       roundingType=FixedRadius
       width=0
       color {colorUseRule=UseColor_Specified, specifiedColor=<rgba 0 0 0 0>}
       cornerRadii=<4 4 4 4>
      }
      tintColor {colorUseRule=UseColor_Specified, specifiedColor=<rgba 0.2159 0.2159 0.2159 1>}
      uVRegion {bIsValid=false, max=<0 0>, min=<0 0>}
     }
     horizontalBackgroundImage {
      bIsDynamicallyLoaded=false
      drawAs=NoDrawType
      imageType=NoImage
      mirroring=NoMirror
      resourceName=None
      resourceObject=null
      tiling=NoTile
      imageSize=<32 32>
      margin=<ltrb 0 0 0 0>
      outlineSettings {
       bUseBrushTransparency=false
       roundingType=HalfHeightRadius
       width=0
       color {colorUseRule=UseColor_Specified, specifiedColor=<rgba 0 0 0 0>}
       cornerRadii=<0 0 0 0>
      }
      tintColor {colorUseRule=UseColor_Specified, specifiedColor=<rgba 1 1 1 1>}
      uVRegion {bIsValid=false, max=<0 0>, min=<0 0>}
     }
     horizontalBottomSlotImage {
      bIsDynamicallyLoaded=false
      drawAs=NoDrawType
      imageType=NoImage
      mirroring=NoMirror
      resourceName=None
      resourceObject=null
      tiling=NoTile
      imageSize=<0 0>
      margin=<ltrb 0 0 0 0>
      outlineSettings {
       bUseBrushTransparency=false
       roundingType=HalfHeightRadius
       width=0
       color {colorUseRule=UseColor_Specified, specifiedColor=<rgba 0 0 0 0>}
       cornerRadii=<0 0 0 0>
      }
      tintColor {colorUseRule=UseColor_Specified, specifiedColor=<rgba 1 1 1 1>}
      uVRegion {bIsValid=false, max=<0 0>, min=<0 0>}
     }
     horizontalTopSlotImage {
      bIsDynamicallyLoaded=false
      drawAs=NoDrawType
      imageType=NoImage
      mirroring=NoMirror
      resourceName=None
      resourceObject=null
      tiling=NoTile
      imageSize=<0 0>
      margin=<ltrb 0 0 0 0>
      outlineSettings {
       bUseBrushTransparency=false
       roundingType=HalfHeightRadius
       width=0
       color {colorUseRule=UseColor_Specified, specifiedColor=<rgba 0 0 0 0>}
       cornerRadii=<0 0 0 0>
      }
      tintColor {colorUseRule=UseColor_Specified, specifiedColor=<rgba 1 1 1 1>}
      uVRegion {bIsValid=false, max=<0 0>, min=<0 0>}
     }
     hoveredThumbImage {
      bIsDynamicallyLoaded=false
      drawAs=RoundedBox
      imageType=NoImage
      mirroring=NoMirror
      resourceName=None
      resourceObject=null
      tiling=NoTile
      imageSize=<0 0>
      margin=<ltrb 0 0 0 0>
      outlineSettings {
       bUseBrushTransparency=false
       roundingType=FixedRadius
       width=0
       color {colorUseRule=UseColor_Specified, specifiedColor=<rgba 0 0 0 0>}
       cornerRadii=<4 4 4 4>
      }
      tintColor {colorUseRule=UseColor_Specified, specifiedColor=<rgba 0.2159 0.2159 0.2159 1>}
      uVRegion {bIsValid=false, max=<0 0>, min=<0 0>}
     }
     normalThumbImage {
      bIsDynamicallyLoaded=false
      drawAs=RoundedBox
      imageType=NoImage
      mirroring=NoMirror
      resourceName=None
      resourceObject=null
      tiling=NoTile
      imageSize=<0 0>
      margin=<ltrb 0 0 0 0>
      outlineSettings {
       bUseBrushTransparency=false
       roundingType=FixedRadius
       width=0
       color {colorUseRule=UseColor_Specified, specifiedColor=<rgba 0 0 0 0>}
       cornerRadii=<4 4 4 4>
      }
      tintColor {colorUseRule=UseColor_Specified, specifiedColor=<rgba 0.0953 0.0953 0.0953 1>}
      uVRegion {bIsValid=false, max=<0 0>, min=<0 0>}
     }
     verticalBackgroundImage {
      bIsDynamicallyLoaded=false
      drawAs=NoDrawType
      imageType=NoImage
      mirroring=NoMirror
      resourceName=None
      resourceObject=null
      tiling=NoTile
      imageSize=<0 0>
      margin=<ltrb 0 0 0 0>
      outlineSettings {
       bUseBrushTransparency=false
       roundingType=HalfHeightRadius
       width=0
       color {colorUseRule=UseColor_Specified, specifiedColor=<rgba 0 0 0 0>}
       cornerRadii=<0 0 0 0>
      }
      tintColor {colorUseRule=UseColor_Specified, specifiedColor=<rgba 1 1 1 1>}
      uVRegion {bIsValid=false, max=<0 0>, min=<0 0>}
     }
     verticalBottomSlotImage {
      bIsDynamicallyLoaded=false
      drawAs=NoDrawType
      imageType=NoImage
      mirroring=NoMirror
      resourceName=None
      resourceObject=null
      tiling=NoTile
      imageSize=<0 0>
      margin=<ltrb 0 0 0 0>
      outlineSettings {
       bUseBrushTransparency=false
       roundingType=HalfHeightRadius
       width=0
       color {colorUseRule=UseColor_Specified, specifiedColor=<rgba 0 0 0 0>}
       cornerRadii=<0 0 0 0>
      }
      tintColor {colorUseRule=UseColor_Specified, specifiedColor=<rgba 1 1 1 1>}
      uVRegion {bIsValid=false, max=<0 0>, min=<0 0>}
     }
     verticalTopSlotImage {
      bIsDynamicallyLoaded=false
      drawAs=NoDrawType
      imageType=NoImage
      mirroring=NoMirror
      resourceName=None
      resourceObject=null
      tiling=NoTile
      imageSize=<0 0>
      margin=<ltrb 0 0 0 0>
      outlineSettings {
       bUseBrushTransparency=false
       roundingType=HalfHeightRadius
       width=0
       color {colorUseRule=UseColor_Specified, specifiedColor=<rgba 0 0 0 0>}
       cornerRadii=<0 0 0 0>
      }
      tintColor {colorUseRule=UseColor_Specified, specifiedColor=<rgba 1 1 1 1>}
      uVRegion {bIsValid=false, max=<0 0>, min=<0 0>}
     }
    }
    widgetStyle {
     barThickness=8
     bottomShadowBrush {
      bIsDynamicallyLoaded=false
      drawAs=Box
      imageType=FullColor
      mirroring=NoMirror
      resourceName=../../../Engine/Content/Slate/Common/ScrollBoxShadowBottom.png
      resourceObject=null
      tiling=NoTile
      imageSize=<16 8>
      margin=<ltrb 0.5 0 0.5 1>
      outlineSettings {
       bUseBrushTransparency=false
       roundingType=HalfHeightRadius
       width=0
       color {colorUseRule=UseColor_Specified, specifiedColor=<rgba 0 0 0 0>}
       cornerRadii=<0 0 0 0>
      }
      tintColor {colorUseRule=UseColor_Specified, specifiedColor=<rgba 1 1 1 1>}
      uVRegion {bIsValid=false, max=<0 0>, min=<0 0>}
     }
     horizontalScrolledContentPadding=<ltrb 0 0 1 0>
     leftShadowBrush {
      bIsDynamicallyLoaded=false
      drawAs=Box
      imageType=FullColor
      mirroring=NoMirror
      resourceName=../../../Engine/Content/Slate/Common/ScrollBoxShadowLeft.png
      resourceObject=null
      tiling=NoTile
      imageSize=<8 16>
      margin=<ltrb 1 0.5 0 0.5>
      outlineSettings {
       bUseBrushTransparency=false
       roundingType=HalfHeightRadius
       width=0
       color {colorUseRule=UseColor_Specified, specifiedColor=<rgba 0 0 0 0>}
       cornerRadii=<0 0 0 0>
      }
      tintColor {colorUseRule=UseColor_Specified, specifiedColor=<rgba 1 1 1 1>}
      uVRegion {bIsValid=false, max=<0 0>, min=<0 0>}
     }
     rightShadowBrush {
      bIsDynamicallyLoaded=false
      drawAs=Box
      imageType=FullColor
      mirroring=NoMirror
      resourceName=../../../Engine/Content/Slate/Common/ScrollBoxShadowRight.png
      resourceObject=null
      tiling=NoTile
      imageSize=<8 16>
      margin=<ltrb 0 0.5 1 0.5>
      outlineSettings {
       bUseBrushTransparency=false
       roundingType=HalfHeightRadius
       width=0
       color {colorUseRule=UseColor_Specified, specifiedColor=<rgba 0 0 0 0>}
       cornerRadii=<0 0 0 0>
      }
      tintColor {colorUseRule=UseColor_Specified, specifiedColor=<rgba 1 1 1 1>}
      uVRegion {bIsValid=false, max=<0 0>, min=<0 0>}
     }
     topShadowBrush {
      bIsDynamicallyLoaded=false
      drawAs=Box
      imageType=FullColor
      mirroring=NoMirror
      resourceName=../../../Engine/Content/Slate/Common/ScrollBoxShadowTop.png
      resourceObject=null
      tiling=NoTile
      imageSize=<16 8>
      margin=<ltrb 0.5 1 0.5 0>
      outlineSettings {
       bUseBrushTransparency=false
       roundingType=HalfHeightRadius
       width=0
       color {colorUseRule=UseColor_Specified, specifiedColor=<rgba 0 0 0 0>}
       cornerRadii=<0 0 0 0>
      }
      tintColor {colorUseRule=UseColor_Specified, specifiedColor=<rgba 1 1 1 1>}
      uVRegion {bIsValid=false, max=<0 0>, min=<0 0>}
     }
     verticalScrolledContentPadding=<ltrb 0 0 0 1>
    }
   }
   slot {class=OverlaySlot, properties {horizontalAlignment=HAlign_Fill, verticalAlignment=VAlign_Fill}}
  }
  Overlay_98 {widgetClassPath=&Overlay, parent=ScrollBox_playlist, widget=Overlay_98, slot {class=ScrollBoxSlot}}
  Image_background {
   widgetClassPath=&Image
   parent=Overlay_98
   widget=Image_background
   properties {
    renderOpacity=0.25
    visibility=Hidden
    brush {
     bIsDynamicallyLoaded=false
     drawAs=Image
     imageType=NoImage
     mirroring=NoMirror
     resourceName=None
     resourceObject=&/T_assetPlaylist_backgroundLarge
     tiling=Both
     imageSize=<400 133>
     margin=<ltrb 0 0 0 0>
     outlineSettings {
      bUseBrushTransparency=false
      roundingType=HalfHeightRadius
      width=0
      color {colorUseRule=UseColor_Specified, specifiedColor=<rgba 0 0 0 0>}
      cornerRadii=<0 0 0 0>
     }
     tintColor {colorUseRule=UseColor_Specified, specifiedColor=<rgba 1 1 1 1>}
     uVRegion {bIsValid=false, max=<0 0>, min=<0 0>}
    }
   }
   slot {class=OverlaySlot, properties {horizontalAlignment=HAlign_Fill, verticalAlignment=VAlign_Fill}}
  }
  HorizontalBox_releaseColumns {
   widgetClassPath=&HorizontalBox
   bIsVariable=true
   parent=Overlay_98
   widget=HorizontalBox_releaseColumns
   properties {visibility=Visible}
   slot {class=OverlaySlot, properties {horizontalAlignment=HAlign_Fill, verticalAlignment=VAlign_Fill}}
  }
  HorizontalBox_playlistColumns {
   widgetClassPath=&HorizontalBox
   bIsVariable=true
   parent=Overlay_98
   widget=HorizontalBox_playlistColumns
   properties {visibility=Visible}
   slot {class=OverlaySlot, properties {horizontalAlignment=HAlign_Fill, verticalAlignment=VAlign_Fill}}
  }
  HorizontalBox_userColumns {
   widgetClassPath=&HorizontalBox
   bIsVariable=true
   parent=Overlay_98
   widget=HorizontalBox_userColumns
   properties {visibility=Visible}
   slot {class=OverlaySlot, properties {horizontalAlignment=HAlign_Fill, verticalAlignment=VAlign_Fill}}
  }
  Overlay_processing {
   widgetClassPath=&Overlay
   bIsVariable=true
   parent=Overlay_46
   widget=Overlay_processing
   properties {visibility=Collapsed}
   slot {class=OverlaySlot, properties {horizontalAlignment=HAlign_Fill, verticalAlignment=VAlign_Fill}}
  }
  SizeBox_48 {
   widgetClassPath=&SizeBox
   parent=Overlay_processing
   widget=SizeBox_48
   properties {bOverride_HeightOverride=true, heightOverride=20}
   slot {class=OverlaySlot, properties {horizontalAlignment=HAlign_Fill}}
  }
  ProgressBar {
   widgetClassPath=/Script/UMG.ProgressBar
   bIsVariable=true
   parent=SizeBox_48
   widget=ProgressBar
   properties {
    percent=1
    fillColorAndOpacity=<rgba 0.0358 0.0699 0.104 1>
    widgetStyle {
     enableFillAnimation=true
     backgroundImage {
      bIsDynamicallyLoaded=false
      drawAs=Image
      imageType=NoImage
      mirroring=NoMirror
      resourceName=None
      resourceObject=null
      tiling=NoTile
      imageSize=<0 0>
      margin=<ltrb 0 0 0 0>
      outlineSettings {
       bUseBrushTransparency=false
       roundingType=HalfHeightRadius
       width=0
       color {colorUseRule=UseColor_Specified, specifiedColor=<rgba 0 0 0 0>}
       cornerRadii=<0 0 0 0>
      }
      tintColor {colorUseRule=UseColor_Specified, specifiedColor=<rgba 0.0048 0.0048 0.0048 1>}
      uVRegion {bIsValid=false, max=<0 0>, min=<0 0>}
     }
     fillImage {
      bIsDynamicallyLoaded=false
      drawAs=Image
      imageType=FullColor
      mirroring=NoMirror
      resourceName=../../../Engine/Content/Slate/Starship/CoreWidgets/ProgressBar/ProgressMarquee.png
      resourceObject=null
      tiling=Horizontal
      imageSize=<20 12>
      margin=<ltrb 0 0 0 0>
      outlineSettings {
       bUseBrushTransparency=false
       roundingType=HalfHeightRadius
       width=0
       color {colorUseRule=UseColor_Specified, specifiedColor=<rgba 0 0 0 0>}
       cornerRadii=<0 0 0 0>
      }
      tintColor {colorUseRule=UseColor_Specified, specifiedColor=<rgba 1 1 1 1>}
      uVRegion {bIsValid=false, max=<0 0>, min=<0 0>}
     }
     marqueeImage {
      bIsDynamicallyLoaded=false
      drawAs=Image
      imageType=FullColor
      mirroring=NoMirror
      resourceName=../../../Engine/Content/Slate/Starship/CoreWidgets/ProgressBar/ProgressMarquee.png
      resourceObject=null
      tiling=Horizontal
      imageSize=<20 12>
      margin=<ltrb 0 0 0 0>
      outlineSettings {
       bUseBrushTransparency=false
       roundingType=HalfHeightRadius
       width=0
       color {colorUseRule=UseColor_Specified, specifiedColor=<rgba 0 0 0 0>}
       cornerRadii=<0 0 0 0>
      }
      tintColor {colorUseRule=UseColor_Specified, specifiedColor=<rgba 0.0044 0.2384 1 1>}
      uVRegion {bIsValid=false, max=<0 0>, min=<0 0>}
     }
    }
   }
   slot {class=SizeBoxSlot}
  }
  CommonTextBlock_processing {
   widgetClassPath=&CommonTextBlock
   bIsVariable=true
   parent=Overlay_processing
   widget=CommonTextBlock_processing
   properties {
    justification=Center
    text=Processing ....
    font {
     bForceMonospaced=false
     bMaterialIsStencil=false
     fontMaterial=null
     fontObject=&NotoSans
     letterSpacing=0
     monospacedWidth=1
     size=12
     skewAmount=0
     typefaceFontName=NotoSans-Regular
     outlineSettings {
      bApplyOutlineToDropShadows=false
      bMiteredCorners=false
      bSeparateFillAlpha=false
      outlineMaterial=null
      outlineSize=0
      outlineColor=<rgba 0 0 0 1>
     }
    }
    renderTransform {angle=0, scale=<1 1>, shear=<0 0>, translation=<-3 0>}
    shadowOffset=<0 0>
   }
   slot {class=OverlaySlot, properties {horizontalAlignment=HAlign_Center, padding=<ltrb 10 20 10 10>}}
  }
  VerticalBox_stack {
   widgetClassPath=&VerticalBox
   bIsVariable=true
   parent=HorizontalBox_184
   widget=VerticalBox_stack
   slot {class=HorizontalBoxSlot}
  }
  SizeBox_57 {
   widgetClassPath=&SizeBox
   parent=HorizontalBox_184
   widget=SizeBox_57
   properties {bOverride_MinDesiredWidth=true, minDesiredWidth=30}
   slot {class=HorizontalBoxSlot}
  }
  VerticalBox_61 {widgetClassPath=&VerticalBox, parent=SizeBox_57, widget=VerticalBox_61, slot {class=SizeBoxSlot}}
  SizeBox_139 {
   widgetClassPath=&SizeBox
   parent=VerticalBox_61
   widget=SizeBox_139
   properties {bOverride_HeightOverride=true, bOverride_WidthOverride=true, heightOverride=30, widthOverride=30}
   slot {class=VerticalBoxSlot, properties {horizontalAlignment=HAlign_Center, padding=<ltrb 0 5 0 0>}}
  }
  Button_reload {
   widgetClassPath=&Button
   parent=SizeBox_139
   widget=Button_reload
   properties {
    toolTipText=Reload Playlist
    widgetStyle {
     clickedSlateSound {resourceObject=null}
     disabled {
      bIsDynamicallyLoaded=false
      drawAs=RoundedBox
      imageType=NoImage
      mirroring=NoMirror
      resourceName=None
      resourceObject=null
      tiling=NoTile
      imageSize=<0 0>
      margin=<ltrb 0 0 0 0>
      outlineSettings {
       bUseBrushTransparency=false
       roundingType=FixedRadius
       width=1
       color {colorUseRule=UseColor_ColorTable, specifiedColor=<rgba 0 0 0 0>}
       cornerRadii=<4 4 4 4>
      }
      tintColor {colorUseRule=UseColor_Specified, specifiedColor=<rgba 0.0395 0.0395 0.0395 1>}
      uVRegion {bIsValid=false, max=<0 0>, min=<0 0>}
     }
     disabledForeground {colorUseRule=UseColor_ColorTable, specifiedColor=<rgba 0 0 0 0>}
     hovered {
      bIsDynamicallyLoaded=false
      drawAs=RoundedBox
      imageType=NoImage
      mirroring=NoMirror
      resourceName=None
      resourceObject=null
      tiling=NoTile
      imageSize=<0 0>
      margin=<ltrb 0 0 0 0>
      outlineSettings {
       bUseBrushTransparency=false
       roundingType=FixedRadius
       width=1
       color {colorUseRule=UseColor_ColorTable, specifiedColor=<rgba 0 0 0 0>}
       cornerRadii=<4 4 4 4>
      }
      tintColor {colorUseRule=UseColor_Specified, specifiedColor=<rgba 0.0953 0.0953 0.0953 1>}
      uVRegion {bIsValid=false, max=<0 0>, min=<0 0>}
     }
     hoveredForeground {colorUseRule=UseColor_ColorTable, specifiedColor=<rgba 0 0 0 0>}
     hoveredSlateSound {resourceObject=null}
     normal {
      bIsDynamicallyLoaded=false
      drawAs=RoundedBox
      imageType=NoImage
      mirroring=NoMirror
      resourceName=None
      resourceObject=&/T_assetPlaylist_reload
      tiling=NoTile
      imageSize=<19 19>
      margin=<ltrb 0 0 0 0>
      outlineSettings {
       bUseBrushTransparency=false
       roundingType=FixedRadius
       width=1
       color {colorUseRule=UseColor_ColorTable, specifiedColor=<rgba 0 0 0 0>}
       cornerRadii=<4 4 4 4>
      }
      tintColor {colorUseRule=UseColor_Specified, specifiedColor=<rgba 1 1 1 1>}
      uVRegion {bIsValid=false, max=<0 0>, min=<0 0>}
     }
     normalForeground {colorUseRule=UseColor_ColorTable, specifiedColor=<rgba 0 0 0 0>}
     normalPadding=<ltrb 12 1.5 12 1.5>
     pressed {
      bIsDynamicallyLoaded=false
      drawAs=RoundedBox
      imageType=NoImage
      mirroring=NoMirror
      resourceName=None
      resourceObject=null
      tiling=NoTile
      imageSize=<0 0>
      margin=<ltrb 0 0 0 0>
      outlineSettings {
       bUseBrushTransparency=false
       roundingType=FixedRadius
       width=1
       color {colorUseRule=UseColor_ColorTable, specifiedColor=<rgba 0 0 0 0>}
       cornerRadii=<4 4 4 4>
      }
      tintColor {colorUseRule=UseColor_Specified, specifiedColor=<rgba 0.0284 0.0284 0.0284 1>}
      uVRegion {bIsValid=false, max=<0 0>, min=<0 0>}
     }
     pressedForeground {colorUseRule=UseColor_ColorTable, specifiedColor=<rgba 0 0 0 0>}
     pressedPadding=<ltrb 12 2 12 0>
     pressedSlateSound {resourceObject=null}
    }
   }
   slot {class=SizeBoxSlot}
  }
  SizeBox {
   widgetClassPath=&SizeBox
   parent=VerticalBox_61
   widget=SizeBox
   properties {bOverride_HeightOverride=true, bOverride_WidthOverride=true, heightOverride=30, widthOverride=30}
   slot {class=VerticalBoxSlot, properties {horizontalAlignment=HAlign_Center, padding=<ltrb 0 5 0 0>}}
  }
  Button_spawn {
   widgetClassPath=&Button
   parent=SizeBox
   widget=Button_spawn
   properties {
    toolTipText=Reload Playlist
    widgetStyle {
     clickedSlateSound {resourceObject=null}
     disabled {
      bIsDynamicallyLoaded=false
      drawAs=RoundedBox
      imageType=NoImage
      mirroring=NoMirror
      resourceName=None
      resourceObject=null
      tiling=NoTile
      imageSize=<0 0>
      margin=<ltrb 0 0 0 0>
      outlineSettings {
       bUseBrushTransparency=false
       roundingType=FixedRadius
       width=1
       color {colorUseRule=UseColor_ColorTable, specifiedColor=<rgba 0 0 0 0>}
       cornerRadii=<4 4 4 4>
      }
      tintColor {colorUseRule=UseColor_Specified, specifiedColor=<rgba 0.0395 0.0395 0.0395 1>}
      uVRegion {bIsValid=false, max=<0 0>, min=<0 0>}
     }
     disabledForeground {colorUseRule=UseColor_ColorTable, specifiedColor=<rgba 0 0 0 0>}
     hovered {
      bIsDynamicallyLoaded=false
      drawAs=RoundedBox
      imageType=NoImage
      mirroring=NoMirror
      resourceName=None
      resourceObject=null
      tiling=NoTile
      imageSize=<0 0>
      margin=<ltrb 0 0 0 0>
      outlineSettings {
       bUseBrushTransparency=false
       roundingType=FixedRadius
       width=1
       color {colorUseRule=UseColor_ColorTable, specifiedColor=<rgba 0 0 0 0>}
       cornerRadii=<4 4 4 4>
      }
      tintColor {colorUseRule=UseColor_Specified, specifiedColor=<rgba 0.0953 0.0953 0.0953 1>}
      uVRegion {bIsValid=false, max=<0 0>, min=<0 0>}
     }
     hoveredForeground {colorUseRule=UseColor_ColorTable, specifiedColor=<rgba 0 0 0 0>}
     hoveredSlateSound {resourceObject=null}
     normal {
      bIsDynamicallyLoaded=false
      drawAs=RoundedBox
      imageType=NoImage
      mirroring=NoMirror
      resourceName=None
      resourceObject=&/T_assetPlaylist_spawn
      tiling=NoTile
      imageSize=<26 26>
      margin=<ltrb 0 0 0 0>
      outlineSettings {
       bUseBrushTransparency=false
       roundingType=FixedRadius
       width=1
       color {colorUseRule=UseColor_ColorTable, specifiedColor=<rgba 0 0 0 0>}
       cornerRadii=<4 4 4 4>
      }
      tintColor {colorUseRule=UseColor_Specified, specifiedColor=<rgba 1 1 1 1>}
      uVRegion {bIsValid=false, max=<0 0>, min=<0 0>}
     }
     normalForeground {colorUseRule=UseColor_ColorTable, specifiedColor=<rgba 0 0 0 0>}
     normalPadding=<ltrb 12 1.5 12 1.5>
     pressed {
      bIsDynamicallyLoaded=false
      drawAs=RoundedBox
      imageType=NoImage
      mirroring=NoMirror
      resourceName=None
      resourceObject=null
      tiling=NoTile
      imageSize=<0 0>
      margin=<ltrb 0 0 0 0>
      outlineSettings {
       bUseBrushTransparency=false
       roundingType=FixedRadius
       width=1
       color {colorUseRule=UseColor_ColorTable, specifiedColor=<rgba 0 0 0 0>}
       cornerRadii=<4 4 4 4>
      }
      tintColor {colorUseRule=UseColor_Specified, specifiedColor=<rgba 0.0284 0.0284 0.0284 1>}
      uVRegion {bIsValid=false, max=<0 0>, min=<0 0>}
     }
     pressedForeground {colorUseRule=UseColor_ColorTable, specifiedColor=<rgba 0 0 0 0>}
     pressedPadding=<ltrb 12 2 12 0>
     pressedSlateSound {resourceObject=null}
    }
   }
   slot {class=SizeBoxSlot}
  }
  EUW_AssetPlaylist_headless {
   widgetClassPath=&/EUW_AssetPlaylist_headless
   bIsVariable=true
   parent=CanvasPanel_0
   widget=EUW_AssetPlaylist_headless
   slot {class=CanvasPanelSlot, properties {layoutData {alignment=<0 0>, anchors {maximum=<0 0>, minimum=<0 0>}, offsets=<ltrb 4 -4 100 30>}}}
  }
 ]
}
